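# Control Rig for the male medium ZenCrash body that adds corrective deformation to its jacket and collar.
# A shared DeformRig function runs first, then a sequence applies weighted bone offsets.
# Jacket weights come from a thigh_l spherical pose reader while Crouching, crouch variables, and pelvis height remapped through a curve.
# Collar weights come from a head pose reader and Head_Left/Head_Right variables, with most weights passing through AlphaInterp.
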ControlRigBlueprint /BRCosmetics/Characters/Player/Male/Medium/Bodies/M_MED_ZenCrash/Meshes/M_MED_ZenCrash_CtrlRig
graphs [
 RigVMFunctionLibrary {links [], nodes []}
 RigVMFunction_Sequence_2_ContainedGraph {nodes [
   n1 = RigVMFunctionEntryNode()
   n2 = &RigVMFunction_Sequence(ExecuteContext=@n1.ExecuteContext)
   n3 = &RigVMFunction_Sequence(ExecuteContext=@n2.B)
   n4 = &RigVMFunction_Sequence(ExecuteContext=@n3.B)
   n5 = &RigVMFunction_Sequence(ExecuteContext=@n4.B)
   n6 = &RigVMFunction_Sequence(ExecuteContext=@n5.B)
   n7 = &RigVMFunction_Sequence(ExecuteContext=@n6.B)
   RigVMFunctionReturnNode(A=@n2.A, B=@n3.A, C=@n4.A, D=@n5.A, E=@n6.A, F=@n7.A, G=@n7.B)
  ]}
 RigVMModel {nodes [
   RigUnit_BeginExecution = &RigUnit_BeginExecution()
   FN_DeformRig_FNC = RigVMFunctionReferenceNode(
    DeformRig_ Curve="(deformRig_l_upperarm_up_alpha_72_#1=False,deformRig_r_upperarm_up_alpha_73_#2=False,deformRig_l_upperarm_fr_alpha_74_#3=False,deformRig_r_upperarm_fr_alpha_75_CB51B8 ...(truncated)"
    DeformRig_Bool="(upperarm_1_#4=True,elbow_3_#5=True,wrist_7_#6=True,groin_8_#7=True,glute_17_#8=true,knee_9_40C408324CA261 ...(truncated)"
    DeformRig_Debug=(upperarm_debug_18_#4=False,wrist_debug_20_#6=False,lat_debug_26_#9=False,groinGlute_debug_27_#7=False)
    DeformRig_Value="(upperarm_up_18_#10=(Rotation=(X=0,Y=0,Z=0,W=1),Translation=(X=-5,Y=0,Z=5),Scale3D=(X=0.7,Y=1,Z=1)),upperarm_fwd_fr_19_#11=(Rotation=(X=0,Y=0,Z=0,W=1),Translation=(X=0,Y=10,Z=0),Scale3D=(X=0.5,Y=1,Z ...(truncated)"
    ExecuteContext=@RigUnit_BeginExecution.ExecutePin
   )
   > RigVMFunction_Sequence_1 = RigVMAggregateNode()
   n1 = &RigVMFunction_ControlFlowBranch(ExecuteContext=@RigVMFunction_Sequence_1.A, Condition=RigVMVariableNode() {variable=Crouching}.Value)
   n2 = &RigUnit_SphericalPoseReader(
    ActiveRegionScaleFactors=(PositiveWidth=1.2,NegativeWidth=1.2,PositiveHeight=1.2,NegativeHeight=1.2)
    ActiveRegionSize=0.15
    Debug=(bDrawDebug=false,bDraw2D=False,bDrawLocalAxes=False,DebugScale=25,DebugSegments=20,DebugThickness=0.25)
    DriverAxis=(X=-1,Y=0,Z=0)
    DriverItem=(Type=Bone,Name="thigh_l")
    FalloffRegionScaleFactors=(PositiveWidth=1.2,NegativeWidth=4,PositiveHeight=1.2,NegativeHeight=1.2)
    FalloffSize=0.3
    FlipHeightScaling=False
    FlipWidthScaling=False
    OptionalParentItem=(Type=Bone,Name="pelvis")
    RotationOffset=(X=119.3239,Y=0,Z=-2.8942)
    ExecutePin=@n1.True
   )
   AlphaInterp = &RigVMFunction_AlphaInterp(InterpSpeedDecreasing=4, InterpSpeedIncreasing=10, Scale=1, bClampResult=False, bInterpResult=False, Value=@n2.OutputParam)
   > &RigUnit_ModifyTransforms(
    ItemToModify="((Item=(Type=Bone,Name=\"base_l_front_jacket_01\"),Transform=(Rotation=(X=0.8009,Y=-0.1799,Z=0.0335,W=0.5701),Translation=(X=0,Y=0,Z=0),Scale3D=(X=1,Y=1,Z=1))),(Item=(Type=Bone,Name=\"base_l_side_jacket_01\"),Transform=(Rotation=(X=0.6055,Y=0.0228,Z=0.3016,W=0 ...(truncated)"
    WeightMinimum=0
    Weight=@AlphaInterp.Result
   )
   n3 = &RigVMFunction_AlphaInterp(
    InterpSpeedDecreasing=5
    InterpSpeedIncreasing=10
    Scale=1
    bClampResult=true
    Value=RigVMVariableNode() {variable=Crouch_jacket_ADJ_float}.Value
   )
   &RigUnit_ModifyTransforms(
    ItemToModify="((Item=(Type=Bone,Name=\"base_l_front_jacket_01\"),Transform=(Rotation=(X=0,Y=-0.2588,Z=0,W=0.9659),Translation=(X=0,Y=0,Z=0),Scale3D=(X=1,Y=1,Z=1))),(Item=(Type=Bone,Name=\"base_l_side_jacket_01\"),Transform=(Rotation=(X=0.0338,Y=-0.1261,Z=0.2566,W=0.9577),Tr ...(truncated)"
    WeightMinimum=0
    Weight=@n3.Result
    ExecutePin=@RigVMFunction_Sequence_1.B
   )
   n4 = &RigVMFunction_AlphaInterp(
    InterpSpeedDecreasing=5
    InterpSpeedIncreasing=10
    Scale=1
    bClampResult=true
    Value=RigVMVariableNode() {variable=Crouch_jacket_up_ADJ_bool}.Value
   )
   &RigUnit_ModifyTransforms(
    ItemToModify="((Item=(Type=Bone,Name=\"base_l_front_jacket_01\"),Transform=(Rotation=(X=0.0302,Y=0.171,Z=-0.171,W=0.9698),Translation=(X=0,Y=-2,Z=0),Scale3D=(X=1,Y=1,Z=1))),(Item=(Type=Bone,Name=\"base_l_side_jacket_01\"),Transform=(Rotation=(X=0.067,Y=-0.25,Z=0.25,W=0.933) ...(truncated)"
    WeightMinimum=0
    Weight=@n4.Result
    ExecutePin=@RigVMFunction_Sequence_1.C
   )
   n5 = &RigVMFunction_AnimEvalRichCurve(
    Curve=(EditorCurveData=(Keys=((),(Time=0.5,Value=1),(Time=0.8))))
    SourceMaximum=2
    SourceMinimum=0
    TargetMinimum=0
    Value=&RigVMFunction_MathDoubleRemap(
     SourceMaximum=75
     SourceMinimum=45
     TargetMaximum=1
     TargetMinimum=0
     bClamp=false
     Value=&RigUnit_GetRelativeTransformForItem(Child=(Type=Bone,Name="pelvis"), Parent=(Type=Bone,Name="root"), bChildInitial=False, bParentInitial=False).RelativeTransform.Translation.Z
    ).Result
   )
   RigVMVariableNode(Value=@n5.Result, ExecuteContext=@RigVMFunction_Sequence_1.D) {variable=Crouch_jacket_walk_ADJ}
   n6 = &RigVMFunction_AlphaInterp(
    InterpSpeedDecreasing=10
    InterpSpeedIncreasing=10
    Scale=1
    bClampResult=False
    bInterpResult=False
    Value=RigVMVariableNode() {variable=Crouch_jacket_walk_ADJ}.Value
   )
   > &RigUnit_ModifyTransforms(
    ItemToModify="((Item=(Type=Bone,Name=\"base_l_front_jacket_01\"),Transform=(Rotation=(X=0.0449,Y=-0.1677,Z=0.2549,W=0.9513),Translation=(X=0,Y=0,Z=0),Scale3D=(X=1,Y=1,Z=1))),(Item=(Type=Bone,Name=\"base_l_side_jacket_01\"),Transform=(Rotation=(X=0.0302,Y=-0.171,Z=0.171,W=0. ...(truncated)"
    WeightMinimum=0
    Weight=@n6.Result
   )
   n7 = &RigUnit_SphericalPoseReader(
    ActiveRegionScaleFactors=(PositiveWidth=1,NegativeWidth=1,PositiveHeight=1,NegativeHeight=1)
    ActiveRegionSize=0.1
    Debug=(bDrawDebug=True,bDraw2D=False,bDrawLocalAxes=False,DebugScale=25,DebugSegments=20,DebugThickness=0.25)
    DriverAxis=(X=1,Y=0,Z=0)
    DriverItem=(Type=Bone,Name="head")
    FalloffRegionScaleFactors=(PositiveWidth=1,NegativeWidth=2,PositiveHeight=2,NegativeHeight=1)
    FalloffSize=0.2
    FlipHeightScaling=False
    FlipWidthScaling=False
    OptionalParentItem=(Type=Bone,Name="spine_05")
    RotationOffset=(X=90,Y=0,Z=-140)
    ExecutePin=@RigVMFunction_Sequence_1.E
   )
   > RigVMVariableNode(Value=@n7.OutputParam) {variable=Collar_ADJ}
   n8 = &RigVMFunction_AlphaInterp(InterpSpeedDecreasing=10, InterpSpeedIncreasing=10, Scale=1, bClampResult=true, Value=RigVMVariableNode() {variable=Collar_ADJ}.Value)
   > &RigUnit_ModifyTransforms(
    ItemToModify="((Item=(Type=Bone,Name=\"dyn_back_collar_01\"),Transform=(Rotation=(X=0,Y=-0.3827,Z=0,W=0.9239),Translation=(X=0,Y=0,Z=0),Scale3D=(X=0.5,Y=1,Z=1))),(Item=(Type=Bone,Name=\"dyn_back_collar_02\"),Transform=(Rotation=(X=0,Y=-0.3827,Z=0,W=0.9239),Translation=(X=0, ...(truncated)"
    WeightMinimum=0
    Weight=@n8.Result
   )
   > &RigUnit_ModifyTransforms(
    ItemToModify="((Item=(Type=Bone,Name=\"dyn_back_collar_01_l\"),Transform=(Rotation=(X=0,Y=0,Z=0,W=1),Translation=(X=0,Y=0,Z=0),Scale3D=(X=0.7,Y=1,Z=1))),(Item=(Type=Bone,Name=\"dyn_back_collar_02\"),Transform=(Rotation=(X=0,Y=0,Z=0,W=1),Translation=(X=0,Y=0,Z=0),Scale3D=(X= ...(truncated)"
    WeightMinimum=0
    Weight=@n8.Result
   )
   > &RigUnit_ModifyTransforms(
    ItemToModify="((Item=(Type=Bone,Name=\"dyn_back_collar_01_r\"),Transform=(Rotation=(X=0,Y=0,Z=0,W=1),Translation=(X=0,Y=0,Z=0),Scale3D=(X=0.7,Y=1,Z=1))),(Item=(Type=Bone,Name=\"dyn_back_collar_02_r\"),Transform=(Rotation=(X=0,Y=0,Z=0,W=1),Translation=(X=0,Y=0,Z=0),Scale3D=( ...(truncated)"
    WeightMinimum=0
    Weight=@n8.Result
   )
   &RigUnit_ModifyTransforms(
    ItemToModify="((Item=(Type=Bone,Name=\"dyn_back_collar_01_r\"),Transform=(Rotation=(X=0,Y=-0.3827,Z=0,W=0.9239),Translation=(X=0,Y=0,Z=0),Scale3D=(X=1,Y=1,Z=1))),(Item=(Type=Bone,Name=\"dyn_back_collar_02_r\"),Transform=(Rotation=(X=0,Y=0,Z=0,W=1),Translation=(X=0,Y=0,Z=0), ...(truncated)"
    WeightMinimum=0
    ExecutePin=@RigVMFunction_Sequence_1.G
    Weight=RigVMVariableNode() {variable=Head_Right}.Value
   )
   &RigUnit_ModifyTransforms(
    ItemToModify="((Item=(Type=Bone,Name=\"dyn_back_collar_01_l\"),Transform=(Rotation=(X=0,Y=-0.3827,Z=0,W=0.9239),Translation=(X=0,Y=0,Z=0),Scale3D=(X=1,Y=1,Z=1))),(Item=(Type=Bone,Name=\"dyn_back_collar_02_l\"),Transform=(Rotation=(X=0,Y=0,Z=0,W=1),Translation=(X=0,Y=0,Z=0), ...(truncated)"
    WeightMinimum=0
    Weight=RigVMVariableNode() {variable=Head_Left}.Value
    ExecutePin=@RigVMFunction_Sequence_1.F
   )
  ]}
]
hierarchy [
 root {loc=<0 0 0>, rot=<0 0 0 1>}
 pelvis {parent=root, loc=<0 -0.7071 93.1191>, rot=<0.0127 0.707 -0.0127 -0.707>}
 spine_01 {parent=pelvis, loc=<6.7494 -0.6388 0>, rot=<0 0 0.034 -0.9994>}
 spine_02 {parent=spine_01, loc=<7.2518 -0.035 0>, rot=<0 0 0.0238 -0.9997>}
 spine_03 {parent=spine_02, loc=<6.0104 0 0>, rot=<0 0 0.0149 -0.9999>}
 spine_04 {parent=spine_03, loc=<8.1024 0 0>, rot=<0 0 -0.0132 -0.9999>}
 spine_05 {parent=spine_04, loc=<10.0146 0 0>, rot=<0 0 -0.0015 -1>}
 neck_01 {parent=spine_05, loc=<11.3138 0.1529 0>, rot=<0 0 -0.2003 -0.9797>}
 neck_02 {parent=neck_01, loc=<4.5627 0 0>, rot=<0 0 0.0504 -0.9987>}
 head {parent=neck_02, loc=<5.3914 0 0>, rot=<0 0 0.115 -0.9934>}
 FACIAL_C_FacialRoot {parent=head, loc=<2.1524 -0.746 -0.0002>, rot=<0.4879 -0.5118 -0.4879 -0.5118>}
 FACIAL_C_Forehead {parent=FACIAL_C_FacialRoot, loc=<-0.0022 -3.4997 9.3065>, rot=<-0.0452 0.0001 -0.0001 -0.999>}
 FACIAL_L_ForeheadIn {parent=FACIAL_C_FacialRoot, loc=<1.63 -3.4976 9.1656>, rot=<-0.0625 -0.0695 0.004 -0.9956>}
 FACIAL_R_ForeheadIn {parent=FACIAL_C_FacialRoot, loc=<-1.6378 -3.5 9.1648>, rot=<-0.0625 0.0697 -0.0042 -0.9956>}
 FACIAL_L_ForeheadMid {parent=FACIAL_C_FacialRoot, loc=<3.1137 -3.7568 8.7225>, rot=<-0.0727 -0.1755 -0.021 -0.9816>}
 FACIAL_R_ForeheadMid {parent=FACIAL_C_FacialRoot, loc=<-3.1254 -3.763 8.7224>, rot=<-0.0728 0.1757 0.0208 -0.9815>}
 FACIAL_L_ForeheadOut {parent=FACIAL_C_FacialRoot, loc=<4.5216 -3.7502 7.7456>, rot=<-0.1231 -0.2554 -0.0797 -0.9557>}
 FACIAL_R_ForeheadOut {parent=FACIAL_C_FacialRoot, loc=<-4.5342 -3.7569 7.7414>, rot=<-0.1231 0.2555 0.0795 -0.9556>}
 FACIAL_L_EyesackUpper {parent=FACIAL_C_FacialRoot, loc=<3.2714 -2.6565 9.2278>, rot=<0.047 -0.084 -0.0648 -0.9932>}
 FACIAL_R_EyesackUpper {parent=FACIAL_C_FacialRoot, loc=<-3.2864 -2.663 9.2297>, rot=<0.047 0.0842 0.0646 -0.9932>}
 FACIAL_L_EyelidUpperFurrow {parent=FACIAL_C_FacialRoot, loc=<3.1342 -2.4607 9.1714>, rot=<-0.0318 -0.1506 -0.08 -0.9848>}
 FACIAL_R_EyelidUpperFurrow {parent=FACIAL_C_FacialRoot, loc=<-3.1397 -2.4668 9.1759>, rot=<-0.0318 0.1508 0.0798 -0.9848>}
 FACIAL_L_EyelidUpperA {parent=FACIAL_C_FacialRoot, loc=<2.7799 -1.9286 7.2696>, rot=<-0.1496 0.0001 -0.0001 -0.9887>}
 FACIAL_R_EyelidUpperA {parent=FACIAL_C_FacialRoot, loc=<-2.7774 -1.9282 7.2724>, rot=<-0.1496 0.0001 -0.0001 -0.9887>}
 FACIAL_L_Eye {parent=FACIAL_C_FacialRoot, loc=<2.7723 -1.9351 7.2654>, rot=<-0.0018 0.0001 -0.0001 -1>}
 FACIAL_R_Eye {parent=FACIAL_C_FacialRoot, loc=<-2.7723 -1.9345 7.2672>, rot=<-0.0018 0.0001 -0.0001 -1>}
 FACIAL_L_EyelidLowerA {parent=FACIAL_C_FacialRoot, loc=<2.7799 -1.9286 7.2696>, rot=<0.2823 0.0001 -0.0001 -0.9593>}
 FACIAL_R_EyelidLowerA {parent=FACIAL_C_FacialRoot, loc=<-2.7774 -1.9282 7.2724>, rot=<0.2823 0.0001 -0.0001 -0.9593>}
 FACIAL_L_EyeCornerInner {parent=FACIAL_C_FacialRoot, loc=<1.7699 -1.3478 8.65>, rot=<-0.0018 0.0001 -0.0001 -1>}
 FACIAL_R_EyeCornerInner {parent=FACIAL_C_FacialRoot, loc=<-1.7718 -1.3511 8.6476>, rot=<-0.0018 0.0001 -0.0001 -1>}
 FACIAL_L_EyeCornerOuter {parent=FACIAL_C_FacialRoot, loc=<4.2948 -1.7773 7.7952>, rot=<-0.0016 -0.3825 -0.0008 -0.9239>}
 FACIAL_R_EyeCornerOuter {parent=FACIAL_C_FacialRoot, loc=<-4.2959 -1.777 7.7917>, rot=<-0.0017 0.3827 0.0006 -0.9239>}
 FACIAL_L_EyesackLower {parent=FACIAL_C_FacialRoot, loc=<3.3683 -0.464 8.0047>, rot=<-0.024 -0.1396 -0.0226 -0.9897>}
 FACIAL_R_EyesackLower {parent=FACIAL_C_FacialRoot, loc=<-3.3701 -0.4646 7.9966>, rot=<-0.0241 0.1397 0.0224 -0.9896>}
 FACIAL_L_CheekInner {parent=FACIAL_C_FacialRoot, loc=<2.0908 0.3027 8.1779>, rot=<-0.1249 -0.0876 -0.1974 -0.9684>}
 FACIAL_R_CheekInner {parent=FACIAL_C_FacialRoot, loc=<-2.0903 0.3099 8.1837>, rot=<-0.125 0.0877 0.1972 -0.9684>}
 FACIAL_L_CheekOuter {parent=FACIAL_C_FacialRoot, loc=<4.0563 0.3854 6.3626>, rot=<0.0536 -0.377 0.0996 -0.9193>}
 FACIAL_R_CheekOuter {parent=FACIAL_C_FacialRoot, loc=<-4.0515 0.3995 6.374>, rot=<0.0535 0.3772 -0.0998 -0.9192>}
 FACIAL_L_NasolabialBulge {parent=FACIAL_C_FacialRoot, loc=<3.4462 3.6291 6.466>, rot=<0.0563 -0.3773 0.023 -0.9241>}
 FACIAL_R_NasolabialBulge {parent=FACIAL_C_FacialRoot, loc=<-3.4503 3.643 6.4978>, rot=<0.0562 0.3775 -0.0232 -0.924>}
 FACIAL_L_NasolabialFurrow {parent=FACIAL_C_FacialRoot, loc=<2.9982 2.4448 8.74>, rot=<-0.0349 -0.1251 -0.3639 -0.9223>}
 FACIAL_R_NasolabialFurrow {parent=FACIAL_C_FacialRoot, loc=<-2.9931 2.4761 8.7596>, rot=<-0.035 0.1252 0.3638 -0.9224>}
 FACIAL_L_CheekLower {parent=FACIAL_C_FacialRoot, loc=<3.9147 3.3848 5.7126>, rot=<0.033 -0.4269 0.0569 -0.9019>}
 FACIAL_R_CheekLower {parent=FACIAL_C_FacialRoot, loc=<-3.925 3.3951 5.7474>, rot=<0.0329 0.4271 -0.0571 -0.9018>}
 FACIAL_C_Nose {parent=FACIAL_C_FacialRoot, loc=<-0.0015 -0.3081 9.5998>, rot=<-0.1321 0.0001 -0.0001 -0.9912>}
 FACIAL_C_NoseLower {parent=FACIAL_C_Nose, loc=<0 2.9475 -0.2734>, rot=<0.1305 0 0 -0.9914>}
 FACIAL_L_Nostril {parent=FACIAL_C_Nose, loc=<1.0369 2.3056 -0.5814>, rot=<0.0173 0.1305 0.0023 -0.9913>}
 FACIAL_R_Nostril {parent=FACIAL_C_Nose, loc=<-1.0324 2.3176 -0.5841>, rot=<0.0173 -0.1305 -0.0023 -0.9913>}
 FACIAL_C_NoseTip {parent=FACIAL_C_Nose, loc=<0.0022 2.065 0.4579>, rot=<0.0872 0 0 -0.9962>}
 FACIAL_C_MouthUpper {parent=FACIAL_C_FacialRoot, loc=<-0.0054 3.8425 7.3728>, rot=<-0.0018 0.0001 -0.0001 -1>}
 FACIAL_C_LipUpper {parent=FACIAL_C_MouthUpper, loc=<0.0062 -0.1793 2.433>, rot=<-0.0872 0 0 -0.9962>}
 FACIAL_L_LipUpper {parent=FACIAL_C_MouthUpper, loc=<1.1975 -0.0289 2.0992>, rot=<-0.0778 -0.1601 -0.0379 -0.9833>}
 FACIAL_R_LipUpper {parent=FACIAL_C_MouthUpper, loc=<-1.1896 -0.0348 2.1466>, rot=<-0.0778 0.1601 0.0379 -0.9833>}
 FACIAL_L_LipUpperOuter {parent=FACIAL_C_MouthUpper, loc=<1.9978 0.2289 1.3368>, rot=<-0.0984 -0.2293 -0.0983 -0.9634>}
 FACIAL_R_LipUpperOuter {parent=FACIAL_C_MouthUpper, loc=<-1.9739 0.2292 1.4272>, rot=<-0.0984 0.2293 0.0983 -0.9634>}
 FACIAL_L_LipCorner {parent=FACIAL_C_MouthUpper, loc=<2.3278 0.3933 0.721>, rot=<0 -0.2078 0 -0.9782>}
 FACIAL_R_LipCorner {parent=FACIAL_C_MouthUpper, loc=<-2.3285 0.4104 0.7972>, rot=<0 0.2078 0 -0.9782>}
 FACIAL_L_Masseter {parent=FACIAL_C_FacialRoot, loc=<6.2522 1.1462 4.0918>, rot=<0.0676 0.113 0.008 -0.9913>}
 FACIAL_R_Masseter {parent=FACIAL_C_FacialRoot, loc=<-6.2532 1.1479 4.0974>, rot=<0.0676 -0.1128 -0.0082 -0.9913>}
 FACIAL_C_UnderChin {parent=FACIAL_C_FacialRoot, loc=<-0.0008 8.08 5.0275>, rot=<0.4983 0.0001 0 -0.867>}
 FACIAL_L_UnderChin {parent=FACIAL_C_FacialRoot, loc=<3.2486 6.9683 3.3816>, rot=<0.4011 -0.3549 0.0816 -0.8406>}
 FACIAL_R_UnderChin {parent=FACIAL_C_FacialRoot, loc=<-3.2313 6.9479 3.3957>, rot=<0.401 0.3551 -0.0817 -0.8405>}
 FACIAL_C_TeethUpper {parent=FACIAL_C_FacialRoot, loc=<0.0029 4.2025 9.022>, rot=<-0.0018 0.0001 -0.0001 -1>}
 FACIAL_C_LowerLipRotation {parent=FACIAL_C_FacialRoot, loc=<-0.0013 1.6664 2.2627>, rot=<0.3404 0.0001 -0.0001 -0.9403>}
 FACIAL_C_MouthLower {parent=FACIAL_C_LowerLipRotation, loc=<-0.0028 -1.599 5.319>, rot=<-0.342 0 0 -0.9397>}
 FACIAL_C_LipLower {parent=FACIAL_C_MouthLower, loc=<0.0061 1.3195 2.0015>, rot=<0.0872 0 0 -0.9962>}
 FACIAL_L_LipLower {parent=FACIAL_C_MouthLower, loc=<1.0989 1.2455 1.7017>, rot=<0.0927 -0.1294 0.0272 -0.9869>}
 FACIAL_R_LipLower {parent=FACIAL_C_MouthLower, loc=<-1.0933 1.2417 1.7867>, rot=<0.0927 0.1294 -0.0272 -0.9869>}
 FACIAL_L_LipLowerOuter {parent=FACIAL_C_MouthLower, loc=<1.9067 0.8073 1.113>, rot=<0.0913 -0.2321 0.0318 -0.9679>}
 FACIAL_R_LipLowerOuter {parent=FACIAL_C_MouthLower, loc=<-1.9069 0.8265 1.2122>, rot=<0.0913 0.2321 -0.0318 -0.9679>}
 FACIAL_C_TeethLower {parent=FACIAL_C_LowerLipRotation, loc=<0.0043 -2.3113 6.7501>, rot=<-0.342 0 0 -0.9397>}
 FACIAL_C_Tongue1 {parent=FACIAL_C_TeethLower, loc=<-0.015 0.1861 -5.1649>, rot=<-0.2588 0 0 -0.9659>}
 FACIAL_C_Tongue2 {parent=FACIAL_C_Tongue1, loc=<0.0045 0.413 1.5268>, rot=<0.2588 0 0 -0.9659>}
 FACIAL_C_Tongue3 {parent=FACIAL_C_Tongue2, loc=<0.0025 0.2203 1.3518>, rot=<0.1305 0 0 -0.9914>}
 FACIAL_C_Tongue4 {parent=FACIAL_C_Tongue3, loc=<0.002 -0.1835 1.1461>, rot=<0.0436 0 0 -0.999>}
 FACIAL_C_TongueUpper3 {parent=FACIAL_C_Tongue3, loc=<0.0002 -0.383 -0.1229>, rot=<0 0 0 1>}
 FACIAL_C_TongueLower3 {parent=FACIAL_C_Tongue3, loc=<-0.0002 0.8269 0.0903>, rot=<0 0 0 1>}
 FACIAL_L_TongueSide3 {parent=FACIAL_C_Tongue3, loc=<1.6206 0.0552 0.1529>, rot=<0 0 0 1>}
 FACIAL_R_TongueSide3 {parent=FACIAL_C_Tongue3, loc=<-1.6171 0.0503 0.1813>, rot=<0 0 0 1>}
 FACIAL_C_TongueUpper2 {parent=FACIAL_C_Tongue2, loc=<0.0008 -0.5675 0.0523>, rot=<0 0 0 1>}
 FACIAL_L_TongueSide2 {parent=FACIAL_C_Tongue2, loc=<1.8841 0.1931 0.381>, rot=<0 0 0 1>}
 FACIAL_R_TongueSide2 {parent=FACIAL_C_Tongue2, loc=<-1.8769 0.1945 0.4036>, rot=<0 0 0 1>}
 FACIAL_C_Jaw {parent=FACIAL_C_FacialRoot, loc=<-0.0013 1.6664 2.2627>, rot=<0.3404 0.0001 -0.0001 -0.9403>}
 FACIAL_C_Jawline {parent=FACIAL_C_Jaw, loc=<-0.0056 2.0886 7.6859>, rot=<0.2586 0 0 -0.966>}
 FACIAL_L_ChinSide {parent=FACIAL_C_Jaw, loc=<2.9225 0.1168 6.8154>, rot=<-0.2208 -0.3738 0.0499 -0.8995>}
 FACIAL_R_ChinSide {parent=FACIAL_C_Jaw, loc=<-2.9148 0.1176 6.8458>, rot=<-0.2208 0.3738 -0.0499 -0.8995>}
 FACIAL_C_Chin {parent=FACIAL_C_Jaw, loc=<-0.0102 0.5992 7.6533>, rot=<-0.2591 0 0 -0.9659>}
 FACIAL_L_Ear {parent=FACIAL_C_FacialRoot, loc=<6.6955 -0.639 1.4864>, rot=<-0.0281 -0.0005 -0.0263 -0.9993>}
 FACIAL_R_Ear {parent=FACIAL_C_FacialRoot, loc=<-6.6935 -0.6384 1.4832>, rot=<-0.0281 0.0007 0.0261 -0.9993>}
 FACIAL_L_Jawline {parent=FACIAL_C_FacialRoot, loc=<4.328 5.0077 4.5591>, rot=<0.2076 0.1579 -0.0544 -0.9638>}
 FACIAL_R_Jawline {parent=FACIAL_C_FacialRoot, loc=<-4.3089 4.988 4.5728>, rot=<0.2076 -0.1577 0.0543 -0.9639>}
 FACIAL_C_NoseBridge {parent=FACIAL_C_FacialRoot, loc=<-0.0008 -2.2169 10.0195>, rot=<0.042 0.0001 -0.0001 -0.9991>}
 FACIAL_L_EyelidUpperB {parent=FACIAL_C_FacialRoot, loc=<2.7799 -1.9286 7.2696>, rot=<-0.2267 0.0001 -0.0001 -0.974>}
 FACIAL_L_EyelidLowerB {parent=FACIAL_C_FacialRoot, loc=<2.7799 -1.9286 7.2696>, rot=<0.3156 0.0001 -0.0001 -0.9489>}
 FACIAL_R_EyelidLowerB {parent=FACIAL_C_FacialRoot, loc=<-2.7774 -1.9282 7.2724>, rot=<0.3239 0.0001 -0.0001 -0.9461>}
 FACIAL_R_EyelidUpperB {parent=FACIAL_C_FacialRoot, loc=<-2.7774 -1.9282 7.2724>, rot=<-0.2267 0.0001 -0.0001 -0.974>}
 attach_cape {parent=spine_05, loc=<9.017 -11.714 0>, rot=<0.0616 -0.7044 -0.0616 -0.7044>}
 cape {parent=attach_cape, type=Null, loc=<0 0 0>, rot=<0 0 0 1>}
 attach_backpack {parent=spine_05, loc=<-0.123 -13.473 0>, rot=<-0.0093 -0.707 0.0093 -0.707>}
 Backpack_BR {parent=attach_backpack, type=Null, loc=<0 -1.4118 0>, rot=<0 0 0 1>}
 clavicle_l {parent=spine_05, loc=<4.3575 -0.1466 -2.3223>, rot=<0.0465 -0.7289 -0.009 -0.6829>}
 upperarm_l {parent=clavicle_l, loc=<15.5111 0 0>, rot=<0.0226 -0.3234 -0.0088 -0.9459>}
 lowerarm_l {parent=upperarm_l, loc=<26.6514 0 0>, rot=<0.0391 -0.0099 -0.2453 -0.9686>}
 hand_l {parent=lowerarm_l, loc=<22.3491 0 0>, rot=<-0.7081 -0.0507 0.0975 -0.6975>}
 weapon_l {parent=hand_l, loc=<-0.1974 -3.4111 0.3811>, rot=<0 0 0 1>}
 thumb_01_l {parent=hand_l, loc=<1.3118 -0.3191 -2.459>, rot=<0.6546 -0.4042 0.0664 -0.6354>}
 thumb_02_l {parent=thumb_01_l, loc=<5.9336 -0.0448 0.0401>, rot=<-0.0037 0.0213 0.298 -0.9543>}
 thumb_03_l {parent=thumb_02_l, loc=<2.2223 0.3682 -0.0907>, rot=<0.0118 -0.013 0.0066 -0.9998>}
 index_metacarpal_l {parent=hand_l, loc=<1.6535 0.0183 -2.96>, rot=<0.0158 -0.0124 -0.0173 -0.9996>}
 index_01_l {parent=index_metacarpal_l, loc=<8.3008 0.0357 -0.0524>, rot=<-0.0044 0.0485 0.0924 -0.9945>}
 index_02_l {parent=index_01_l, loc=<3.3587 -0.3004 0.0452>, rot=<-0.002 0.0027 0.3225 -0.9466>}
 index_03_l {parent=index_02_l, loc=<2.5578 -0.123 -0.0252>, rot=<-0.002 0.0067 0.2558 -0.9667>}
 middle_metacarpal_l {parent=hand_l, loc=<2.1856 -0.1725 -0.5984>, rot=<0.006 0.0168 -0.0329 -0.9993>}
 middle_01_l {parent=middle_metacarpal_l, loc=<7.3275 -0.2447 0.0129>, rot=<-0.0208 0.0042 0.0858 -0.9961>}
 middle_02_l {parent=middle_01_l, loc=<4.0123 0.1401 0.1175>, rot=<-0.0218 0.0258 0.3512 -0.9357>}
 middle_03_l {parent=middle_02_l, loc=<2.6077 0.0534 0.1373>, rot=<0.0082 0.0188 0.1708 -0.9851>}
 ring_metacarpal_l {parent=hand_l, loc=<2.2074 -0.3625 1.3638>, rot=<-0.0284 0.0722 -0.0458 -0.9959>}
 ring_01_l {parent=ring_metacarpal_l, loc=<7.0295 -0.6197 -0.1446>, rot=<-0.0489 -0.0252 0.1892 -0.9804>}
 ring_02_l {parent=ring_01_l, loc=<3.6174 0.4647 -0.1027>, rot=<0 0 0.2806 -0.9598>}
 ring_03_l {parent=ring_02_l, loc=<2.4152 0.114 0.0199>, rot=<0 0 0.1189 -0.9929>}
 pinky_metacarpal_l {parent=hand_l, loc=<1.7616 -0.9753 3.3814>, rot=<-0.0397 0.0814 -0.069 -0.9935>}
 pinky_01_l {parent=pinky_metacarpal_l, loc=<7.1709 -0.4831 -0.1094>, rot=<-0.0654 -0.0035 0.2242 -0.9723>}
 pinky_02_l {parent=pinky_01_l, loc=<2.8274 0.0893 -0.1043>, rot=<0.0037 0.0011 0.2854 -0.9584>}
 pinky_03_l {parent=pinky_02_l, loc=<2.2046 0.0307 -0.0078>, rot=<-0.0002 0.0098 0.1445 -0.9895>}
 hand_attach_l {parent=hand_l, loc=<8.5 -3 0>, rot=<0.7071 0.7071 0 0>}
 deform_wrist_dn_base_l {parent=hand_l, loc=<0 -2.2342 0>, rot=<0 0 0 -1>}
 deform_wrist_dn_l {parent=deform_wrist_dn_base_l, loc=<0 0 0>, rot=<0 0 0 1>}
 deform_wrist_up_base_l {parent=hand_l, loc=<0 2.2682 0>, rot=<0 0 0 -1>}
 deform_wrist_up_l {parent=deform_wrist_up_base_l, loc=<0 0 0>, rot=<0 0 0 1>}
 lowerarm_twist_02_l {parent=lowerarm_l, loc=<9.6199 0 0>, rot=<0 0 0 -1>}
 lowerarm_twist_01_l {parent=lowerarm_l, loc=<17.8549 0 0>, rot=<0 0 0 -1>}
 deform_elbow_in_base_l {parent=lowerarm_l, loc=<0.2112 5.259 0.8274>, rot=<0 0 0 -1>}
 deform_elbow_in_l {parent=deform_elbow_in_base_l, loc=<0 0 0>, rot=<0 0 0 1>}
 deform_elbow_out_base_l {parent=lowerarm_l, loc=<0 -4 0>, rot=<0 0 0 -1>}
 deform_elbow_out_l {parent=deform_elbow_out_base_l, loc=<0 0 0>, rot=<0 0 0 1>}
 base_bracelate_01 {parent=lowerarm_l, loc=<20.1727 -0.0012 0.4332>, rot=<-0.09 -0.0941 0.0633 -0.9895>}
 dyn_bracelate_01 {parent=base_bracelate_01, loc=<0 0 0>, rot=<0 0 0 1>}
 upperarm_twist_01_l {parent=upperarm_l, loc=<0.0041 0 0>, rot=<0 0 0 -1>}
 deform_upperarm_bk_base_l {parent=upperarm_twist_01_l, loc=<1.0813 -7.2193 -1.9453>, rot=<0 0 0 1>}
 deform_upperarm_bk_l {parent=deform_upperarm_bk_base_l, loc=<0 0 0>, rot=<0 0 0 1>}
 deform_upperarm_fr_base_l {parent=upperarm_twist_01_l, loc=<-0.3691 7.0506 -2.765>, rot=<0 0 0 1>}
 deform_upperarm_fr_l {parent=deform_upperarm_fr_base_l, loc=<0 0 0>, rot=<0 0 0 1>}
 deform_upperarm_up_base_l {parent=upperarm_twist_01_l, loc=<-3.2216 -1.1391 3.3311>, rot=<0 0 0 1>}
 deform_upperarm_up_l {parent=deform_upperarm_up_base_l, loc=<0 0 0>, rot=<0 0 0 1>}
 upperarm_twist_02_l {parent=upperarm_l, loc=<13.7941 0 0>, rot=<0 0 0 -1>}
 clavicle_r {parent=spine_05, loc=<4.3575 -0.1466 2.3223>, rot=<0.7289 0.0465 -0.6829 0.009>}
 upperarm_r {parent=clavicle_r, loc=<-15.5111 0 0>, rot=<0.0226 -0.3234 -0.0088 -0.9459>}
 lowerarm_r {parent=upperarm_r, loc=<-26.6514 0 0>, rot=<0.0391 -0.0099 -0.2453 -0.9686>}
 hand_r {parent=lowerarm_r, loc=<-22.3491 0 0>, rot=<-0.7081 -0.0507 0.0975 -0.6975>}
 weapon_r {parent=hand_r, loc=<0.1974 3.4111 -0.3811>, rot=<0 0 0 1>}
 thumb_01_r {parent=hand_r, loc=<-1.3118 0.3191 2.459>, rot=<0.6546 -0.4042 0.0664 -0.6354>}
 thumb_02_r {parent=thumb_01_r, loc=<-5.9336 0.0448 -0.0401>, rot=<-0.0037 0.0213 0.298 -0.9543>}
 thumb_03_r {parent=thumb_02_r, loc=<-2.2223 -0.3682 0.0907>, rot=<0.0118 -0.013 0.0066 -0.9998>}
 index_metacarpal_r {parent=hand_r, loc=<-1.6535 -0.0183 2.96>, rot=<0.0158 -0.0124 -0.0173 -0.9996>}
 index_01_r {parent=index_metacarpal_r, loc=<-8.3008 -0.0357 0.0524>, rot=<-0.0044 0.0485 0.0924 -0.9945>}
 index_02_r {parent=index_01_r, loc=<-3.3587 0.3004 -0.0452>, rot=<-0.002 0.0027 0.3225 -0.9466>}
 index_03_r {parent=index_02_r, loc=<-2.5578 0.123 0.0252>, rot=<-0.002 0.0067 0.2558 -0.9667>}
 middle_metacarpal_r {parent=hand_r, loc=<-2.1856 0.1725 0.5984>, rot=<0.006 0.0168 -0.0329 -0.9993>}
 middle_01_r {parent=middle_metacarpal_r, loc=<-7.3275 0.2447 -0.0129>, rot=<-0.0208 0.0042 0.0858 -0.9961>}
 middle_02_r {parent=middle_01_r, loc=<-4.0123 -0.1401 -0.1175>, rot=<-0.0218 0.0258 0.3512 -0.9357>}
 middle_03_r {parent=middle_02_r, loc=<-2.6077 -0.0534 -0.1373>, rot=<0.0082 0.0188 0.1708 -0.9851>}
 ring_metacarpal_r {parent=hand_r, loc=<-2.2074 0.3625 -1.3638>, rot=<-0.0284 0.0722 -0.0458 -0.9959>}
 ring_01_r {parent=ring_metacarpal_r, loc=<-7.0295 0.6197 0.1446>, rot=<-0.0489 -0.0252 0.1892 -0.9804>}
 ring_02_r {parent=ring_01_r, loc=<-3.6174 -0.4647 0.1027>, rot=<0 0 0.2806 -0.9598>}
 ring_03_r {parent=ring_02_r, loc=<-2.4152 -0.114 -0.0199>, rot=<0 0 0.1189 -0.9929>}
 pinky_metacarpal_r {parent=hand_r, loc=<-1.7616 0.9753 -3.3814>, rot=<-0.0397 0.0814 -0.069 -0.9935>}
 pinky_01_r {parent=pinky_metacarpal_r, loc=<-7.1709 0.4831 0.1094>, rot=<-0.0654 -0.0035 0.2242 -0.9723>}
 pinky_02_r {parent=pinky_01_r, loc=<-2.8274 -0.0893 0.1043>, rot=<0.0037 0.0011 0.2854 -0.9584>}
 pinky_03_r {parent=pinky_02_r, loc=<-2.2046 -0.0307 0.0078>, rot=<-0.0002 0.0098 0.1445 -0.9895>}
 hand_attach_r {parent=hand_r, loc=<-8.5001 2.9999 0>, rot=<0 0 -0.7071 -0.7071>}
 deform_wrist_dn_base_r {parent=hand_r, loc=<-0.0004 2.2345 -0.0001>, rot=<0 0 0 -1>}
 deform_wrist_dn_r {parent=deform_wrist_dn_base_r, loc=<0 0 0>, rot=<0 0 0 1>}
 deform_wrist_up_base_r {parent=hand_r, loc=<0.0003 -2.2684 0.0001>, rot=<0 0 0 -1>}
 deform_wrist_up_r {parent=deform_wrist_up_base_r, loc=<0 0 0>, rot=<0 0 0 1>}
 lowerarm_twist_02_r {parent=lowerarm_r, loc=<-9.6199 0 0>, rot=<0 0 0 -1>}
 lowerarm_twist_01_r {parent=lowerarm_r, loc=<-17.8549 0 0>, rot=<0 0 0 -1>}
 deform_elbow_in_base_r {parent=lowerarm_r, loc=<-0.2111 -5.2591 -0.8275>, rot=<0 0 0 -1>}
 deform_elbow_in_r {parent=deform_elbow_in_base_r, loc=<-0.0001 0 0.0002>, rot=<0 0 0 1>}
 deform_elbow_out_base_r {parent=lowerarm_r, loc=<0 4 -0.0001>, rot=<0 0 0 -1>}
 deform_elbow_out_r {parent=deform_elbow_out_base_r, loc=<0 0 0>, rot=<0 0 0 1>}
 upperarm_twist_01_r {parent=upperarm_r, loc=<-0.0041 0 0>, rot=<0 0 0 -1>}
 deform_upperarm_bk_base_r {parent=upperarm_twist_01_r, loc=<-1.0814 7.2192 1.9455>, rot=<0 0 0 1>}
 deform_upperarm_bk_r {parent=deform_upperarm_bk_base_r, loc=<0 0 0>, rot=<0 0 0 1>}
 deform_upperarm_fr_base_r {parent=upperarm_twist_01_r, loc=<0.3689 -7.0506 2.7652>, rot=<0 0 0 1>}
 deform_upperarm_fr_r {parent=deform_upperarm_fr_base_r, loc=<0 0 0>, rot=<0 0 0 1>}
 deform_upperarm_up_base_r {parent=upperarm_twist_01_r, loc=<3.2214 1.1391 -3.3308>, rot=<0 0 0 1>}
 deform_upperarm_up_r {parent=deform_upperarm_up_base_r, loc=<0 0 0>, rot=<0 0 0 -1>}
 upperarm_twist_02_r {parent=upperarm_r, loc=<-13.7941 0 0>, rot=<0 0 0 -1>}
 base_necklace_main_01 {parent=spine_05, loc=<11.3138 0.1529 0>, rot=<0.2736 0.6492 0.6523 0.2796>}
 dyn_necklace_main_01 {parent=base_necklace_main_01, loc=<0 0 0>, rot=<0 0 0 1>}
 base_necklace_01 {parent=dyn_necklace_main_01, loc=<11.6373 0.0815 1.3213>, rot=<0 -0.0787 0 -0.9969>}
 dyn_necklace_01 {parent=base_necklace_01, loc=<0 0 0>, rot=<0 0 0 1>}
 dyn_necklace_02 {parent=dyn_necklace_01, loc=<3.048 0 0.1212>, rot=<0.0003 -0.0381 -0.0037 -0.9993>}
 base_back_collar_01_l {parent=spine_05, loc=<14.2442 -3.1169 -7.0849>, rot=<0.9165 0.2133 0.2096 0.2659>}
 dyn_back_collar_01_l {parent=base_back_collar_01_l, loc=<0 0 0>, rot=<0 0 0 1>}
 dyn_back_collar_02_l {parent=dyn_back_collar_01_l, loc=<3.915 0 0>, rot=<0 0 0 1>}
 base_back_collar_01 {parent=spine_05, loc=<14.9358 -4.781 0>, rot=<-0.6881 -0.1626 -0.1626 -0.6881>}
 dyn_back_collar_01 {parent=base_back_collar_01, loc=<0 0 0>, rot=<0 0 0 1>}
 dyn_back_collar_02 {parent=dyn_back_collar_01, loc=<3.915 0 0>, rot=<0 0 0 1>}
 base_back_collar_01_r {parent=spine_05, loc=<14.2446 -3.1169 7.0849>, rot=<-0.2659 -0.2096 -0.2133 -0.9165>}
 dyn_back_collar_01_r {parent=base_back_collar_01_r, loc=<0 0 0>, rot=<0 0 0 1>}
 dyn_back_collar_02_r {parent=dyn_back_collar_01_r, loc=<3.915 0 0>, rot=<0 0 0 1>}
 deform_pec_aim_base_l {parent=spine_04, loc=<7.8732 12.6388 -2.4123>, rot=<0.0102 -0.4602 0.3979 -0.7936>}
 deform_pec_aim_l {parent=deform_pec_aim_base_l, loc=<0 0 0>, rot=<0 0 0 1>}
 deform_pec_base_l {parent=deform_pec_aim_l, loc=<10.766 0 0>, rot=<0 0 0 1>}
 deform_pec_l {parent=deform_pec_base_l, loc=<0.055 0 0>, rot=<0 0 0 1>}
 deform_pec_aim_base_r {parent=spine_04, loc=<7.8727 12.6388 2.4123>, rot=<-0.4602 -0.0102 0.7936 0.3979>}
 deform_pec_aim_r {parent=deform_pec_aim_base_r, loc=<0 0 0>, rot=<0 0 0 1>}
 deform_pec_base_r {parent=deform_pec_aim_r, loc=<-10.7662 -0.0006 -0.0007>, rot=<0 0 0 1>}
 deform_pec_r {parent=deform_pec_base_r, loc=<0 0 0>, rot=<0 0 0 1>}
 base_l_front_jacket_01 {parent=spine_02, loc=<10.3302 10.5963 -8.8745>, rot=<0.0881 -0.8465 -0.5248 0.0142>}
 dyn_l_front_jacket_01 {parent=base_l_front_jacket_01, loc=<0 0 0>, rot=<0 0 0 1>}
 dyn_l_front_jacket_02 {parent=dyn_l_front_jacket_01, loc=<7.4479 0.0374 -0.02>, rot=<0 -0.0057 -0.0107 -0.9999>}
 dyn_l_front_jacket_03 {parent=dyn_l_front_jacket_02, loc=<8.3766 0 0>, rot=<0 0 0 1>}
 base_l_side_jacket_01 {parent=spine_02, loc=<7.8869 4.4513 -14.0699>, rot=<0.0466 -0.9722 -0.2266 -0.0367>}
 dyn_l_side_jacket_01 {parent=base_l_side_jacket_01, loc=<0 0 0>, rot=<0 0 0 1>}
 dyn_l_side_jacket_02 {parent=dyn_l_side_jacket_01, loc=<5.7594 -0.108 -0.213>, rot=<-0.0009 0.0301 -0.0049 -0.9995>}
 dyn_l_side_jacket_03 {parent=dyn_l_side_jacket_02, loc=<5.7245 -0.15 -0.129>, rot=<0 0 0 1>}
 base_l_left_jacket_01 {parent=spine_02, loc=<5.6051 -2.8179 -16.3822>, rot=<0.0513 -0.9971 0.052 -0.0204>}
 dyn_l_left_jacket_01 {parent=base_l_left_jacket_01, loc=<0 0 0>, rot=<0 0 0 1>}
 dyn_l_left_jacket_02 {parent=dyn_l_left_jacket_01, loc=<4.2041 0.2072 0.0242>, rot=<-0.012 0.0229 0.0088 -0.9996>}
 dyn_l_left_jacket_03 {parent=dyn_l_left_jacket_02, loc=<4.2525 0.0444 -0.0796>, rot=<0 0 0 1>}
 base_back_jacket_01 {parent=spine_02, loc=<7.0737 -12.0054 -0.0943>, rot=<0.0473 -0.7004 0.712 0.0134>}
 dyn_back_jacket_01 {parent=base_back_jacket_01, loc=<0 0 0>, rot=<0 0 0 1>}
 dyn_back_jacket_02 {parent=dyn_back_jacket_01, loc=<3.5646 0.3011 0.0412>, rot=<-0.0124 0.0253 -0.0432 -0.9987>}
 dyn_back_jacket_03 {parent=dyn_back_jacket_02, loc=<4.04 -0.0045 -0.0367>, rot=<0 0 0 1>}
 base_l_back_jacket_01 {parent=spine_02, loc=<6.767 -11.4261 -9.4076>, rot=<-0.0059 -0.8087 0.5865 0.0442>}
 dyn_l_back_jacket_01 {parent=base_l_back_jacket_01, loc=<0 0 0>, rot=<0 0 0 1>}
 dyn_l_back_jacket_02 {parent=dyn_l_back_jacket_01, loc=<4.0459 0.2316 0.1226>, rot=<-0.0114 0.0619 -0.0263 -0.9977>}
 dyn_l_back_jacket_03 {parent=dyn_l_back_jacket_02, loc=<3.6242 -0.0997 -0.0327>, rot=<0 0 0 1>}
 base_r_front_jacket_01 {parent=spine_02, loc=<10.3305 10.5963 8.8745>, rot=<-0.0142 0.5248 0.8465 -0.0881>}
 dyn_r_front_jacket_01 {parent=base_r_front_jacket_01, loc=<0 0 0>, rot=<0 0 0 1>}
 dyn_r_front_jacket_02 {parent=dyn_r_front_jacket_01, loc=<7.4487 -0.0375 -0.02>, rot=<0 -0.0057 0.0107 -0.9999>}
 dyn_r_front_jacket_03 {parent=dyn_r_front_jacket_02, loc=<8.376 0 0>, rot=<0 0 0 1>}
 base_r_side_jacket_01 {parent=spine_02, loc=<7.8865 4.4513 14.0699>, rot=<0.0367 0.2266 0.9722 -0.0466>}
 dyn_r_side_jacket_01 {parent=base_r_side_jacket_01, loc=<0 0 0>, rot=<0 0 0 1>}
 dyn_r_side_jacket_02 {parent=dyn_r_side_jacket_01, loc=<5.7594 0.108 -0.213>, rot=<0.0009 0.0301 0.0049 -0.9995>}
 dyn_r_side_jacket_03 {parent=dyn_r_side_jacket_02, loc=<5.7238 0.15 -0.1289>, rot=<0 0 0 1>}
 base_r_left_jacket_01 {parent=spine_02, loc=<5.6055 -2.8179 16.3822>, rot=<-0.0018 -0.0531 0.9973 -0.0501>}
 dyn_r_left_jacket_01 {parent=base_r_left_jacket_01, loc=<0 0 0>, rot=<0 0 0 -1>}
 dyn_r_left_jacket_02 {parent=dyn_r_left_jacket_01, loc=<4.1993 -0.2072 0.2113>, rot=<0.0122 0.0451 -0.0085 -0.9989>}
 dyn_r_left_jacket_03 {parent=dyn_r_left_jacket_02, loc=<4.2527 -0.0444 -0.0796>, rot=<0 0 0 -1>}
 base_r_back_jacket_01 {parent=spine_02, loc=<6.7671 -11.4261 9.4076>, rot=<-0.0442 -0.5865 0.8087 0.0059>}
 dyn_r_back_jacket_01 {parent=base_r_back_jacket_01, loc=<0 0 0>, rot=<0 0 0 1>}
 dyn_r_back_jacket_02 {parent=dyn_r_back_jacket_01, loc=<4.0458 -0.2316 0.1226>, rot=<0.0114 0.0619 0.0263 -0.9977>}
 dyn_r_back_jacket_03 {parent=dyn_r_back_jacket_02, loc=<3.6244 0.0997 -0.0327>, rot=<0 0 0 1>}
 thigh_l {parent=pelvis, loc=<-2.5693 -0.0109 -7.7501>, rot=<0.0012 0.0572 0.0155 -0.9982>}
 calf_l {parent=thigh_l, loc=<-39.0484 0 0>, rot=<0.0006 -0.0161 -0.0391 -0.9991>}
 foot_l {parent=calf_l, loc=<-42.561 0.1077 0.4292>, rot=<0.0418 -0.0455 0.0343 -0.9975>}
 ball_l {parent=foot_l, loc=<-6.1599 13.8264 0.6806>, rot=<0.0001 -0.0001 -0.7063 -0.7079>}
 base_l_leg_lace_01 {parent=foot_l, loc=<0.6455 7.1224 0.0777>, rot=<0.2662 -0.6804 -0.2561 -0.6329>}
 dyn_l_leg_lace_01 {parent=base_l_leg_lace_01, loc=<0 0 0>, rot=<0 0 0 1>}
 dyn_l_leg_left_lace_01 {parent=dyn_l_leg_lace_01, loc=<2.3272 -0.2139 -0.0269>, rot=<-0.0003 -0.0058 0.0458 -0.9989>}
 dyn_l_leg_right_lace_01 {parent=dyn_l_leg_lace_01, loc=<-2.5483 0.2506 -0.2144>, rot=<0.0058 -0.0003 0.9989 0.0458>}
 dyn_l_leg_left_lace_02 {parent=dyn_l_leg_lace_01, loc=<1.8476 1.2787 0.6861>, rot=<-0.304 -0.1534 -0.4606 -0.8197>}
 dyn_l_leg_right_lace_02 {parent=dyn_l_leg_lace_01, loc=<-1.947 2.1559 0.0426>, rot=<-0.0602 0.0327 0.9034 0.4233>}
 calf_twist_02_l {parent=calf_l, loc=<-18.8064 -0.0967 0.0397>, rot=<0.0237 0.0095 -0.0035 -0.9997>}
 calf_twist_01_l {parent=calf_l, loc=<-35.813 -0.0967 0.0397>, rot=<0.0237 0.0095 -0.0035 -0.9997>}
 deform_knee_in_base_l {parent=calf_l, loc=<1.298 -5.2957 -0.0496>, rot=<0 0 0 -1>}
 deform_knee_in_l {parent=deform_knee_in_base_l, loc=<0 0 0>, rot=<0 0 0 1>}
 deform_knee_out_base_l {parent=calf_l, loc=<0 4 0>, rot=<0 0 0 -1>}
 deform_knee_out_l {parent=deform_knee_out_base_l, loc=<0 0 0>, rot=<0 0 0 1>}
 thigh_twist_01_l {parent=thigh_l, loc=<-21.6303 0 0>, rot=<0.0013 0 0 -1>}
 deform_groin_base_l {parent=thigh_l, loc=<-2.1908 8.1142 -1.7442>, rot=<0 0 0 -1>}
 deform_groin_l {parent=deform_groin_base_l, loc=<0 0 0>, rot=<0 0 0 1>}
 thigh_r {parent=pelvis, loc=<-2.5693 -0.0109 7.7501>, rot=<0.0572 -0.0012 0.9982 0.0155>}
 calf_r {parent=thigh_r, loc=<39.0484 0 0>, rot=<0.0006 -0.0161 -0.0391 -0.9991>}
 foot_r {parent=calf_r, loc=<42.561 -0.1077 -0.4292>, rot=<0.0418 -0.0455 0.0343 -0.9975>}
 ball_r {parent=foot_r, loc=<6.1599 -13.8264 -0.6806>, rot=<0.0001 -0.0001 -0.7063 -0.7079>}
 base_r_leg_lace_01 {parent=foot_r, loc=<-0.6455 -7.1224 -0.0777>, rot=<-0.6329 -0.2561 0.6804 -0.2662>}
 dyn_r_leg_lace_01 {parent=base_r_leg_lace_01, loc=<0 0 0>, rot=<0 0 0 -1>}
 dyn_r_leg_right_lace_01 {parent=dyn_r_leg_lace_01, loc=<2.5483 0.2506 -0.2144>, rot=<-0.0003 0.0058 -0.0458 -0.9989>}
 dyn_r_leg_left_lace_01 {parent=dyn_r_leg_lace_01, loc=<-2.3272 -0.2139 -0.0269>, rot=<-0.0058 -0.0003 0.9989 -0.0458>}
 dyn_r_leg_right_lace_02 {parent=dyn_r_leg_lace_01, loc=<1.947 2.1559 0.0426>, rot=<0.0399 -0.0755 -0.4227 -0.9022>}
 dyn_r_leg_left_lace_02 {parent=dyn_r_leg_lace_01, loc=<-1.8476 1.2787 0.6861>, rot=<-0.0528 -0.2611 0.8637 0.4278>}
 calf_twist_02_r {parent=calf_r, loc=<18.8064 0.0967 -0.0397>, rot=<0.0237 0.0095 -0.0035 -0.9997>}
 calf_twist_01_r {parent=calf_r, loc=<35.813 0.0967 -0.0397>, rot=<0.0237 0.0095 -0.0035 -0.9997>}
 deform_knee_in_base_r {parent=calf_r, loc=<-1.298 5.2957 0.0496>, rot=<0 0 0 -1>}
 deform_knee_in_r {parent=deform_knee_in_base_r, loc=<0 0 0>, rot=<0 0 0 1>}
 deform_knee_out_base_r {parent=calf_r, loc=<0 -4 0>, rot=<0 0 0 -1>}
 deform_knee_out_r {parent=deform_knee_out_base_r, loc=<0 0 0>, rot=<0 0 0 1>}
 thigh_twist_01_r {parent=thigh_r, loc=<21.6303 0 0>, rot=<0.0013 0 0 -1>}
 deform_groin_base_r {parent=thigh_r, loc=<2.1908 -8.1142 1.7442>, rot=<0 0 0 -1>}
 deform_groin_r {parent=deform_groin_base_r, loc=<0 0 0>, rot=<0 0 0 1>}
 deform_glute_base_l {parent=pelvis, loc=<-8.8064 -8.8222 -8.4287>, rot=<0.0275 0.0501 -0.4472 -0.8926>}
 deform_glute_l {parent=deform_glute_base_l, loc=<0 0 0>, rot=<0 0 0 1>}
 deform_glute_base_r {parent=pelvis, loc=<-8.8062 -8.8222 8.4287>, rot=<0.0501 -0.0275 0.8926 -0.4472>}
 deform_glute_r {parent=deform_glute_base_r, loc=<0 0 0>, rot=<0 0 0 1>}
 base_l_pant_string_01 {parent=pelvis, loc=<5.8484 9.4158 -1.8969>, rot=<0.056 0.7049 0.7049 0.056>}
 dyn_l_pant_string_01 {parent=base_l_pant_string_01, loc=<0 0 0>, rot=<0 0 0 1>}
 dyn_l_pant_string_02 {parent=dyn_l_pant_string_01, loc=<4.8118 0 0>, rot=<0 -0.0512 0 -0.9987>}
 dyn_l_pant_string_03 {parent=dyn_l_pant_string_02, loc=<5.0701 0 0>, rot=<-0.0001 -0.0111 0 -0.9999>}
 dyn_l_pant_string_04 {parent=dyn_l_pant_string_03, loc=<5.051 0 0>, rot=<0 0 0 1>}
 base_r_pant_string_01 {parent=pelvis, loc=<5.8484 9.4158 1.8969>, rot=<-0.056 -0.7049 -0.7049 -0.056>}
 dyn_r_pant_string_01 {parent=base_r_pant_string_01, loc=<0 0 0>, rot=<0 0 0 1>}
 dyn_r_pant_string_02 {parent=dyn_r_pant_string_01, loc=<4.8118 0 0>, rot=<0 -0.0512 0 -0.9987>}
 dyn_r_pant_string_03 {parent=dyn_r_pant_string_02, loc=<5.07 0 0>, rot=<0.0001 -0.0111 0 -0.9999>}
 dyn_r_pant_string_04 {parent=dyn_r_pant_string_03, loc=<5.051 0 0>, rot=<0 0 0 1>}
 thigh_l_null {parent=pelvis, type=Null, loc=<-2.5693 -0.0109 -7.7501>, rot=<0.0012 0.0572 0.0155 -0.9982>}
 attach {parent=root, loc=<0 0 0>, rot=<0 0 0 1>}
 ik_hand_root {parent=root, loc=<0 0 0>, rot=<0 0 0 1>}
 ik_hand_gun {parent=ik_hand_root, loc=<-51.9727 4.0883 102.6935>, rot=<0.6084 0.1793 0.4613 -0.6205>}
 ik_hand_r {parent=ik_hand_gun, loc=<0 0 0>, rot=<0 0 0 1>}
 ik_hand_l {parent=ik_hand_gun, loc=<53.0285 82.1781 35.2051>, rot=<-0.5102 -0.7906 -0.3387 0>}
 ik_foot_root {parent=root, loc=<0 0 0>, rot=<0 0 0 1>}
 ik_foot_l {parent=ik_foot_root, loc=<15.2608 -4.4176 9.4432>, rot=<0.0357 0.7045 0.0284 -0.7082>}
 ik_foot_r {parent=ik_foot_root, loc=<-15.2608 -4.4176 9.4432>, rot=<0.7082 0.0284 -0.7045 0.0357>}
 root_fp {parent=root, loc=<0 0 0>, rot=<0 0 0 1>}
 attach_fp {parent=root_fp, loc=<0 0 0>, rot=<0 0 0 1>}
 camera_root {parent=root_fp, loc=<0 0 0>, rot=<0 0 0 1>}
 camera {parent=camera_root, loc=<0 0 0>, rot=<0 0 0 1>}
 attach_emote_01 {parent=root, loc=<0 0 0>, rot=<0 0 0 1>}
 attach_emote_02 {parent=root, loc=<0 0 0>, rot=<0 0 0 1>}
 attach_emote_03 {parent=root, loc=<0 0 0>, rot=<0 0 0 1>}
 warp_point_01 {parent=root, loc=<0 0 0>, rot=<0 0 0 1>}
 warp_point_02 {parent=root, loc=<0 0 0>, rot=<0 0 0 1>}
 warp_point_03 {parent=root, loc=<0 0 0>, rot=<0 0 0 1>}
]
variables [
 Collar_ADJ {type=double, default=0}
 Collar_ADJ_Bool {type=bool, default=False}
 Crouch_jacket_ADJ_bool {type=bool, default=False}
 Crouch_jacket_ADJ_float {type=double, default=1}
 Crouch_jacket_up_ADJ_bool {type=double, default=0}
 Crouch_jacket_walk_ADJ {type=double, default=1}
 Crouching {type=bool, default=False}
 Crouching_Aim {type=bool, default=False}
 Crouching_Alpha {type=double, default=0}
 Head_Left {type=double, default=0}
 Head_Right {type=double, default=0}
 NewVar {type=double, default=0}
 NewVar_0 {type=bool, default=False}
]
