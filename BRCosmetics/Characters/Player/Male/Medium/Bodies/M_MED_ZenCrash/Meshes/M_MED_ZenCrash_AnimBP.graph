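# Post-process-style cosmetic AnimBP for the M_MED_ZenCrash body part, whose output pose only copies the attached parent mesh.
# A Control Rig, additive jacket bone offsets and AnimDynamics chain for necklace and leg laces exists but is not wired to the output.
# Each update computes gravity override from a parameter asset, crouch alpha, and head look alphas from head-to-spine_05 deltas, suppressed during melee.
# These computed values are not read by the AnimGraph.
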AnimBlueprint /BRCosmetics/Characters/Player/Male/Medium/Bodies/M_MED_ZenCrash/Meshes/M_MED_ZenCrash_AnimBP
target_skeleton=&/M_MED_ZenCrash_Skeleton
graphs [
 AnimGraph {nodes [
   n1 = Control Rig!(Source, Alpha, Crouching:bool=Get bIsCrouching()) -> (Pose) {controlRigAssetReference {blueprintRigClass=&/M_MED_ZenCrash_CtrlRig.M_MED_ZenCrash_CtrlRig_C}}
   n2 = "Transform (Modify) Bone - Bone: dyn_l_front_jacket_02"(ComponentPose=Local To Component(LocalPose=@n1), Alpha=1, Translation="0,0,0", Rotation="7,0,0", Scale="1,1,1") {
    rotationMode=BMM_Additive
    rotationSpace=BCS_BoneSpace
    translationMode=BMM_Additive
    translationSpace=BCS_BoneSpace
    boneToModify {boneName=dyn_l_front_jacket_02}
    rotation {pitch=7}
   }
   n3 = "Transform (Modify) Bone - Bone: dyn_l_side_jacket_02"(ComponentPose=@n2, Alpha=1, Translation="0,0,0", Rotation="5,0,-30", Scale="1,1,1") {
    rotationMode=BMM_Additive
    rotationSpace=BCS_BoneSpace
    translationMode=BMM_Additive
    translationSpace=BCS_BoneSpace
    boneToModify {boneName=dyn_l_side_jacket_02}
    rotation {pitch=5, roll=-30}
   }
   n4 = "Transform (Modify) Bone - Bone: dyn_l_side_jacket_02"(ComponentPose=@n3, Alpha=1, Translation="0,0,0", Rotation="20,0,0", Scale="1,1,1") {
    rotationMode=BMM_Additive
    rotationSpace=BCS_BoneSpace
    translationMode=BMM_Additive
    translationSpace=BCS_BoneSpace
    boneToModify {boneName=dyn_l_side_jacket_02}
    rotation {pitch=20}
   }
   n5 = Anim Dynamics - Bone: dyn_necklace_01!(ComponentPose=@n4, Alpha=1, GravityOverride:Vector="0,0,-980", ExternalForce="0,0,0") -> (Pose) {
    angularDampingOverride=1
    angularSpringConstant=5000
    bAngularSpring=true
    bOverrideAngularDamping=true
    bUseGravityOverride=true
    bUseSphericalLimits=true
    gravityScale=0.35
    simulationSpace=RootRelative
    boundBone {boneName=dyn_necklace_01}
    componentAppliedLinearAccClamp=<5000 5000 5000>
    componentLinearAccScale=<0.2 0.2 0.2>
    componentLinearVelScale=<0.2 0.2 0.2>
    gravityOverride {z=-980}
    physicsBodyDefinitions [{
      boundBone {boneName=dyn_necklace_01}
      boxExtents=<15 15 15>
      constraintSetup {angularLimitsMax=<0 0 2>, angularLimitsMin=<0 5 -2>, angularTarget=<1 0 0>}
      localJointOffset=<4 0 0>
     }]
    sphericalLimits [{limitRadius=8, drivingBone {boneName=spine_05}, sphereLocalOffset=<-2.7966 1.7508 0>}]
   }
   n6 = Anim Dynamics - Bone: dyn_necklace_02!(ComponentPose=@n5, Alpha=1, GravityOverride:Vector="0,0,-980", ExternalForce="0,0,0") -> (Pose) {
    angularDampingOverride=1
    angularSpringConstant=5000
    bAngularSpring=true
    bOverrideAngularDamping=true
    bUseGravityOverride=true
    bUseSphericalLimits=true
    gravityScale=0.35
    simulationSpace=RootRelative
    boundBone {boneName=dyn_necklace_02}
    componentAppliedLinearAccClamp=<5000 5000 5000>
    componentLinearAccScale=<0.2 0.2 0.2>
    componentLinearVelScale=<0.2 0.2 0.2>
    gravityOverride {z=-980}
    physicsBodyDefinitions [{
      boundBone {boneName=dyn_necklace_02}
      boxExtents=<15 15 15>
      constraintSetup {angularLimitsMax=<0 0 2>, angularLimitsMin=<0 -10 -2>, angularTarget=<1 0 0>}
      localJointOffset=<4 0 0>
     }]
    sphericalLimits [{limitRadius=8, drivingBone {boneName=spine_05}, sphereLocalOffset=<-3.1343 5.5967 0>}]
   }
   n7 = Anim Dynamics - Bone: dyn_l_leg_left_lace_01!(ComponentPose=@n6, Alpha=1, GravityOverride:Vector="0,0,-980", ExternalForce="0,0,0") -> (Pose) {
    angularDampingOverride=1
    angularSpringConstant=5000
    bAngularSpring=true
    bOverrideAngularDamping=true
    bUseGravityOverride=true
    gravityScale=0.35
    simulationSpace=RootRelative
    boundBone {boneName=dyn_l_leg_left_lace_01}
    componentLinearAccScale=<1 1 1>
    componentLinearVelScale=<1 1 1>
    gravityOverride {z=-980}
    physicsBodyDefinitions [{
      boundBone {boneName=dyn_l_leg_left_lace_01}
      boxExtents=<10 10 10>
      constraintSetup {angularLimitsMax=<0 15 20>, angularLimitsMin=<0 -30 -20>, angularTarget=<1 0 0>}
      localJointOffset=<4 0 0>
     }]
   }
   n8 = Anim Dynamics - Bone: dyn_l_leg_right_lace_01!(ComponentPose=@n7, Alpha=1, GravityOverride:Vector="0,0,-980", ExternalForce="0,0,0") -> (Pose) {
    angularDampingOverride=1
    angularSpringConstant=5000
    bAngularSpring=true
    bOverrideAngularDamping=true
    bUseGravityOverride=true
    gravityScale=0.35
    simulationSpace=RootRelative
    boundBone {boneName=dyn_l_leg_right_lace_01}
    componentLinearAccScale=<1 1 1>
    componentLinearVelScale=<1 1 1>
    gravityOverride {z=-980}
    physicsBodyDefinitions [{
      boundBone {boneName=dyn_l_leg_right_lace_01}
      boxExtents=<10 10 10>
      constraintSetup {angularLimitsMax=<0 15 20>, angularLimitsMin=<0 -30 -20>, angularTarget=<1 0 0>}
      localJointOffset=<4 0 0>
     }]
   }
   n9 = Anim Dynamics - Bone: dyn_l_leg_left_lace_02!(ComponentPose=@n8, Alpha=1, GravityOverride:Vector="0,0,-980", ExternalForce="0,0,0") -> (Pose) {
    angularDampingOverride=1
    angularSpringConstant=5000
    bAngularSpring=true
    bOverrideAngularDamping=true
    bUseGravityOverride=true
    gravityScale=0.35
    simulationSpace=RootRelative
    boundBone {boneName=dyn_l_leg_left_lace_02}
    componentLinearAccScale=<1 1 1>
    componentLinearVelScale=<1 1 1>
    gravityOverride {z=-980}
    physicsBodyDefinitions [{
      boundBone {boneName=dyn_l_leg_left_lace_02}
      boxExtents=<10 10 10>
      constraintSetup {angularLimitsMax=<0 15 20>, angularLimitsMin=<0 -20 -20>, angularTarget=<1 0 0>}
      localJointOffset=<4 0 0>
     }]
   }
   n10 = Anim Dynamics - Bone: dyn_l_leg_right_lace_02!(ComponentPose=@n9, Alpha=1, GravityOverride:Vector="0,0,-980", ExternalForce="0,0,0") -> (Pose) {
    angularDampingOverride=1
    angularSpringConstant=5000
    bAngularSpring=true
    bOverrideAngularDamping=true
    bUseGravityOverride=true
    gravityScale=0.35
    simulationSpace=RootRelative
    boundBone {boneName=dyn_l_leg_right_lace_02}
    componentLinearAccScale=<1 1 1>
    componentLinearVelScale=<1 1 1>
    gravityOverride {z=-980}
    physicsBodyDefinitions [{
      boundBone {boneName=dyn_l_leg_right_lace_02}
      boxExtents=<10 10 10>
      constraintSetup {angularLimitsMax=<0 15 20>, angularLimitsMin=<0 -20 -20>, angularTarget=<1 0 0>}
      localJointOffset=<4 0 0>
     }]
   }
   n11 = Anim Dynamics - Bone: dyn_r_leg_left_lace_01!(ComponentPose=@n10, Alpha=1, GravityOverride:Vector="0,0,-980", ExternalForce="0,0,0") -> (Pose) {
    angularDampingOverride=1
    angularSpringConstant=5000
    bAngularSpring=true
    bOverrideAngularDamping=true
    bUseGravityOverride=true
    gravityScale=0.35
    simulationSpace=RootRelative
    boundBone {boneName=dyn_r_leg_left_lace_01}
    componentLinearAccScale=<1 1 1>
    componentLinearVelScale=<1 1 1>
    gravityOverride {z=-980}
    physicsBodyDefinitions [{
      boundBone {boneName=dyn_r_leg_left_lace_01}
      boxExtents=<10 10 10>
      constraintSetup {angularLimitsMax=<0 15 20>, angularLimitsMin=<0 -30 -20>, angularTarget=<1 0 0>}
      localJointOffset=<4 0 0>
     }]
   }
   n12 = Anim Dynamics - Bone: dyn_r_leg_right_lace_01!(ComponentPose=@n11, Alpha=1, GravityOverride:Vector="0,0,-980", ExternalForce="0,0,0") -> (Pose) {
    angularDampingOverride=1
    angularSpringConstant=5000
    bAngularSpring=true
    bOverrideAngularDamping=true
    bUseGravityOverride=true
    gravityScale=0.35
    simulationSpace=RootRelative
    boundBone {boneName=dyn_r_leg_right_lace_01}
    componentLinearAccScale=<1 1 1>
    componentLinearVelScale=<1 1 1>
    gravityOverride {z=-980}
    physicsBodyDefinitions [{
      boundBone {boneName=dyn_r_leg_right_lace_01}
      boxExtents=<10 10 10>
      constraintSetup {angularLimitsMax=<0 15 20>, angularLimitsMin=<0 -30 -20>, angularTarget=<1 0 0>}
      localJointOffset=<4 0 0>
     }]
   }
   n13 = Anim Dynamics - Bone: dyn_r_leg_left_lace_02!(ComponentPose=@n12, Alpha=1, GravityOverride:Vector="0,0,-980", ExternalForce="0,0,0") -> (Pose) {
    angularDampingOverride=1
    angularSpringConstant=5000
    bAngularSpring=true
    bOverrideAngularDamping=true
    bUseGravityOverride=true
    gravityScale=0.35
    simulationSpace=RootRelative
    boundBone {boneName=dyn_r_leg_left_lace_02}
    componentLinearAccScale=<1 1 1>
    componentLinearVelScale=<1 1 1>
    gravityOverride {z=-980}
    physicsBodyDefinitions [{
      boundBone {boneName=dyn_r_leg_left_lace_02}
      boxExtents=<10 10 10>
      constraintSetup {angularLimitsMax=<0 15 20>, angularLimitsMin=<0 -20 -20>, angularTarget=<1 0 0>}
      localJointOffset=<4 0 0>
     }]
   }
   n14 = Anim Dynamics - Bone: dyn_r_leg_right_lace_02!(ComponentPose=@n13, Alpha=1, GravityOverride:Vector="0,0,-980", ExternalForce="0,0,0") -> (Pose) {
    angularDampingOverride=1
    angularSpringConstant=5000
    bAngularSpring=true
    bOverrideAngularDamping=true
    bUseGravityOverride=true
    gravityScale=0.35
    simulationSpace=RootRelative
    boundBone {boneName=dyn_r_leg_right_lace_02}
    componentLinearAccScale=<1 1 1>
    componentLinearVelScale=<1 1 1>
    gravityOverride {z=-980}
    physicsBodyDefinitions [{
      boundBone {boneName=dyn_r_leg_right_lace_02}
      boxExtents=<10 10 10>
      constraintSetup {angularLimitsMax=<0 15 20>, angularLimitsMin=<0 -20 -20>, angularTarget=<1 0 0>}
      localJointOffset=<4 0 0>
     }]
   }
   Component To Local(ComponentPose=@n14)
   Output Pose(Result=Copy Pose From Mesh() {bUseAttachedParent=true}) {layerGroup=None}
  ]}
 EventGraph {nodes [
   Event BlueprintUpdateAnimation()
   > n1 = Sequence()
   > n2 = "Gravity Override State Function (Parameter Asset) (Nativized)"(
    CharacterPartAnimInstance=Self-Reference()
    GameTimeInSeconds=GetGameTimeInSeconds()
    Parameters=&/M_MED_ZenCrash_Parameters
   )
   > Set Gravity Override(Gravity Override=@n2)
   Set Crouching_Alpha(execute=@n1.then_1, Crouching_Alpha:double="To Float (Boolean)"(InBool=Get bIsCrouching())) -> (Output_Get:double)
   n3 = NOT Boolean(A=Get bIsPlayingMeleeAnim())
   n4 = GetPartSkeletalMeshComponent()
   n5 = float + float(
    A=GetAlphaFromReferenceTransformDelta(SkeletalMeshComponent=@n4, BoneName=head, BaseName=spine_05, InputRange=(X=-25,Y=-50), TwistAxis="1,0,0")
    B=GetAlphaFromReferenceTransformDelta(SkeletalMeshComponent=@n4, BoneName=head, BaseName=spine_05, DeltaComponentType=EulerY, InputRange=(X=5,Y=30), TwistAxis="1,0,0")
   )
   > Set Head look up alpha(Head look up alpha:double=SelectFloat(A="Clamp (Float)"(Value=@n5, Min=0, Max=1), B=0, bPickA=@n3)) -> (Output_Get:double)
   n6 = GetAlphaFromReferenceTransformDelta(
    SkeletalMeshComponent=@n4
    BoneName=head
    BaseName=spine_05
    DeltaComponentType=EulerY
    InputRange=(X=-15,Y=75)
    OutputRange=(X=1,Y=0)
    TwistAxis="1,0,0"
   )
   > Set Head look down alpha(Head look down alpha:double=SelectFloat(A=@n6, B=0, bPickA=@n3)) -> (Output_Get:double)
   n7 = Get bIsCrouching()
   n8 = Get bIsTargeting()
   > Set Crouch_head_down(Crouch_head_down:double=SelectFloat(A=Get Head look down alpha() -> (Head look down alpha:double), B=0, bPickA=AND Boolean(A=@n7, B=@n8))) -> (Output_Get:double)
   Set Machine gun corr(Machine gun corr:double=SelectFloat(A=@n6, B=0, bPickA=false)) -> (Output_Get:double)
  ]}
]
variables [
 Crouch_head_down {type=number}
 Crouching_Alpha {type=number}
 Gravity Override {type=Vector}
 Head look down alpha {type=number}
 Head look up alpha {type=number}
]
