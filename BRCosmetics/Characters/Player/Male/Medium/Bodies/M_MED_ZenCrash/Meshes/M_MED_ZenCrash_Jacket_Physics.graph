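# Physics asset for the ZenCrash male medium body's jacket mesh, rigging two dangling pant-string bone chains, left and right.
# Each chain runs from a small base capsule pinned to the pelvis through four dynamic capsules, with zero twist and tight swing limits that are generally wider toward the tip.
# Pelvis and thigh capsules supply body volume around the strings, with thigh-to-pelvis rotation left free.
# Left-chain capsules are fatter than right-chain ones, and their shape names are swapped between sides.
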PhysicsAsset /BRCosmetics/Characters/Player/Male/Medium/Bodies/M_MED_ZenCrash/Meshes/M_MED_ZenCrash_Jacket_Physics
bodies {
 base_l_pant_string_01 [{shapeType=Capsule, length=0.5, radius=0.5, shapeName=base_l_pant_string_01_capsule}]
 base_r_pant_string_01 [{shapeType=Capsule, length=0.5, radius=0.5, shapeName=base_l_pant_string_01_capsule}]
 dyn_l_pant_string_01 [{shapeType=Capsule, length=3.5, radius=1.7, shapeName=dyn_r_pant_string_01_capsule, center=<4 0 0>, rotation=<rot 90 0 0>}]
 dyn_l_pant_string_02 [{shapeType=Capsule, length=3.5, radius=1.7, shapeName=dyn_r_pant_string_02_capsule, center=<4 0 0>, rotation=<rot 90 0 0>}]
 dyn_l_pant_string_03 [{shapeType=Capsule, length=3.5, radius=1.7, shapeName=dyn_r_pant_string_03_capsule, center=<4 0 0>, rotation=<rot 90 0 0>}]
 dyn_l_pant_string_04 [{shapeType=Capsule, length=3.5, radius=1.7, shapeName=dyn_r_pant_string_04_capsule, center=<4 0 0>, rotation=<rot 90 0 0>}]
 dyn_r_pant_string_01 [{shapeType=Capsule, length=3.5, radius=1.4, shapeName=dyn_l_pant_string_01_capsule, center=<4 0 0>, rotation=<rot -85 180 180>}]
 dyn_r_pant_string_02 [{shapeType=Capsule, length=3.5, radius=1.4, shapeName=dyn_l_pant_string_02_capsule, center=<4 0 0>, rotation=<rot -90 0 0>}]
 dyn_r_pant_string_03 [{shapeType=Capsule, length=3.5, radius=1.4, shapeName=dyn_l_pant_string_03_capsule, center=<4 0 0>, rotation=<rot -90 0 0>}]
 dyn_r_pant_string_04 [{shapeType=Capsule, length=3.5, radius=1.4, shapeName=dyn_l_pant_string_04_capsule, center=<4 0 0>, rotation=<rot -90 0 0>}]
 pelvis [{shapeType=Capsule, length=8.8054, radius=15.0009, shapeName=pelvis_capsule, center=<-9.2199 -4.6445 0.0269>, rotation=<rot 0 179.9998 180>}]
 thigh_l [{shapeType=Capsule, length=25, radius=10, shapeName=thigh_l_capsule, center=<-14.9323 0.8439 0.1915>, rotation=<rot -85.7394 0.6358 -2.1889>}]
 thigh_r [{
   shapeType=Capsule
   length=25
   radius=10
   shapeName=thigh_r_capsule
   center=<14.9323 -0.8439 -0.1915>
   rotation=<rot -85.7394 0.6358 177.8111>
  }]
}
constraints [
 {
  bone1Name=thigh_l
  bone2Name=pelvis
  swing1LimitDegrees=45
  swing1Motion=Free
  swing2LimitDegrees=45
  swing2Motion=Free
  twistLimitDegrees=45
  twistMotion=Free
 }
 {
  bone1Name=thigh_r
  bone2Name=pelvis
  swing1LimitDegrees=45
  swing1Motion=Free
  swing2LimitDegrees=45
  swing2Motion=Free
  twistLimitDegrees=45
  twistMotion=Free
 }
 {bone1Name=base_l_pant_string_01, bone2Name=pelvis, swing1LimitDegrees=0, swing2LimitDegrees=0, twistLimitDegrees=0}
 {bone1Name=dyn_l_pant_string_01, bone2Name=base_l_pant_string_01, swing1LimitDegrees=5, swing2LimitDegrees=5, twistLimitDegrees=0}
 {bone1Name=dyn_l_pant_string_02, bone2Name=dyn_l_pant_string_01, swing1LimitDegrees=7, swing2LimitDegrees=10, twistLimitDegrees=0}
 {bone1Name=dyn_l_pant_string_03, bone2Name=dyn_l_pant_string_02, swing1LimitDegrees=10, swing2LimitDegrees=10, twistLimitDegrees=0}
 {bone1Name=dyn_l_pant_string_04, bone2Name=dyn_l_pant_string_03, swing1LimitDegrees=10, swing2LimitDegrees=15, twistLimitDegrees=0}
 {bone1Name=base_r_pant_string_01, bone2Name=pelvis, swing1LimitDegrees=0, swing2LimitDegrees=0, twistLimitDegrees=0}
 {bone1Name=dyn_r_pant_string_01, bone2Name=base_r_pant_string_01, swing1LimitDegrees=10, swing2LimitDegrees=10, twistLimitDegrees=0}
 {bone1Name=dyn_r_pant_string_02, bone2Name=dyn_r_pant_string_01, swing1LimitDegrees=7, swing2LimitDegrees=10, twistLimitDegrees=0}
 {bone1Name=dyn_r_pant_string_03, bone2Name=dyn_r_pant_string_02, swing1LimitDegrees=10, swing2LimitDegrees=10, twistLimitDegrees=0}
 {bone1Name=dyn_r_pant_string_04, bone2Name=dyn_r_pant_string_03, swing1LimitDegrees=10, swing2LimitDegrees=15, twistLimitDegrees=0}
]
