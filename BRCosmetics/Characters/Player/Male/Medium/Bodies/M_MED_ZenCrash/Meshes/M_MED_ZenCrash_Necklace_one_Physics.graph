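# Physics asset for the necklace on the ZenCrash male medium player cosmetic body, giving the necklace bones capsule bodies for simulated swing.
# Large capsules on spine_04 and spine_05 form the torso volume, with the necklace chained from spine_05 through base and dyn bones.
# Base bones attach with free angular joints, while dyn joints use narrow swing limits and zero twist limit, loosening slightly at the tip.
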PhysicsAsset /BRCosmetics/Characters/Player/Male/Medium/Bodies/M_MED_ZenCrash/Meshes/M_MED_ZenCrash_Necklace_one_Physics
bodies {
 base_necklace_01 [{shapeType=Capsule, length=0.505, radius=0.505, shapeName=base_necklace_01_capsule, rotation=<rot 90 0 0>}]
 base_necklace_main_01 [{shapeType=Capsule, length=0.505, radius=0.505, shapeName=base_necklace_main_01_capsule, rotation=<rot 0 -90 0>}]
 dyn_necklace_01 [{shapeType=Capsule, length=5, radius=2, shapeName=dyn_necklace_01_capsule, center=<4 0 2.1>, rotation=<rot 90 0 0>}]
 dyn_necklace_02 [{shapeType=Capsule, length=6.3056, radius=3.3848, shapeName=dyn_necklace_02_capsule, center=<6.3758 0 1.7285>, rotation=<rot 72.9909 0 0>}]
 dyn_necklace_main_01 [{shapeType=Capsule, length=10, radius=4, shapeName=dyn_necklace_main_01_capsule, center=<11.583 0 5.4424>, rotation=<rot 82.189 0 0>}]
 spine_04 [{shapeType=Capsule, length=17.8904, radius=14.4089, center=<4.8656 -1.2747 0.3443>, rotation=<rot 0 -90 0>}]
 spine_05 [{shapeType=Capsule, length=20.6974, radius=14.4504, center=<0.7939 -4.004 -0.2733>, rotation=<rot 0 -90 0>}]
}
constraints [
 {bone1Name=spine_05, bone2Name=spine_04, swing1LimitDegrees=45, swing2LimitDegrees=45, twistLimitDegrees=45}
 {
  bone1Name=base_necklace_main_01
  bone2Name=spine_05
  swing1LimitDegrees=45
  swing1Motion=Free
  swing2LimitDegrees=45
  swing2Motion=Free
  twistLimitDegrees=45
  twistMotion=Free
 }
 {bone1Name=dyn_necklace_main_01, bone2Name=base_necklace_main_01, swing1LimitDegrees=10, swing2LimitDegrees=20, twistLimitDegrees=0}
 {
  bone1Name=base_necklace_01
  bone2Name=dyn_necklace_main_01
  swing1LimitDegrees=45
  swing1Motion=Free
  swing2LimitDegrees=45
  swing2Motion=Free
  twistLimitDegrees=45
  twistMotion=Free
 }
 {bone1Name=dyn_necklace_01, bone2Name=base_necklace_01, swing1LimitDegrees=10, swing2LimitDegrees=20, twistLimitDegrees=0}
 {bone1Name=dyn_necklace_02, bone2Name=dyn_necklace_01, swing1LimitDegrees=15, swing2LimitDegrees=30, twistLimitDegrees=0}
]
