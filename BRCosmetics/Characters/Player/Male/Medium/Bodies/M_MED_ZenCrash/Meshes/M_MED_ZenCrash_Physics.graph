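# Physics asset for the M_MED_ZenCrash male medium player body, rigging the jacket as capsule chains around pelvis, spine and thighs.
# Nine strands each run from a tiny base capsule, constrained to spine_01, through three dyn segments.
# Tight angular limits restrict each chain, with one swing axis locked on back strands.
# Angular-free constraints cross-link some bottom segments, pairing front with side, and center back with left back and right side.
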PhysicsAsset /BRCosmetics/Characters/Player/Male/Medium/Bodies/M_MED_ZenCrash/Meshes/M_MED_ZenCrash_Physics
bodies {
 base_back_jacket_01 [{shapeType=Capsule, length=0.5, radius=0.5, shapeName=base_back_jacket_01_capsule}]
 base_l_back_jacket_01 [{shapeType=Capsule, length=0.5, radius=0.5, shapeName=base_l_back_jacket_01_capsule}]
 base_l_front_jacket_01 [{shapeType=Capsule, length=0.5, radius=0.5, shapeName=base_l_front_jacket_01_capsule}]
 base_l_left_jacket_01 [{shapeType=Capsule, length=0.5, radius=0.5, shapeName=base_l_left_jacket_01_capsule}]
 base_l_side_jacket_01 [{shapeType=Capsule, length=0.5, radius=0.5, shapeName=base_l_side_jacket_01_capsule}]
 base_r_back_jacket_01 [{shapeType=Capsule, length=0.5, radius=0.5, shapeName=base_r_back_jacket_01_capsule}]
 base_r_front_jacket_01 [{shapeType=Capsule, length=0.5, radius=0.5, shapeName=base_r_front_jacket_01_capsule}]
 base_r_left_jacket_01 [{shapeType=Capsule, length=0.5, radius=0.5, shapeName=base_r_left_jacket_01_capsule}]
 base_r_side_jacket_01 [{shapeType=Capsule, length=0.5, radius=0.5, shapeName=base_r_side_jacket_01_capsule}]
 dyn_back_jacket_01 [{
   shapeType=Capsule
   length=10.5909
   radius=3.7078
   shapeName=dyn_l_front_jacket_01_capsule
   center=<0 0 3.5>
   rotation=<rot 0.8674 4.9247 90.0373>
  }]
 dyn_back_jacket_02 [{
   shapeType=Capsule
   length=10.7159
   radius=3.9578
   shapeName=dyn_l_front_jacket_01_capsule
   center=<2 0 3.5>
   rotation=<rot 0.0004 0.0003 89.9999>
  }]
 dyn_back_jacket_03 [{
   shapeType=Capsule
   length=10.8409
   radius=4.2078
   shapeName=dyn_l_front_jacket_01_capsule
   center=<2 0 3.5>
   rotation=<rot 0.0004 0.0003 89.9999>
  }]
 dyn_l_back_jacket_01 [{shapeType=Capsule, length=7.3409, radius=3.5828, shapeName=dyn_l_front_jacket_01_capsule, center=<0 0 3.5>, rotation=<rot 0 0 95>}]
 dyn_l_back_jacket_02 [{shapeType=Capsule, length=7.4659, radius=3.8328, shapeName=dyn_l_front_jacket_01_capsule, center=<0 0 3.5>, rotation=<rot 0 0 95>}]
 dyn_l_back_jacket_03 [{shapeType=Capsule, length=7.5909, radius=4.0828, shapeName=dyn_l_front_jacket_01_capsule, center=<2 0 3.5>, rotation=<rot 0 0 95>}]
 dyn_l_front_jacket_01 [{
   shapeType=Capsule
   length=6.9659
   radius=4.2078
   shapeName=dyn_l_front_jacket_01_capsule
   center=<0 0 3>
   rotation=<rot 0.0006 0.0003 79.9999>
  }]
 dyn_l_front_jacket_02 [{
   shapeType=Capsule
   length=8.2612
   radius=5.8828
   shapeName=dyn_l_front_jacket_01_capsule
   center=<6.5415 -1.7401 4.5565>
   rotation=<rot -0.0011 -179.9994 95.0001>
  }]
 dyn_l_front_jacket_03 [{
   shapeType=Capsule
   length=9.3217
   radius=5.443
   shapeName=dyn_l_front_jacket_01_capsule
   center=<5.5398 1.0013 4.7597>
   rotation=<rot 9.1133 -0.7734 94.8791>
  }]
 dyn_l_left_jacket_01 [{
   shapeType=Capsule
   length=8.7159
   radius=3.3328
   shapeName=dyn_l_front_jacket_01_capsule
   center=<0 -2.8929 2.8333>
   rotation=<rot 0.0006 0.0004 74.9999>
  }]
 dyn_l_left_jacket_02 [{
   shapeType=Capsule
   length=8.8409
   radius=3.5828
   shapeName=dyn_l_front_jacket_01_capsule
   center=<0 -2.9682 3.2601>
   rotation=<rot 0.0006 0.0004 74.9999>
  }]
 dyn_l_left_jacket_03 [{
   shapeType=Capsule
   length=8.9659
   radius=3.8328
   shapeName=dyn_l_front_jacket_01_capsule
   center=<2 -2.9216 2.996>
   rotation=<rot 0.0006 0.0004 74.9999>
  }]
 dyn_l_side_jacket_01 [{
   shapeType=Capsule
   length=8.0909
   radius=4.0828
   shapeName=dyn_l_front_jacket_01_capsule
   center=<0 -2 4>
   rotation=<rot 0.0006 0.0003 79.9999>
  }]
 dyn_l_side_jacket_02 [{
   shapeType=Capsule
   length=8.2159
   radius=4.3328
   shapeName=dyn_l_front_jacket_01_capsule
   center=<2.3 -2 3.6>
   rotation=<rot 0.0006 0.0003 79.9999>
  }]
 dyn_l_side_jacket_03 [{
   shapeType=Capsule
   length=11.476
   radius=4.7179
   shapeName=dyn_l_front_jacket_01_capsule
   center=<3.5 -2 3.5>
   rotation=<rot 0.0006 0.0003 79.9999>
  }]
 dyn_r_back_jacket_01 [{
   shapeType=Capsule
   length=9.2159
   radius=3.4578
   shapeName=dyn_l_front_jacket_01_capsule
   center=<0 0 3>
   rotation=<rot 0.8673 4.9247 85.0372>
  }]
 dyn_r_back_jacket_02 [{
   shapeType=Capsule
   length=9.3409
   radius=3.7078
   shapeName=dyn_l_front_jacket_01_capsule
   center=<0 0 3>
   rotation=<rot 0.8673 4.9247 85.0372>
  }]
 dyn_r_back_jacket_03 [{
   shapeType=Capsule
   length=9.4659
   radius=3.9578
   shapeName=dyn_l_front_jacket_01_capsule
   center=<2 0 3>
   rotation=<rot 0.8673 4.9247 85.0372>
  }]
 dyn_r_front_jacket_01 [{
   shapeType=Capsule
   length=6.9659
   radius=4.2078
   shapeName=dyn_l_front_jacket_01_capsule
   center=<0 0 3>
   rotation=<rot 0.0006 0.0003 79.9999>
  }]
 dyn_r_front_jacket_02 [{
   shapeType=Capsule
   length=7.0909
   radius=5.3328
   shapeName=dyn_l_front_jacket_01_capsule
   center=<2 -1.4192 4.3775>
   rotation=<rot -0.001 -179.9995 100.0001>
  }]
 dyn_r_front_jacket_03 [{
   shapeType=Capsule
   length=6.0222
   radius=5.7078
   shapeName=dyn_l_front_jacket_01_capsule
   center=<4.0223 0.4233 4.5056>
   rotation=<rot 0.4241 -179.9053 84.9814>
  }]
 dyn_r_left_jacket_01 [{
   shapeType=Capsule
   length=10.5909
   radius=3.0828
   shapeName=dyn_l_front_jacket_01_capsule
   center=<0 0 2.5>
   rotation=<rot 0.0006 0.0002 94.9999>
  }]
 dyn_r_left_jacket_02 [{
   shapeType=Capsule
   length=10.7159
   radius=3.3328
   shapeName=dyn_l_front_jacket_01_capsule
   center=<0 0 2.5>
   rotation=<rot 0.0006 0.0002 94.9999>
  }]
 dyn_r_left_jacket_03 [{
   shapeType=Capsule
   length=10.8409
   radius=3.5828
   shapeName=dyn_l_front_jacket_01_capsule
   center=<2 0 2.5>
   rotation=<rot 0.0006 0.0002 94.9999>
  }]
 dyn_r_side_jacket_01 [{
   shapeType=Capsule
   length=8.0909
   radius=4.0828
   shapeName=dyn_l_front_jacket_01_capsule
   center=<0 0 4>
   rotation=<rot 0.0006 0.0003 79.9999>
  }]
 dyn_r_side_jacket_02 [{
   shapeType=Capsule
   length=8.2159
   radius=4.3328
   shapeName=dyn_l_front_jacket_01_capsule
   center=<0 0 4>
   rotation=<rot 0.0006 0.0003 79.9999>
  }]
 dyn_r_side_jacket_03 [{
   shapeType=Capsule
   length=8.3409
   radius=4.5828
   shapeName=dyn_l_front_jacket_01_capsule
   center=<4.947 0.1305 2.6572>
   rotation=<rot 0.0006 0.0002 84.9999>
  }]
 pelvis [{shapeType=Capsule, length=9.8912, radius=10.1249, shapeName=pelvis_capsule, center=<6.0347 -0.9847 0.0269>, rotation=<rot 0 179.9998 180>}]
 spine_01 [{shapeType=Capsule, length=14.505, radius=11.255, shapeName=spine_01_capsule, center=<11.0481 0 0>, rotation=<rot 0 -80.0002 -90>}]
 thigh_l [{
   shapeType=Capsule
   length=42.6475
   radius=10.0367
   shapeName=thigh_l_capsule
   center=<-11.6019 0.4731 -0.1468>
   rotation=<rot -85 -0.0013 0.0007>
  }]
 thigh_r [{
   shapeType=Capsule
   length=42.9024
   radius=9.7818
   shapeName=thigh_r_capsule
   center=<11.6432 0.4102 -0.3251>
   rotation=<rot -85 -0.0006 -179.9996>
  }]
}
constraints [
 {
  bone1Name=dyn_back_jacket_01
  bone2Name=base_back_jacket_01
  swing1LimitDegrees=5
  swing1Motion=Locked
  swing2LimitDegrees=10
  twistLimitDegrees=0
 }
 {
  bone1Name=dyn_back_jacket_02
  bone2Name=dyn_back_jacket_01
  swing1LimitDegrees=5
  swing1Motion=Locked
  swing2LimitDegrees=5
  twistLimitDegrees=0
 }
 {
  bone1Name=dyn_back_jacket_03
  bone2Name=dyn_back_jacket_02
  swing1LimitDegrees=5
  swing1Motion=Locked
  swing2LimitDegrees=5
  twistLimitDegrees=0
 }
 {
  bone1Name=dyn_l_back_jacket_01
  bone2Name=base_l_back_jacket_01
  swing1LimitDegrees=5
  swing1Motion=Locked
  swing2LimitDegrees=5
  twistLimitDegrees=0
 }
 {
  bone1Name=dyn_l_back_jacket_02
  bone2Name=dyn_l_back_jacket_01
  swing1LimitDegrees=5
  swing1Motion=Locked
  swing2LimitDegrees=5
  twistLimitDegrees=0
 }
 {
  bone1Name=dyn_l_back_jacket_03
  bone2Name=dyn_l_back_jacket_02
  swing1LimitDegrees=5
  swing1Motion=Locked
  swing2LimitDegrees=5
  twistLimitDegrees=0
 }
 {bone1Name=dyn_l_front_jacket_01, bone2Name=base_l_front_jacket_01, swing1LimitDegrees=5, swing2LimitDegrees=5, twistLimitDegrees=0}
 {bone1Name=dyn_l_front_jacket_02, bone2Name=dyn_l_front_jacket_01, swing1LimitDegrees=5, swing2LimitDegrees=10, twistLimitDegrees=0}
 {bone1Name=dyn_l_front_jacket_03, bone2Name=dyn_l_front_jacket_02, swing1LimitDegrees=2, swing2LimitDegrees=15, twistLimitDegrees=5}
 {bone1Name=dyn_l_left_jacket_01, bone2Name=base_l_left_jacket_01, swing1LimitDegrees=2, swing2LimitDegrees=5, twistLimitDegrees=0}
 {bone1Name=dyn_l_left_jacket_02, bone2Name=dyn_l_left_jacket_01, swing1LimitDegrees=2, swing2LimitDegrees=5, twistLimitDegrees=0}
 {bone1Name=dyn_l_left_jacket_03, bone2Name=dyn_l_left_jacket_02, swing1LimitDegrees=2, swing2LimitDegrees=5, twistLimitDegrees=0}
 {bone1Name=dyn_l_side_jacket_01, bone2Name=base_l_side_jacket_01, swing1LimitDegrees=1, swing2LimitDegrees=5, twistLimitDegrees=0}
 {bone1Name=dyn_l_side_jacket_02, bone2Name=dyn_l_side_jacket_01, swing1LimitDegrees=1, swing2LimitDegrees=8, twistLimitDegrees=0}
 {bone1Name=dyn_l_side_jacket_03, bone2Name=dyn_l_side_jacket_02, swing1LimitDegrees=1, swing2LimitDegrees=8, twistLimitDegrees=0}
 {
  bone1Name=dyn_r_back_jacket_01
  bone2Name=base_r_back_jacket_01
  swing1LimitDegrees=5
  swing1Motion=Locked
  swing2LimitDegrees=5
  twistLimitDegrees=0
 }
 {
  bone1Name=dyn_r_back_jacket_02
  bone2Name=dyn_r_back_jacket_01
  swing1LimitDegrees=5
  swing1Motion=Locked
  swing2LimitDegrees=5
  twistLimitDegrees=0
 }
 {
  bone1Name=dyn_r_back_jacket_03
  bone2Name=dyn_r_back_jacket_02
  swing1LimitDegrees=5
  swing1Motion=Locked
  swing2LimitDegrees=5
  twistLimitDegrees=0
 }
 {bone1Name=dyn_r_front_jacket_01, bone2Name=base_r_front_jacket_01, swing1LimitDegrees=5, swing2LimitDegrees=5, twistLimitDegrees=0}
 {bone1Name=dyn_r_front_jacket_02, bone2Name=dyn_r_front_jacket_01, swing1LimitDegrees=5, swing2LimitDegrees=10, twistLimitDegrees=0}
 {bone1Name=dyn_r_front_jacket_03, bone2Name=dyn_r_front_jacket_02, swing1LimitDegrees=5, swing2LimitDegrees=15, twistLimitDegrees=0}
 {bone1Name=dyn_r_left_jacket_01, bone2Name=base_r_left_jacket_01, swing1LimitDegrees=0, swing2LimitDegrees=5, twistLimitDegrees=0}
 {bone1Name=dyn_r_left_jacket_02, bone2Name=dyn_r_left_jacket_01, swing1LimitDegrees=0, swing2LimitDegrees=5, twistLimitDegrees=0}
 {bone1Name=dyn_r_left_jacket_03, bone2Name=dyn_r_left_jacket_02, swing1LimitDegrees=0, swing2LimitDegrees=5, twistLimitDegrees=0}
 {bone1Name=dyn_r_side_jacket_01, bone2Name=base_r_side_jacket_01, swing1LimitDegrees=5, swing2LimitDegrees=5, twistLimitDegrees=0}
 {bone1Name=dyn_r_side_jacket_02, bone2Name=dyn_r_side_jacket_01, swing1LimitDegrees=5, swing2LimitDegrees=5, twistLimitDegrees=0}
 {bone1Name=dyn_r_side_jacket_03, bone2Name=dyn_r_side_jacket_02, swing1LimitDegrees=5, swing2LimitDegrees=5, twistLimitDegrees=0}
 {
  bone1Name=base_l_left_jacket_01
  bone2Name=spine_01
  swing1LimitDegrees=45
  swing1Motion=Free
  swing2LimitDegrees=45
  swing2Motion=Free
  twistLimitDegrees=45
  twistMotion=Free
 }
 {
  bone1Name=base_l_back_jacket_01
  bone2Name=spine_01
  swing1LimitDegrees=45
  swing1Motion=Free
  swing2LimitDegrees=45
  swing2Motion=Free
  twistLimitDegrees=45
  twistMotion=Free
 }
 {
  bone1Name=base_r_left_jacket_01
  bone2Name=spine_01
  swing1LimitDegrees=45
  swing1Motion=Free
  swing2LimitDegrees=45
  swing2Motion=Free
  twistLimitDegrees=45
  twistMotion=Free
 }
 {
  bone1Name=base_r_back_jacket_01
  bone2Name=spine_01
  swing1LimitDegrees=45
  swing1Motion=Free
  swing2LimitDegrees=45
  swing2Motion=Free
  twistLimitDegrees=45
  twistMotion=Free
 }
 {
  bone1Name=base_back_jacket_01
  bone2Name=spine_01
  swing1LimitDegrees=45
  swing1Motion=Free
  swing2LimitDegrees=45
  swing2Motion=Free
  twistLimitDegrees=45
  twistMotion=Free
 }
 {
  bone1Name=base_l_front_jacket_01
  bone2Name=spine_01
  swing1LimitDegrees=45
  swing1Motion=Free
  swing2LimitDegrees=45
  swing2Motion=Free
  twistLimitDegrees=45
  twistMotion=Free
 }
 {
  bone1Name=base_r_front_jacket_01
  bone2Name=spine_01
  swing1LimitDegrees=45
  swing1Motion=Free
  swing2LimitDegrees=45
  swing2Motion=Free
  twistLimitDegrees=45
  twistMotion=Free
 }
 {
  bone1Name=base_l_side_jacket_01
  bone2Name=spine_01
  swing1LimitDegrees=45
  swing1Motion=Free
  swing2LimitDegrees=45
  swing2Motion=Free
  twistLimitDegrees=45
  twistMotion=Free
 }
 {
  bone1Name=base_r_side_jacket_01
  bone2Name=spine_01
  swing1LimitDegrees=45
  swing1Motion=Free
  swing2LimitDegrees=45
  swing2Motion=Free
  twistLimitDegrees=45
  twistMotion=Free
 }
 {
  bone1Name=thigh_l
  bone2Name=pelvis
  swing1LimitDegrees=45
  swing1Motion=Free
  swing2LimitDegrees=45
  swing2Motion=Free
  twistLimitDegrees=45
  twistMotion=Free
 }
 {
  bone1Name=thigh_r
  bone2Name=pelvis
  swing1LimitDegrees=45
  swing1Motion=Free
  swing2LimitDegrees=45
  swing2Motion=Free
  twistLimitDegrees=45
  twistMotion=Free
 }
 {
  bone1Name=spine_01
  bone2Name=pelvis
  swing1LimitDegrees=45
  swing1Motion=Free
  swing2LimitDegrees=45
  swing2Motion=Free
  twistLimitDegrees=45
  twistMotion=Free
 }
 {
  bone1Name=dyn_r_side_jacket_03
  bone2Name=dyn_back_jacket_03
  swing1LimitDegrees=45
  swing1Motion=Free
  swing2LimitDegrees=45
  swing2Motion=Free
  twistLimitDegrees=45
  twistMotion=Free
 }
 {
  bone1Name=dyn_l_back_jacket_03
  bone2Name=dyn_back_jacket_03
  swing1LimitDegrees=45
  swing1Motion=Free
  swing2LimitDegrees=45
  swing2Motion=Free
  twistLimitDegrees=45
  twistMotion=Free
 }
 {
  bone1Name=dyn_l_front_jacket_03
  bone2Name=dyn_l_side_jacket_03
  swing1LimitDegrees=45
  swing1Motion=Free
  swing2LimitDegrees=45
  swing2Motion=Free
  twistLimitDegrees=45
  twistMotion=Free
 }
 {
  bone1Name=dyn_r_front_jacket_03
  bone2Name=dyn_r_side_jacket_03
  swing1LimitDegrees=45
  swing1Motion=Free
  swing2LimitDegrees=45
  swing2Motion=Free
  twistLimitDegrees=45
  twistMotion=Free
 }
]
