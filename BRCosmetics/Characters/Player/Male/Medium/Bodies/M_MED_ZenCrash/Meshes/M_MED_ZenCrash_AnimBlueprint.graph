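# Post-process style animation blueprint for the ZenCrash male body part, adding corrective and secondary motion to the pose copied from its attached parent mesh.
# A Control Rig adjusts jacket and collar using crouch, aim, movement and head-rotation values computed each update.
# RigidBody nodes then simulate jacket, necklace and body physics, with AnimDynamics for laces and bracelet, before fixed additive jacket-bone rotations.
# The update also stores parameter-asset gravity and rigid-body settings.
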AnimBlueprint /BRCosmetics/Characters/Player/Male/Medium/Bodies/M_MED_ZenCrash/Meshes/M_MED_ZenCrash_AnimBlueprint
target_skeleton=&/M_MED_ZenCrash_Skeleton
graphs [
 AnimGraph {nodes [
   n1 = Control Rig!(
    Source=Copy Pose From Mesh() {bUseAttachedParent=true}
    Alpha
    Crouching:bool=Get bIsCrouching()
    Crouching_Aim:bool=Get Crouching_Aim() -> (Crouching_Aim:bool)
    Crouch_jacket_ADJ_bool:bool=false
    Crouch_jacket_ADJ_float:double=Get Crouch_jacket_ADJ() -> (Crouch_jacket_ADJ:double)
    Crouch_jacket_up_ADJ_bool:double=Get Crouch_jacket_up_ADJ_bool() -> (Crouch_jacket_up_ADJ_bool:double)
    Crouch_jacket_walk_ADJ:double=Get Crouch_jacket_walk_ADJ() -> (Crouch_jacket_walk_ADJ:double)
    Collar_ADJ:double=Get Collar_ADJ() -> (Collar_ADJ:double)
    Collar_ADJ_Bool:bool=false
    Head_Left:double=Get Head_Left() -> (Head_Left:double)
    Head_Right:double=Get Head_Right() -> (Head_Right:double)
   ) -> (Pose) {
    alphaScaleBiasClamp {interpSpeedDecreasing=5, interpSpeedIncreasing=5}
    controlRigAssetReference {blueprintRigClass=&/M_MED_ZenCrash_CtrlRig.M_MED_ZenCrash_CtrlRig_C}
   }
   n2 = RigidBody!(
    ComponentPose=Local To Component(LocalPose=@n1)
    Alpha
    OverrideWorldGravity:Vector="0,0,0"
    ExternalForce
    SimSpaceSettings:Sim Space Settings Structure=(WorldAlpha=1,VelocityScaleZ=1,DampingAlpha=1,MaxLinearVelocity=700,MaxAngularVelocity=8,MaxLinearAcceleration=3500,MaxAngularAcceleration=40,ExternalLinearDragV=(X=0,Y=0,Z=0),ExternalLinearVelocity=(X=0,Y=0,Z=0),ExternalAngularVelocity=(X=0,Y=0,Z=0))
   ) -> (Pose) {
    bOverrideWorldGravity=true
    overridePhysicsAsset=&/M_MED_ZenCrash_Jacket_Physics
    simulationSpace=BaseBoneSpace
    baseBoneRef {boneName=pelvis}
    simSpaceSettings {maxAngularAcceleration=40, maxAngularVelocity=8, maxLinearAcceleration=3500, maxLinearVelocity=700, worldAlpha=1}
   }
   n3 = RigidBody(
    ComponentPose=@n2
    OverrideWorldGravity="0,0,0"
    SimSpaceSettings=(WorldAlpha=1,VelocityScaleZ=1,DampingAlpha=1,MaxLinearVelocity=700,MaxAngularVelocity=8,MaxLinearAcceleration=3500,MaxAngularAcceleration=40,ExternalLinearDragV=(X=0,Y=0,Z=0),ExternalLinearVelocity=(X=0,Y=0,Z=0),ExternalAngularVelocity=(X=0,Y=0,Z=0))
   ) {
    bOverrideWorldGravity=true
    overridePhysicsAsset=&/M_MED_ZenCrash_Necklace_one_Physics
    simulationSpace=BaseBoneSpace
    baseBoneRef {boneName=spine_05}
    simSpaceSettings {maxAngularAcceleration=40, maxAngularVelocity=8, maxLinearAcceleration=3500, maxLinearVelocity=700, worldAlpha=1}
   }
   n4 = Anim Dynamics - Bone: dyn_l_leg_left_lace_01!(ComponentPose=@n3, Alpha=1, GravityOverride:Vector="0,0,-980", ExternalForce="0,0,0") -> (Pose) {
    angularDampingOverride=1
    angularSpringConstant=5000
    bAngularSpring=true
    bOverrideAngularDamping=true
    bUseGravityOverride=true
    gravityScale=0.35
    simulationSpace=RootRelative
    boundBone {boneName=dyn_l_leg_left_lace_01}
    componentLinearAccScale=<1 1 1>
    componentLinearVelScale=<1 1 1>
    gravityOverride {z=-980}
    physicsBodyDefinitions [{
      boundBone {boneName=dyn_l_leg_left_lace_01}
      boxExtents=<10 10 10>
      constraintSetup {angularLimitsMax=<0 15 20>, angularLimitsMin=<0 -30 -20>, angularTarget=<1 0 0>}
      localJointOffset=<4 0 0>
     }]
   }
   n5 = Anim Dynamics - Bone: dyn_l_leg_right_lace_01!(ComponentPose=@n4, Alpha=1, GravityOverride:Vector="0,0,-980", ExternalForce="0,0,0") -> (Pose) {
    angularDampingOverride=1
    angularSpringConstant=5000
    bAngularSpring=true
    bOverrideAngularDamping=true
    bUseGravityOverride=true
    gravityScale=0.35
    simulationSpace=RootRelative
    boundBone {boneName=dyn_l_leg_right_lace_01}
    componentLinearAccScale=<1 1 1>
    componentLinearVelScale=<1 1 1>
    gravityOverride {z=-980}
    physicsBodyDefinitions [{
      boundBone {boneName=dyn_l_leg_right_lace_01}
      boxExtents=<10 10 10>
      constraintSetup {angularLimitsMax=<0 15 20>, angularLimitsMin=<0 -30 -20>, angularTarget=<1 0 0>}
      localJointOffset=<4 0 0>
     }]
   }
   n6 = Anim Dynamics - Bone: dyn_bracelate_01!(ComponentPose=@n5, Alpha=1, GravityOverride:Vector="0,0,-980", ExternalForce="0,0,0") -> (Pose) {
    angularDampingOverride=1
    angularSpringConstant=5000
    bAngularSpring=true
    bLinearSpring=true
    bOverrideAngularDamping=true
    bUseGravityOverride=true
    bUseSphericalLimits=true
    gravityScale=0.35
    linearSpringConstant=5000
    simulationSpace=RootRelative
    boundBone {boneName=dyn_bracelate_01}
    componentAppliedLinearAccClamp=<5000 5000 5000>
    componentLinearAccScale {x=0.2}
    componentLinearVelScale {x=0.2}
    gravityOverride {z=-980}
    physicsBodyDefinitions [{
      boundBone {boneName=dyn_bracelate_01}
      boxExtents=<40 40 40>
      constraintSetup {angularLimitsMax=<1 0 1>, angularLimitsMin=<-1 0 -1>, angularTarget=<1 0 0>, linearAxesMin=<-1 0 0>}
      localJointOffset=<4 0 0>
     }]
    relativeSpaceBone {boneName=lowerarm_l}
   }
   n7 = Anim Dynamics - Bone: dyn_l_leg_left_lace_02!(ComponentPose=@n6, Alpha=1, GravityOverride:Vector="0,0,-980", ExternalForce="0,0,0") -> (Pose) {
    angularDampingOverride=1
    angularSpringConstant=5000
    bAngularSpring=true
    bOverrideAngularDamping=true
    bUseGravityOverride=true
    gravityScale=0.35
    simulationSpace=RootRelative
    boundBone {boneName=dyn_l_leg_left_lace_02}
    componentLinearAccScale=<1 1 1>
    componentLinearVelScale=<1 1 1>
    gravityOverride {z=-980}
    physicsBodyDefinitions [{
      boundBone {boneName=dyn_l_leg_left_lace_02}
      boxExtents=<10 10 10>
      constraintSetup {angularLimitsMax=<0 15 20>, angularLimitsMin=<0 -20 -20>, angularTarget=<1 0 0>}
      localJointOffset=<4 0 0>
     }]
   }
   n8 = Anim Dynamics - Bone: dyn_l_leg_right_lace_02!(ComponentPose=@n7, Alpha=1, GravityOverride:Vector="0,0,-980", ExternalForce="0,0,0") -> (Pose) {
    angularDampingOverride=1
    angularSpringConstant=5000
    bAngularSpring=true
    bOverrideAngularDamping=true
    bUseGravityOverride=true
    gravityScale=0.35
    simulationSpace=RootRelative
    boundBone {boneName=dyn_l_leg_right_lace_02}
    componentLinearAccScale=<1 1 1>
    componentLinearVelScale=<1 1 1>
    gravityOverride {z=-980}
    physicsBodyDefinitions [{
      boundBone {boneName=dyn_l_leg_right_lace_02}
      boxExtents=<10 10 10>
      constraintSetup {angularLimitsMax=<0 15 20>, angularLimitsMin=<0 -20 -20>, angularTarget=<1 0 0>}
      localJointOffset=<4 0 0>
     }]
   }
   n9 = Anim Dynamics - Bone: dyn_r_leg_left_lace_01!(ComponentPose=@n8, Alpha=1, GravityOverride:Vector="0,0,-980", ExternalForce="0,0,0") -> (Pose) {
    angularDampingOverride=1
    angularSpringConstant=5000
    bAngularSpring=true
    bOverrideAngularDamping=true
    bUseGravityOverride=true
    gravityScale=0.35
    simulationSpace=RootRelative
    boundBone {boneName=dyn_r_leg_left_lace_01}
    componentLinearAccScale=<1 1 1>
    componentLinearVelScale=<1 1 1>
    gravityOverride {z=-980}
    physicsBodyDefinitions [{
      boundBone {boneName=dyn_r_leg_left_lace_01}
      boxExtents=<10 10 10>
      constraintSetup {angularLimitsMax=<0 15 20>, angularLimitsMin=<0 -30 -20>, angularTarget=<1 0 0>}
      localJointOffset=<4 0 0>
     }]
   }
   n10 = Anim Dynamics - Bone: dyn_r_leg_right_lace_01!(ComponentPose=@n9, Alpha=1, GravityOverride:Vector="0,0,-980", ExternalForce="0,0,0") -> (Pose) {
    angularDampingOverride=1
    angularSpringConstant=5000
    bAngularSpring=true
    bOverrideAngularDamping=true
    bUseGravityOverride=true
    gravityScale=0.35
    simulationSpace=RootRelative
    boundBone {boneName=dyn_r_leg_right_lace_01}
    componentLinearAccScale=<1 1 1>
    componentLinearVelScale=<1 1 1>
    gravityOverride {z=-980}
    physicsBodyDefinitions [{
      boundBone {boneName=dyn_r_leg_right_lace_01}
      boxExtents=<10 10 10>
      constraintSetup {angularLimitsMax=<0 15 20>, angularLimitsMin=<0 -30 -20>, angularTarget=<1 0 0>}
      localJointOffset=<4 0 0>
     }]
   }
   n11 = Anim Dynamics - Bone: dyn_r_leg_left_lace_02!(ComponentPose=@n10, Alpha=1, GravityOverride:Vector="0,0,-980", ExternalForce="0,0,0") -> (Pose) {
    angularDampingOverride=1
    angularSpringConstant=5000
    bAngularSpring=true
    bOverrideAngularDamping=true
    bUseGravityOverride=true
    gravityScale=0.35
    simulationSpace=RootRelative
    boundBone {boneName=dyn_r_leg_left_lace_02}
    componentLinearAccScale=<1 1 1>
    componentLinearVelScale=<1 1 1>
    gravityOverride {z=-980}
    physicsBodyDefinitions [{
      boundBone {boneName=dyn_r_leg_left_lace_02}
      boxExtents=<10 10 10>
      constraintSetup {angularLimitsMax=<0 15 20>, angularLimitsMin=<0 -20 -20>, angularTarget=<1 0 0>}
      localJointOffset=<4 0 0>
     }]
   }
   n12 = Anim Dynamics - Bone: dyn_r_leg_right_lace_02!(ComponentPose=@n11, Alpha=1, GravityOverride:Vector="0,0,-980", ExternalForce="0,0,0") -> (Pose) {
    angularDampingOverride=1
    angularSpringConstant=5000
    bAngularSpring=true
    bOverrideAngularDamping=true
    bUseGravityOverride=true
    gravityScale=0.35
    simulationSpace=RootRelative
    boundBone {boneName=dyn_r_leg_right_lace_02}
    componentLinearAccScale=<1 1 1>
    componentLinearVelScale=<1 1 1>
    gravityOverride {z=-980}
    physicsBodyDefinitions [{
      boundBone {boneName=dyn_r_leg_right_lace_02}
      boxExtents=<10 10 10>
      constraintSetup {angularLimitsMax=<0 15 20>, angularLimitsMin=<0 -20 -20>, angularTarget=<1 0 0>}
      localJointOffset=<4 0 0>
     }]
   }
   n13 = RigidBody(
    ComponentPose=@n12
    OverrideWorldGravity="0,0,0"
    SimSpaceSettings=(WorldAlpha=1,VelocityScaleZ=0.5,DampingAlpha=1,MaxLinearVelocity=700,MaxAngularVelocity=8,MaxLinearAcceleration=3500,MaxAngularAcceleration=40,ExternalLinearDragV=(X=0,Y=0,Z=0),ExternalLinearVelocity=(X=0,Y=0,Z=0),ExternalAngularVelocity=(X=0,Y=0,Z=0))
   ) {
    bOverrideWorldGravity=true
    overridePhysicsAsset=&/M_MED_ZenCrash_Physics
    simulationSpace=BaseBoneSpace
    baseBoneRef {boneName=pelvis}
    simSpaceSettings {maxAngularAcceleration=40, maxAngularVelocity=8, maxLinearAcceleration=3500, maxLinearVelocity=700, velocityScaleZ=0.5, worldAlpha=1}
   }
   n14 = "Transform (Modify) Bone - Bone: dyn_r_front_jacket_02"(ComponentPose=@n13, Alpha=1, Translation="0,0,0", Rotation="15,0,0", Scale="1,1,1") {
    rotationMode=BMM_Additive
    rotationSpace=BCS_BoneSpace
    translationMode=BMM_Additive
    translationSpace=BCS_BoneSpace
    alphaScaleBiasClamp {bInterpResult=true, interpSpeedIncreasing=4}
    boneToModify {boneName=dyn_r_front_jacket_02}
    rotation {pitch=15}
   }
   n15 = "Transform (Modify) Bone - Bone: dyn_r_front_jacket_03"(ComponentPose=@n14, Alpha=1, Translation="0,0,0", Rotation="15,0,0", Scale="1,1,1") {
    rotationMode=BMM_Additive
    rotationSpace=BCS_BoneSpace
    translationMode=BMM_Additive
    translationSpace=BCS_BoneSpace
    alphaScaleBiasClamp {bInterpResult=true, interpSpeedIncreasing=4}
    boneToModify {boneName=dyn_r_front_jacket_03}
    rotation {pitch=15}
   }
   n16 = "Transform (Modify) Bone - Bone: base_l_front_jacket_01"(ComponentPose=@n15, Alpha=1, Translation="0,0,0", Rotation="-10,0,0", Scale="1,1,1") {
    rotationMode=BMM_Additive
    rotationSpace=BCS_BoneSpace
    translationMode=BMM_Additive
    translationSpace=BCS_BoneSpace
    alphaScaleBiasClamp {bInterpResult=true, interpSpeedIncreasing=4}
    boneToModify {boneName=base_l_front_jacket_01}
    rotation {pitch=-10}
   }
   Output Pose(Result=Component To Local(ComponentPose=@n16))
  ]}
 EventGraph {nodes [
   Event BlueprintUpdateAnimation()
   > n1 = Sequence!(execute) -> (then_0, then_1, then_2:Exec, then_3:Exec)
   > n2 = "Gravity Override State Function (Parameter Asset) (Nativized)"(
    CharacterPartAnimInstance=Self-Reference()
    GameTimeInSeconds=GetGameTimeInSeconds()
    Parameters=&/M_MED_ZenCrash_Parameters
   )
   > Set Gravity Override(Gravity Override=@n2)
   > Set RBAN Settings Out(RBAN Settings Out=@n2.OutRigidBodyAnimNodeParameters)
   n3 = Get bIsTargeting()
   n4 = NOT Boolean(A=@n3)
   n5 = Get bIsCrouching()
   Set Crouching_Alpha(execute=@n1.then_1, Crouching_Alpha:double="To Float (Boolean)"(InBool=AND Boolean(A=@n5, B=@n4))) -> (Output_Get:double)
   > Set Crouching_Target_Alpha(Crouching_Target_Alpha:double="To Float (Boolean)"(InBool=AND Boolean(A=@n5, B=@n3))) -> (Output_Get:double)
   n6 = NOT Boolean(A=Get bIsPlayingMeleeAnim())
   n7 = GetPartSkeletalMeshComponent()
   n8 = float + float(
    A=GetAlphaFromReferenceTransformDelta(SkeletalMeshComponent=@n7, BoneName=head, BaseName=spine_05, InputRange=(X=-25,Y=-50), TwistAxis="1,0,0")
    B=GetAlphaFromReferenceTransformDelta(SkeletalMeshComponent=@n7, BoneName=head, BaseName=spine_05, DeltaComponentType=EulerY, InputRange=(X=5,Y=30), TwistAxis="1,0,0")
   )
   > Set Head look up alpha(Head look up alpha:double=SelectFloat(A="Clamp (Float)"(Value=@n8, Min=0, Max=1), B=0, bPickA=@n6)) -> (Output_Get:double)
   n9 = GetAlphaFromReferenceTransformDelta(
    SkeletalMeshComponent=@n7
    BoneName=head
    BaseName=spine_05
    DeltaComponentType=EulerY
    InputRange=(X=-15,Y=75)
    OutputRange=(X=1,Y=0)
    TwistAxis="1,0,0"
   )
   > Set Head look down alpha(Head look down alpha:double=SelectFloat(A=@n9, B=0, bPickA=@n6)) -> (Output_Get:double)
   n10 = Get bIsCrouching()
   n11 = Get bIsTargeting()
   > Set Crouch_head_down(Crouch_head_down:double=SelectFloat(A=Get Head look down alpha() -> (Head look down alpha:double), B=0, bPickA=AND Boolean(A=@n10, B=@n11))) -> (Output_Get:double)
   n12 = Get bIsCrouching()
   n13 = Get bIsTargeting()
   > Set Crouch_head_up(Crouch_head_up:double=SelectFloat(A=Get Head look up alpha() -> (Head look up alpha:double), B=0, bPickA=AND Boolean(A=@n12, B=@n13))) -> (Output_Get:double)
   > Set Crouchingalpha(Crouchingalpha:double="To Float (Boolean)"(InBool=@n12)) -> (Output_Get:double)
   > Set Crouching_Aim(Crouching_Aim:bool=AND Boolean(A=Get bIsCrouching(), B=Get bIsTargeting())) -> (Output_Get:bool)
   n14 = Get bIsTargeting()
   > Set Crouching_not_targetting(Crouching_not_targetting:bool=AND Boolean(A=Get bIsCrouching(), B=NOT Boolean(A=@n14))) -> (Output_Get:bool)
   n15 = GetOwnerSkeletalMeshComponent()
   n16 = BreakRotator(InRot=BreakTransform(InTransform=GetDeltaTransformFromRefPose(self=@n15, BoneName=head, BaseName=spine_05)).Rotation)
   n17 = Set Head_down(execute=@n1.then_2, Head_down:double=MapRangeClamped(Value=@n16.Yaw, InRangeA=-13, InRangeB=13)) -> (Output_Get:double)
   n18 = Get bIsTargeting()
   n19 = Get bIsCrouching()
   > Set Crouch_jacket_ADJ(Crouch_jacket_ADJ:double=SelectFloat(A=@n17, B=0, bPickA=AND Boolean(A=@n19, B=@n18))) -> (Output_Get:double)
   > Set Crouch_jacket_ADJ_bool(Crouch_jacket_ADJ_bool:bool=AND Boolean!(A=@n19, B=@n18, C:bool="Equal (Float)"(A=@n17)) -> (ReturnValue)) -> (Output_Get:bool)
   > Set Head_up(Head_up:double=MapRangeClamped(Value=@n16.Yaw, InRangeA=-20, InRangeB=-75)) -> (Output_Get:double)
   n20 = Get bIsTargeting()
   n21 = Get bIsCrouching()
   n22 = SelectFloat(A=Get Head_up() -> (Head_up:double), B=0, bPickA=NOT Boolean(A=OR Boolean(A=@n21, B=@n20)))
   > Set Collar_ADJ(Collar_ADJ:double=@n22) -> (Output_Get:double)
   > Set Crouch_jacket_up_ADJ_bool(Crouch_jacket_up_ADJ_bool:double=0) -> (Output_Get:double)
   n23 = Get bIsCrouching()
   n24 = Get bIsMoving2D()
   > Set Crouch_jacket_walk_ADJ(Crouch_jacket_walk_ADJ:double="To Float (Boolean)"(InBool=AND Boolean!(A=@n23, B=@n24) -> (ReturnValue))) -> (Output_Get:double)
   n25 = GetOwnerSkeletalMeshComponent()
   n26 = BreakRotator(InRot=BreakTransform(InTransform=GetDeltaTransformFromRefPose(self=@n25, BoneName=head, BaseName=spine_05)).Rotation)
   Set Head_Left(execute=@n1.then_3, Head_Left:double=MapRangeClamped(Value=@n26, InRangeA=0, InRangeB=75)) -> (Output_Get:double)
   > Set Head_Right(Head_Right:double=MapRangeClamped(Value=@n26, InRangeA=0, InRangeB=-75)) -> (Output_Get:double)
   TryGetPawnOwner()
   Set Machine gun corr(Machine gun corr:double=SelectFloat(A=@n9, B=0, bPickA=false)) -> (Output_Get:double)
   Set Collar_ADJ_Bool(Collar_ADJ_Bool:bool="float > float"(A=@n22)) -> (Output_Get:bool)
  ]}
]
variables [
 Collar_ADJ {type=number}
 Collar_ADJ_Bool {type=boolean}
 Crouch_head_down {type=number}
 Crouch_head_up {type=number}
 Crouch_jacket_ADJ {type=number}
 Crouch_jacket_ADJ_bool {type=boolean}
 Crouch_jacket_up_ADJ_bool {type=number}
 Crouch_jacket_walk_ADJ {type=number}
 Crouching_Aim {type=boolean}
 Crouching_Alpha {type=number}
 Crouching_not_targetting {type=boolean}
 Crouching_Target_Alpha {type=number}
 Crouchingalpha {type=number}
 Gravity Override {type=Vector}
 Head look down alpha {type=number}
 Head look up alpha {type=number}
 Head_down {type=number}
 Head_Left {type=number}
 Head_Right {type=number}
 Head_up {type=number}
 Machine gun corr {type=number}
 RBAN Settings Out {type=FortRigidBodyAnimNodeParameters}
]
