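# Physics asset for the necklace on the ZenCrash male medium character body, covering a short dynamic bone chain and the upper spine.
# Two overlapping capsules on spine_05 approximate the upper torso, and capsules on the base and two dyn necklace bones form the necklace.
# The necklace base has its swing and twist locked to spine_05, and the dyn bones chain from it with small swing limit angles.
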PhysicsAsset /BRCosmetics/Characters/Player/Male/Medium/Bodies/M_MED_ZenCrash/Meshes/M_MED_ZenCrash_Necklace_Physics
bodies {
 base_necklace_01 [{shapeType=Capsule, length=0.5, radius=0.5, shapeName=base_necklace_01_capsule}]
 dyn_necklace_01 [{shapeType=Capsule, length=13.3565, radius=1.8255, shapeName=dyn_necklace_01_capsule, center=<1.5 0 2>, rotation=<rot 0.0002 0 -90>}]
 dyn_necklace_02 [{shapeType=Capsule, length=9.5828, radius=2.9609, shapeName=dyn_necklace_02_capsule, center=<4.5 0 1>, rotation=<rot -18.4414 0.1598 -90>}]
 spine_05 [
  {shapeType=Capsule, length=20.8368, radius=11.5493, shapeName=spine_05_capsule, center=<-3.7589 0.4055 0>}
  {shapeType=Capsule, length=15, radius=15, shapeName=spine_05_capsule, center=<-5.7981 -3.2848 0>}
 ]
}
constraints [
 {
  bone1Name=base_necklace_01
  bone2Name=spine_05
  swing1LimitDegrees=0
  swing1Motion=Locked
  swing2LimitDegrees=0
  swing2Motion=Locked
  twistLimitDegrees=0
  twistMotion=Locked
 }
 {bone1Name=dyn_necklace_01, bone2Name=base_necklace_01, swing1LimitDegrees=5, swing2LimitDegrees=5, twistLimitDegrees=0}
 {bone1Name=dyn_necklace_02, bone2Name=dyn_necklace_01, swing1LimitDegrees=5, swing2LimitDegrees=10, twistLimitDegrees=0}
]
