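# Control Rig for ZenCrash face-accessory hair, applying pose-based corrections to the hair tail and lower hair bones.
# The tail base's rotation and scale are constrained between head and spine_05, then sequenced branches cover head-up, crouch, crouch-aim-move, targeting and skydive states.
# Spherical pose readers on head relative to spine_05 weight additional transforms, with pelvis height scaling crouch-move.
# RifleUP directly weights a final tail offset; the DeltaRotation helper is unused.
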ControlRigBlueprint /BRCosmetics/Characters/Player/Male/Medium/Bodies/M_MED_ZenCrash/Meshes/Parts/M_MED_ZenCrash_FaceAcc_CtrlRig
graphs [
 New Function_ContainedGraph {nodes [
   n1 = RigVMFunctionEntryNode()
   RigVMFunction_MathBoolToInteger = &RigVMFunction_MathBoolToInteger(Value=@n1.onlyTwist)
   RigVMFunction_MathQuaternionSwingTwist = &RigVMFunction_MathQuaternionSwingTwist(
    TwistAxis=@n1.twistAxis
    Input=&RigUnit_GetRelativeTransformForItem(bChildInitial=False, bParentInitial=False, Child=@n1.child, Parent=@n1.parent).RelativeTransform.Rotation
   )
   n2 = RigVMDispatchNode(
    Values=((),())
    Index=@RigVMFunction_MathBoolToInteger.Result
    Values.0=@RigVMFunction_MathQuaternionSwingTwist.Swing
    Values.1=@RigVMFunction_MathQuaternionSwingTwist.Twist
   ) {template="DISPATCH_RigVMDispatch_SelectInt32(in Index,in Values,out Result)"}
   RigVMFunction_MathQuaternionToEuler = &RigVMFunction_MathQuaternionToEuler(Value=@n2.Result, RotationOrder=@n1.Rotation Order)
   > DISPATCH_RigVMDispatch_Print = RigVMDispatchNode(ScreenColor=(R=1,G=1,B=1,A=1), ScreenDuration=1.5, Enabled=@n1.debugRotation, Value=@RigVMFunction_MathQuaternionToEuler.Result) {template="DISPATCH_RigVMDispatch_Print(in Prefix,in Value,in Enabled,in ScreenDuration,in ScreenColor)"}
   Cast = RigVMDispatchNode(Value=@n1.triggerAxis) {template="DISPATCH_RigVMDispatch_CastEnumToInt(in Value,out Result)"}
   DISPATCH_RigVMDispatch_SelectInt32 = RigVMDispatchNode(
    Values=(0,0,0)
    Index=@Cast.Result
    Values.0=@RigVMFunction_MathQuaternionToEuler.Result.X
    Values.1=@RigVMFunction_MathQuaternionToEuler.Result.Y
    Values.2=@RigVMFunction_MathQuaternionToEuler.Result.Z
   ) {template="DISPATCH_RigVMDispatch_SelectInt32(in Index,in Values,out Result)"}
   Remap = &RigVMFunction_MathDoubleRemap(
    bClamp=False
    Value=@DISPATCH_RigVMDispatch_SelectInt32.Result
    TargetMinimum=@n1.outputMinMax.X
    TargetMaximum=@n1.outputMinMax.Y
    SourceMinimum=@n1.rotationMinMax.X
    SourceMaximum=@n1.rotationMinMax.Y
   )
   Absolute = &RigVMFunction_MathDoubleAbs(Value=@Remap.Result)
   Multiply = &RigVMFunction_MathDoubleMul(
    A=&RigVMFunction_MathDoubleClamp(Value=RigVMDispatchNode(True=@Absolute.Result, Condition=@n1.absoluteValue, False=@Remap.Result) {template="DISPATCH_RigVMDispatch_If(in Condition,in True,in False,out Result)"}.Result).Result
    B=@n1.premult
   )
   > RigVMFunctionReturnNode(outputQuat=@n2.Result, outputWeight=@Multiply.Result)
  ]}
 RigVMFunctionLibrary {nodes [DeltaRotation = RigVMCollapseNode(
    ExecuteContext=()
    Rotation Order=ZYX
    child=(Type=Bone,Name="None")
    outputMinMax=(X=0,Y=1)
    parent=(Type=Bone,Name="None")
    premult=1
    rotationMinMax=(X=0,Y=0)
    twistAxis=(X=1,Y=0,Z=0)
   )]}
 RigVMFunction_Sequence_1_ContainedGraph {nodes [
   n1 = RigVMFunctionEntryNode()
   RigVMFunction_Sequence = &RigVMFunction_Sequence(ExecuteContext=@n1.ExecuteContext)
   n2 = &RigVMFunction_Sequence(ExecuteContext=@RigVMFunction_Sequence.B)
   n3 = &RigVMFunction_Sequence(ExecuteContext=@n2.B)
   n4 = &RigVMFunction_Sequence(ExecuteContext=@n3.B)
   n5 = &RigVMFunction_Sequence(ExecuteContext=@n4.B)
   RigVMFunctionReturnNode(A=@RigVMFunction_Sequence.A, B=@n2.A, C=@n3.A, D=@n4.A, E=@n5.A, F=@n5.B)
  ]}
 RigVMModel {nodes [
   RigUnit_BeginExecution = &RigUnit_BeginExecution()
   RigVMFunction_Sequence = RigVMAggregateNode(ExecuteContext=@RigUnit_BeginExecution.ExecutePin)
   n1 = &RigUnit_ParentConstraint(
    AdvancedSettings=(InterpolationType=Average,RotationOrderForFilter=XZY)
    Child=(Type=Bone,Name="base_hair_tail_01")
    Filter=(TranslationFilter=(bX=false,bY=false,bZ=false),RotationFilter=(bX=true,bY=true,bZ=true),ScaleFilter=(bX=True,bY=True,bZ=True))
    Parents=((Item=(Type=Bone,Name="head"),Weight=0.6),(Item=(Type=Bone,Name="spine_05"),Weight=0.4))
    Weight=1
    ExecutePin=@RigVMFunction_Sequence.A
   )
   n2 = &RigVMFunction_ControlFlowBranch(ExecuteContext=@n1.ExecutePin, Condition=RigVMVariableNode() {variable=IsHeadUp}.Value)
   n3 = &RigUnit_SphericalPoseReader(
    ActiveRegionScaleFactors=(PositiveWidth=1,NegativeWidth=1,PositiveHeight=1,NegativeHeight=1)
    ActiveRegionSize=0.01
    Debug=(bDrawDebug=true,bDraw2D=False,bDrawLocalAxes=False,DebugScale=25,DebugSegments=20,DebugThickness=0.25)
    DriverAxis=(X=1,Y=0,Z=0)
    DriverItem=(Type=Bone,Name="head")
    FalloffRegionScaleFactors=(PositiveWidth=1,NegativeWidth=1,PositiveHeight=1,NegativeHeight=1)
    FalloffSize=0.12
    FlipHeightScaling=False
    FlipWidthScaling=False
    OptionalParentItem=(Type=Bone,Name="spine_05")
    RotationOffset=(X=0,Y=-80,Z=-25)
    ExecutePin=@n2.True
   )
   > &RigUnit_ModifyTransforms(
    ItemToModify=((Item=(Type=Bone,Name="base_hair_tail_01"),Transform=(Rotation=(X=0,Y=0.1305,Z=0,W=0.9914),Translation=(X=0,Y=0,Z=0),Scale3D=(X=1,Y=1,Z=1))))
    WeightMinimum=0
    Weight=@n3.OutputParam
   )
   > n4 = &RigUnit_SphericalPoseReader(
    ActiveRegionScaleFactors=(PositiveWidth=1,NegativeWidth=1,PositiveHeight=1,NegativeHeight=1)
    ActiveRegionSize=0.08
    Debug=(bDrawDebug=true,bDraw2D=False,bDrawLocalAxes=False,DebugScale=25,DebugSegments=20,DebugThickness=0.25)
    DriverAxis=(X=1,Y=0,Z=0)
    DriverItem=(Type=Bone,Name="head")
    FalloffRegionScaleFactors=(PositiveWidth=2,NegativeWidth=1,PositiveHeight=2,NegativeHeight=1)
    FalloffSize=0.15
    FlipHeightScaling=False
    FlipWidthScaling=False
    OptionalParentItem=(Type=Bone,Name="spine_05")
    RotationOffset=(X=-2.3098,Y=-60.8819,Z=-50)
   )
   > RigVMVariableNode(Value=@n4.OutputParam) {variable=HeadUp_Tragetting_Alpha}
   n5 = &RigVMFunction_AlphaInterp(
    InterpSpeedDecreasing=5
    InterpSpeedIncreasing=5
    Scale=1
    bClampResult=False
    bInterpResult=False
    Value=RigVMVariableNode() {variable=HeadUp_Tragetting_Alpha}.Value
   )
   > &RigUnit_ModifyTransforms(
    ItemToModify=((Item=(Type=Bone,Name="dyn_hair_tail_01"),Transform=(Rotation=(X=0,Y=-0.0436,Z=0,W=0.999),Translation=(X=-1,Y=0,Z=-2),Scale3D=(X=1,Y=1,Z=1))))
    WeightMinimum=0
    Weight=@n5.Result
   )
   n6 = &RigVMFunction_ControlFlowBranch(ExecuteContext=@RigVMFunction_Sequence.B, Condition=RigVMVariableNode() {variable=IsCrouching}.Value)
   n7 = &RigUnit_SphericalPoseReader(
    ActiveRegionScaleFactors=(PositiveWidth=1,NegativeWidth=1,PositiveHeight=1,NegativeHeight=1)
    ActiveRegionSize=0.06
    Debug=(bDrawDebug=false,bDraw2D=False,bDrawLocalAxes=False,DebugScale=25,DebugSegments=20,DebugThickness=0.25)
    DriverAxis=(X=1,Y=0,Z=0)
    DriverItem=(Type=Bone,Name="head")
    FalloffRegionScaleFactors=(PositiveWidth=1,NegativeWidth=1,PositiveHeight=1,NegativeHeight=1)
    FalloffSize=0.18
    FlipHeightScaling=False
    FlipWidthScaling=False
    OptionalParentItem=(Type=Bone,Name="spine_05")
    RotationOffset=(X=0.5305,Y=-62.3594,Z=-59.6979)
    ExecutePin=@n6.True
   )
   > RigVMVariableNode(Value=@n7.OutputParam) {variable=HeadUp_Tragetting_Crouch_Alpha}
   n8 = &RigVMFunction_AlphaInterp(
    InterpSpeedDecreasing=5
    InterpSpeedIncreasing=5
    Scale=1
    bClampResult=False
    bInterpResult=False
    Value=RigVMVariableNode() {variable=HeadUp_Tragetting_Crouch_Alpha}.Value
   )
   > &RigUnit_ModifyTransforms(
    ItemToModify=((Item=(Type=Bone,Name="dyn_hair_tail_01"),Transform=(Rotation=(X=0,Y=0,Z=0,W=1),Translation=(X=0,Y=0,Z=0),Scale3D=(X=1,Y=1,Z=1))))
    WeightMinimum=0
    Weight=@n8.Result
   )
   n9 = &RigVMFunction_ControlFlowBranch(ExecuteContext=@RigVMFunction_Sequence.C, Condition=RigVMVariableNode() {variable=IsCrouching_Aim_move}.Value)
   n10 = &RigVMFunction_AnimEvalRichCurve(
    Curve=(EditorCurveData=(Keys=((),(Time=0.5,Value=1),(Time=0.8))))
    SourceMaximum=2
    SourceMinimum=0
    TargetMinimum=0
    Value=&RigVMFunction_MathDoubleRemap(
     SourceMaximum=75
     SourceMinimum=45
     TargetMaximum=1
     TargetMinimum=0
     bClamp=false
     Value=&RigUnit_GetRelativeTransformForItem(Child=(Type=Bone,Name="pelvis"), Parent=(Type=Bone,Name="root"), bChildInitial=False, bParentInitial=False).RelativeTransform.Translation.Z
    ).Result
   )
   RigVMVariableNode(Value=@n10.Result, ExecuteContext=@n9.True) {variable=Crouch_move}
   > n11 = &RigUnit_SphericalPoseReader(
    ActiveRegionScaleFactors=(PositiveWidth=1,NegativeWidth=1,PositiveHeight=1,NegativeHeight=1)
    ActiveRegionSize=0.1
    Debug=(bDrawDebug=false,bDraw2D=False,bDrawLocalAxes=False,DebugScale=25,DebugSegments=20,DebugThickness=0.25)
    DriverAxis=(X=1,Y=0,Z=0)
    DriverItem=(Type=Bone,Name="head")
    FalloffRegionScaleFactors=(PositiveWidth=1,NegativeWidth=1,PositiveHeight=1,NegativeHeight=1)
    FalloffSize=0.2
    FlipHeightScaling=False
    FlipWidthScaling=False
    OptionalParentItem=(Type=Bone,Name="spine_05")
    RotationOffset=(X=4.7673,Y=-62.4175,Z=-100.6681)
   )
   > RigVMVariableNode(Value=@n11.OutputParam) {variable=HeadUp_Tragetting_move_Crouch_Alpha}
   n12 = &RigVMFunction_AlphaInterp(
    InterpSpeedDecreasing=10
    InterpSpeedIncreasing=10
    Scale=1
    bClampResult=False
    bInterpResult=False
    Value=&RigVMFunction_MathDoubleMul(A=RigVMVariableNode() {variable=HeadUp_Tragetting_move_Crouch_Alpha}.Value, B=RigVMVariableNode() {variable=Crouch_move}.Value).Result
   )
   > &RigUnit_ModifyTransforms(
    ItemToModify=((Item=(Type=Bone,Name="dyn_hair_tail_01"),Transform=(Rotation=(X=-0.017,Y=-0.1294,Z=-0.1294,W=0.983),Translation=(X=4,Y=0,Z=1),Scale3D=(X=0.5,Y=0.5,Z=0.5))))
    WeightMinimum=0
    Weight=@n12.Result
   )
   > &RigUnit_ModifyTransforms(
    ItemToModify="((Item=(Type=Bone,Name=\"dyn_lower_hair_a_l_01\"),Transform=(Rotation=(X=0,Y=-0.2588,Z=0,W=0.9659),Translation=(X=0,Y=0,Z=0),Scale3D=(X=1,Y=1,Z=1))),(Item=(Type=Bone,Name=\"dyn_lower_hair_b_l_01\"),Transform=(Rotation=(X=0,Y=-0.2588,Z=0,W=0.9659),Translation=( ...(truncated)"
    WeightMinimum=0
    Weight=@n12.Result
   )
   n13 = &RigVMFunction_ControlFlowBranch(ExecuteContext=@RigVMFunction_Sequence.D, Condition=RigVMVariableNode() {variable=IsHeadUp_Targetting}.Value)
   n14 = &RigUnit_SphericalPoseReader(
    ActiveRegionScaleFactors=(PositiveWidth=1,NegativeWidth=1,PositiveHeight=1,NegativeHeight=1)
    ActiveRegionSize=0
    Debug=(bDrawDebug=true,bDraw2D=False,bDrawLocalAxes=False,DebugScale=25,DebugSegments=20,DebugThickness=0.25)
    DriverAxis=(X=1,Y=0,Z=0)
    DriverItem=(Type=Bone,Name="head")
    FalloffRegionScaleFactors=(PositiveWidth=1,NegativeWidth=1,PositiveHeight=1,NegativeHeight=1)
    FalloffSize=0.1
    FlipHeightScaling=False
    FlipWidthScaling=False
    OptionalParentItem=(Type=Bone,Name="spine_05")
    RotationOffset=(X=-2,Y=-80,Z=-20)
    ExecutePin=@n13.True
   )
   n15 = &RigVMFunction_AlphaInterp(InterpSpeedDecreasing=5, InterpSpeedIncreasing=5, Scale=1, bClampResult=False, bInterpResult=False, Value=@n14.OutputParam)
   > &RigUnit_ModifyTransforms(
    ItemToModify=((Item=(Type=Bone,Name="base_hair_tail_01"),Transform=(Rotation=(X=0,Y=0.1736,Z=0,W=0.9848),Translation=(X=0,Y=0,Z=0),Scale3D=(X=1,Y=1,Z=1))))
    WeightMinimum=0
    Weight=@n15.Result
   )
   > n16 = &RigUnit_SphericalPoseReader(
    ActiveRegionScaleFactors=(PositiveWidth=1,NegativeWidth=1,PositiveHeight=1,NegativeHeight=1)
    ActiveRegionSize=0.08
    Debug=(bDrawDebug=true,bDraw2D=False,bDrawLocalAxes=False,DebugScale=25,DebugSegments=20,DebugThickness=0.25)
    DriverAxis=(X=1,Y=0,Z=0)
    DriverItem=(Type=Bone,Name="head")
    FalloffRegionScaleFactors=(PositiveWidth=1,NegativeWidth=1,PositiveHeight=1,NegativeHeight=1)
    FalloffSize=0.2
    FlipHeightScaling=False
    FlipWidthScaling=False
    OptionalParentItem=(Type=Bone,Name="spine_05")
    RotationOffset=(X=-2.3098,Y=-60.8819,Z=-49.9424)
   )
   > RigVMVariableNode(Value=@n16.OutputParam) {variable=HeadUp_Tragetting_Alpha}
   n17 = &RigVMFunction_AlphaInterp(
    InterpSpeedDecreasing=5
    InterpSpeedIncreasing=5
    Scale=1
    bClampResult=False
    bInterpResult=False
    Value=RigVMVariableNode() {variable=HeadUp_Tragetting_Alpha}.Value
   )
   > &RigUnit_ModifyTransforms(
    ItemToModify=((Item=(Type=Bone,Name="dyn_hair_tail_01"),Transform=(Rotation=(X=0,Y=0.0436,Z=0,W=0.999),Translation=(X=0,Y=0,Z=-0.5),Scale3D=(X=1,Y=1,Z=1))))
    WeightMinimum=0
    Weight=@n17.Result
   )
   > n18 = &RigUnit_SphericalPoseReader(
    ActiveRegionScaleFactors=(PositiveWidth=1,NegativeWidth=1,PositiveHeight=1,NegativeHeight=1)
    ActiveRegionSize=0.08
    Debug=(bDrawDebug=false,bDraw2D=False,bDrawLocalAxes=False,DebugScale=25,DebugSegments=20,DebugThickness=0.25)
    DriverAxis=(X=1,Y=0,Z=0)
    DriverItem=(Type=Bone,Name="head")
    FalloffRegionScaleFactors=(PositiveWidth=3,NegativeWidth=1,PositiveHeight=3,NegativeHeight=1)
    FalloffSize=0.15
    FlipHeightScaling=False
    FlipWidthScaling=False
    OptionalParentItem=(Type=Bone,Name="spine_05")
    RotationOffset=(X=-2.3098,Y=-60,Z=-30)
   )
   n19 = &RigVMFunction_AlphaInterp(InterpSpeedDecreasing=5, InterpSpeedIncreasing=5, Scale=1, bClampResult=False, bInterpResult=False, Value=@n18.OutputParam)
   > &RigUnit_ModifyTransforms(
    ItemToModify="((Item=(Type=Bone,Name=\"dyn_lower_hair_a_l_01\"),Transform=(Rotation=(X=0,Y=-0.1736,Z=0,W=0.9848),Translation=(X=0,Y=0,Z=0),Scale3D=(X=1,Y=1,Z=1))),(Item=(Type=Bone,Name=\"dyn_lower_hair_b_l_01\"),Transform=(Rotation=(X=0,Y=-0.2588,Z=0,W=0.9659),Translation=( ...(truncated)"
    WeightMinimum=0
    Weight=@n19.Result
   )
   n20 = &RigVMFunction_ControlFlowBranch(ExecuteContext=@RigVMFunction_Sequence.E, Condition=RigVMVariableNode() {variable=Skydive_Bool}.Value)
   n21 = &RigUnit_SphericalPoseReader(
    ActiveRegionScaleFactors=(PositiveWidth=1,NegativeWidth=1,PositiveHeight=1,NegativeHeight=1)
    ActiveRegionSize=0.08
    Debug=(bDrawDebug=false,bDraw2D=False,bDrawLocalAxes=False,DebugScale=25,DebugSegments=20,DebugThickness=0.25)
    DriverAxis=(X=1,Y=0,Z=0)
    DriverItem=(Type=Bone,Name="head")
    FalloffRegionScaleFactors=(PositiveWidth=1,NegativeWidth=1,PositiveHeight=1,NegativeHeight=1)
    FalloffSize=0.15
    FlipHeightScaling=False
    FlipWidthScaling=False
    OptionalParentItem=(Type=Bone,Name="spine_05")
    RotationOffset=(X=-6.2317,Y=-91.482,Z=-33.3576)
    ExecutePin=@n20.True
   )
   > RigVMVariableNode(Value=@n21.OutputParam) {variable=HeadUp_Skydive}
   n22 = &RigVMFunction_AlphaInterp(
    InterpSpeedDecreasing=5
    InterpSpeedIncreasing=5
    Scale=1
    bClampResult=False
    bInterpResult=False
    Value=RigVMVariableNode() {variable=HeadUp_Skydive}.Value
   )
   > &RigUnit_ModifyTransforms(
    ItemToModify=((Item=(Type=Bone,Name="dyn_hair_tail_01"),Transform=(Rotation=(X=0,Y=-0.0872,Z=0,W=0.9962),Translation=(X=0,Y=0,Z=-1),Scale3D=(X=1,Y=1,Z=1))))
    WeightMinimum=0
    Weight=@n22.Result
   )
   &RigUnit_ModifyTransforms(
    ItemToModify=((Item=(Type=Bone,Name="dyn_hair_tail_01"),Transform=(Rotation=(X=0,Y=0,Z=0,W=1),Translation=(X=0,Y=0,Z=-2),Scale3D=(X=1,Y=1,Z=1))))
    WeightMinimum=0
    ExecutePin=@RigVMFunction_Sequence.F
    Weight=RigVMVariableNode() {variable=RifleUP}.Value
   )
   &RigUnit_ModifyTransforms(
    ExecutePin=()
    ItemToModify="((Item=(Type=Bone,Name=\"dyn_lower_hair_a_l_01\"),Transform=(Rotation=(X=0,Y=-0.342,Z=0,W=0.9397),Translation=(X=0,Y=0,Z=0),Scale3D=(X=1,Y=1,Z=1))),(Item=(Type=Bone,Name=\"dyn_lower_hair_b_l_01\"),Transform=(Rotation=(X=0,Y=-0.2588,Z=0,W=0.9659),Translation=(X ...(truncated)"
    WeightMinimum=0
    Weight=@n5.Result
   )
   RigVMVariableNode() {variable=HeadUp_Tragetting_Alpha}
  ]}
]
hierarchy [
 root {loc=<0 0 0>, rot=<0 0 0 1>}
 pelvis {parent=root, loc=<0 -0.7071 93.1191>, rot=<0.0127 0.707 -0.0127 -0.707>}
 spine_01 {parent=pelvis, loc=<6.7494 -0.6388 0>, rot=<0 0 0.034 -0.9994>}
 spine_02 {parent=spine_01, loc=<7.2518 -0.035 0>, rot=<0 0 0.0238 -0.9997>}
 spine_03 {parent=spine_02, loc=<6.0104 0 0>, rot=<0 0 0.0149 -0.9999>}
 spine_04 {parent=spine_03, loc=<8.1024 0 0>, rot=<0 0 -0.0132 -0.9999>}
 spine_05 {parent=spine_04, loc=<10.0146 0 0>, rot=<0 0 -0.0015 -1>}
 neck_01 {parent=spine_05, loc=<11.3138 0.1529 0>, rot=<0 0 -0.2003 -0.9797>}
 neck_02 {parent=neck_01, loc=<4.5627 0 0>, rot=<0 0 0.0504 -0.9987>}
 head {parent=neck_02, loc=<5.3914 0 0>, rot=<0 0 0.115 -0.9934>}
 base_hair_tail_01 {parent=head, loc=<1.9479 -8.1627 0.3331>, rot=<0.3217 -0.5827 0.6281 -0.4031>}
 dyn_hair_tail_01 {parent=base_hair_tail_01, loc=<0 0 0>, rot=<0 0 0 1>}
 dyn_hair_tail_lwr_01 {parent=dyn_hair_tail_01, loc=<4.1296 -0.4916 -2.301>, rot=<0.1037 -0.1559 0.0165 -0.9822>}
 dyn_hair_tail_uppr_01 {parent=dyn_hair_tail_01, loc=<3.2558 -0.311 1.6972>, rot=<0.1018 0.2436 -0.0257 -0.9642>}
 dyn_hair_tail_mid_01 {parent=dyn_hair_tail_01, loc=<3.9957 -0.0526 -0.2462>, rot=<0.105 0 0 -0.9945>}
 base_l_upper_hair_01 {parent=head, loc=<5.4899 -7.7417 -5.976>, rot=<0.393 -0.4225 0.3106 -0.7553>}
 dyn_l_upper_hair_01 {parent=base_l_upper_hair_01, loc=<0 0 0>, rot=<0 0 0 1>}
 base_l_upper_hair_02 {parent=head, loc=<4.7097 -8.3265 -3.1027>, rot=<0.464 -0.5296 0.4092 -0.5803>}
 dyn_l_upper_hair_02 {parent=base_l_upper_hair_02, loc=<0 0 0>, rot=<0 0 0 1>}
 base_l_upper_hair_03 {parent=head, loc=<4.9982 -3.4697 -8.0595>, rot=<0.3276 -0.5571 0.3791 -0.6623>}
 dyn_l_upper_hair_03 {parent=base_l_upper_hair_03, loc=<0 0 0>, rot=<0 0 0 1>}
 base_l_upper_hair_04 {parent=head, loc=<2.2911 -4.5567 -7.663>, rot=<0.3243 -0.8782 0.2174 -0.2762>}
 dyn_l_upper_hair_04 {parent=base_l_upper_hair_04, loc=<0 0 0>, rot=<0 0 0 1>}
 base_r_upper_hair_01 {parent=head, loc=<5.2989 -7.224 5.612>, rot=<0.7799 -0.4551 0.3536 -0.2442>}
 dyn_r_upper_hair_01 {parent=base_r_upper_hair_01, loc=<0 0 0>, rot=<0 0 0 1>}
 base_r_upper_hair_02 {parent=head, loc=<5.582 -5.7347 7.0515>, rot=<0.6451 -0.236 0.5396 -0.4868>}
 dyn_r_upper_hair_02 {parent=base_r_upper_hair_02, loc=<0 0 0>, rot=<0 0 0 1>}
 base_r_upper_hair_03 {parent=head, loc=<4.9701 -3.0133 7.6834>, rot=<0.3995 -0.0795 0.8337 -0.373>}
 dyn_r_upper_hair_03 {parent=base_r_upper_hair_03, loc=<0 0 0>, rot=<0 0 0 1>}
 base_l_side_hair_3a {parent=head, loc=<0.6421 2.7489 -6.7146>, rot=<-0.1901 -0.9724 0.027 0.1322>}
 dyn_l_side_hair_3a {parent=base_l_side_hair_3a, loc=<0 0 0>, rot=<0 0 0 1>}
 dyn_l_side_hair_3b {parent=dyn_l_side_hair_3a, loc=<1.854 0 0>, rot=<0 0 0.0825 -0.9966>}
 base_l_side_hair_2a {parent=head, loc=<1.678 4.6969 -6.6905>, rot=<-0.2696 -0.9615 0.0526 -0.0086>}
 dyn_l_side_hair_2a {parent=base_l_side_hair_2a, loc=<0 0 0>, rot=<0 0 0 1>}
 dyn_l_side_hair_2b {parent=dyn_l_side_hair_2a, loc=<1.854 0 0>, rot=<0 0 0.0825 -0.9966>}
 base_l_side_hair_1a {parent=head, loc=<4.2404 5.7462 -6.558>, rot=<-0.3296 -0.9303 -0.1519 -0.0534>}
 dyn_l_side_hair_1a {parent=base_l_side_hair_1a, loc=<0 0 0>, rot=<0 0 0 1>}
 dyn_l_side_hair_1b {parent=dyn_l_side_hair_1a, loc=<1.9832 -0.0349 -0.4515>, rot=<-0.0545 -0.2507 -0.0472 -0.9654>}
 base_r_side_hair_1a {parent=head, loc=<4.6215 7.7758 6.9122>, rot=<0.1565 -0.0131 0.9834 -0.0908>}
 dyn_r_side_hair_1a {parent=base_r_side_hair_1a, loc=<0 0 0>, rot=<0 0 0 1>}
 dyn_r_side_hair_1b {parent=dyn_r_side_hair_1a, loc=<1.9667 0 0>, rot=<0 0 -0.0551 -0.9985>}
 base_r_side_hair_2a {parent=head, loc=<2.2128 5.4974 6.7353>, rot=<-0.1091 0.2216 0.9554 0.1617>}
 dyn_r_side_hair_2a {parent=base_r_side_hair_2a, loc=<0 0 0>, rot=<0 0 0 1>}
 dyn_r_side_hair_2b {parent=dyn_r_side_hair_2a, loc=<2.2694 0.0031 0.0185>, rot=<-0.011 0.2485 0.0737 -0.9658>}
 base_r_side_hair_3a {parent=head, loc=<1.5168 3.4346 7.1779>, rot=<-0.0331 0.0792 0.9781 0.1895>}
 dyn_r_side_hair_3a {parent=base_r_side_hair_3a, loc=<0 0 0>, rot=<0 0 0 1>}
 dyn_r_side_hair_3b {parent=dyn_r_side_hair_3a, loc=<2.4228 0.0392 -0.1019>, rot=<0.0135 -0.129 -0.0159 -0.9914>}
 base_m_upper_hair_01 {parent=head, loc=<5.266 -8.7525 0.1821>, rot=<0.6946 -0.2551 0.5177 -0.4295>}
 dyn_m_upper_hair_01 {parent=base_m_upper_hair_01, loc=<0 0 0>, rot=<0 0 0 1>}
 head_null {parent=head, type=Null, loc=<0 0 0>, rot=<0 0 0 1>}
 base_lower_hair_a_l_01 {parent=head, loc=<-0.9323 -3.5172 -5.6678>, rot=<-0.0068 -0.8749 0.4286 -0.2255>}
 dyn_lower_hair_a_l_01 {parent=base_lower_hair_a_l_01, loc=<0 0 0>, rot=<0 0 0 1>}
 dyn_lower_hair_a_l_02 {parent=dyn_lower_hair_a_l_01, loc=<1.7094 0 -0.0286>, rot=<0 0.1674 0 -0.9859>}
 base_lower_hair_b_l_01 {parent=head, loc=<0.2999 -4.5675 -5.4524>, rot=<0.2887 -0.8073 0.5132 0.0396>}
 dyn_lower_hair_b_l_01 {parent=base_lower_hair_b_l_01, loc=<0 0 0>, rot=<0 0 0 1>}
 dyn_lower_hair_b_l_02 {parent=dyn_lower_hair_b_l_01, loc=<1.6596 0 -0.0316>, rot=<0 0.0304 0 -0.9995>}
 base_lower_hair_c_l_01 {parent=head, loc=<0.7266 -5.411 -3.2432>, rot=<0.0531 -0.8544 0.5023 -0.1218>}
 dyn_lower_hair_c_l_01 {parent=base_lower_hair_c_l_01, loc=<0 0 0>, rot=<0 0 0 1>}
 dyn_lower_hair_c_l_02 {parent=dyn_lower_hair_c_l_01, loc=<2.011 0 -0.0102>, rot=<0 0.2443 0 -0.9697>}
 clavicle_l {parent=spine_05, loc=<4.3575 -0.1466 -2.3223>, rot=<0.0465 -0.7289 -0.009 -0.6829>}
 upperarm_l {parent=clavicle_l, loc=<15.5111 0 0>, rot=<0.0226 -0.3234 -0.0088 -0.9459>}
 lowerarm_l {parent=upperarm_l, loc=<26.6514 0 0>, rot=<0.0391 -0.0099 -0.2453 -0.9686>}
 upperarm_twist_01_l {parent=upperarm_l, loc=<0.0041 0 0>, rot=<0 0 0 -1>}
 clavicle_r {parent=spine_05, loc=<4.3575 -0.1466 2.3223>, rot=<0.7289 0.0465 -0.6829 0.009>}
 upperarm_r {parent=clavicle_r, loc=<-15.5111 0 0>, rot=<0.0226 -0.3234 -0.0088 -0.9459>}
 lowerarm_r {parent=upperarm_r, loc=<-26.6514 0 0>, rot=<0.0391 -0.0099 -0.2453 -0.9686>}
 upperarm_twist_01_r {parent=upperarm_r, loc=<-0.0041 0 0>, rot=<0 0 0 -1>}
 attach_cape {parent=spine_05, loc=<9.017 -11.714 0>, rot=<0.0616 -0.7044 -0.0616 -0.7044>}
 attach_backpack {parent=spine_05, loc=<-0.123 -13.473 0>, rot=<-0.0093 -0.707 0.0093 -0.707>}
]
variables [
 CrouchAlpha_Extreme {type=double, default=0}
 Crouch_move {type=double, default=0}
 HeadRot {type=double, default=0}
 HeadUp_Skydive {type=double, default=0}
 HeadUp_Tragetting_Alpha {type=double, default=0}
 HeadUp_Tragetting_Crouch_Alpha {type=double, default=0}
 HeadUp_Tragetting_move_Crouch_Alpha {type=double, default=0}
 IsCrouching {type=bool, default=False}
 IsCrouching_Aim_move {type=bool, default=False}
 IsHeadUp {type=bool, default=False}
 IsHeadUp_Targetting {type=bool, default=False}
 IsHeadUp_float {type=double, default=0}
 RifleUP {type=double, default=0}
 Skydive {type=bool, default=False}
 Skydive_Bool {type=bool, default=False}
]
