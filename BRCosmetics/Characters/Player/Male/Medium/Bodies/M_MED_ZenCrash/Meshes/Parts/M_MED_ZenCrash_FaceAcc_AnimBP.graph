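# Animation Blueprint for the ZenCrash male character's face-accessory part, adding secondary motion to its hair.
# The parent mesh pose runs through a Control Rig, a head-space RigidBody node, then chained spring AnimDynamics for tail, upper and side hair.
# EventGraph derives gravity settings and many head-angle/stance alphas, but only the standing-while-targeting alpha visibly feeds the Control Rig.
# Several tail-bone Transform nodes and one AnimDynamics node sit disconnected from the output.
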AnimBlueprint /BRCosmetics/Characters/Player/Male/Medium/Bodies/M_MED_ZenCrash/Meshes/Parts/M_MED_ZenCrash_FaceAcc_AnimBP
target_skeleton=&/M_MED_ZenCrash_FaceAcc_Skeleton
graphs [
 AnimGraph {nodes [
   n1 = "Transform (Modify) Bone - Bone: dyn_hair_tail_01"(Alpha=1, Translation="0,0,0.5", Rotation="10,0,0", Scale="1,1,1") {
    rotationMode=BMM_Additive
    rotationSpace=BCS_BoneSpace
    translationMode=BMM_Additive
    translationSpace=BCS_BoneSpace
    alphaScaleBiasClamp {bInterpResult=true, interpSpeedIncreasing=4}
    boneToModify {boneName=dyn_hair_tail_01}
    rotation {pitch=10}
    translation {z=0.5}
   }
   n2 = "Transform (Modify) Bone - Bone: dyn_hair_tail_01"(ComponentPose=@n1, Alpha=1, Translation="2,0,1", Rotation="25,0,0", Scale="1,1,1") {
    rotationMode=BMM_Additive
    rotationSpace=BCS_BoneSpace
    translationMode=BMM_Additive
    translationSpace=BCS_BoneSpace
    alphaScaleBiasClamp {bInterpResult=true, interpSpeedIncreasing=4}
    boneToModify {boneName=dyn_hair_tail_01}
    rotation {pitch=25}
    translation {x=2, z=1}
   }
   n3 = "Transform (Modify) Bone - Bone: base_hair_tail_01"(ComponentPose=@n2, Alpha=1, Translation="0,0,0", Rotation="15,-30,0", Scale="1,1,1") {
    rotationMode=BMM_Additive
    rotationSpace=BCS_BoneSpace
    translationMode=BMM_Additive
    translationSpace=BCS_BoneSpace
    alphaScaleBiasClamp {bInterpResult=true}
    boneToModify {boneName=base_hair_tail_01}
    rotation {pitch=15, yaw=-30}
   }
   n4 = "Transform (Modify) Bone - Bone: base_hair_tail_01"(ComponentPose=@n3, Alpha=1, Translation="0,1,0", Rotation="15,30,0", Scale="1,1,1") {
    rotationMode=BMM_Additive
    rotationSpace=BCS_BoneSpace
    translationMode=BMM_Additive
    translationSpace=BCS_BoneSpace
    alphaScaleBiasClamp {bInterpResult=true}
    boneToModify {boneName=base_hair_tail_01}
    rotation {pitch=15, yaw=30}
    translation {y=1}
   }
   n5 = "Transform (Modify) Bone - Bone: dyn_hair_tail_01"(ComponentPose=@n4, Alpha=1, Translation="0,0,0", Rotation="20,0,0", Scale="1,1,1") {
    rotationMode=BMM_Additive
    rotationSpace=BCS_BoneSpace
    translationMode=BMM_Additive
    translationSpace=BCS_BoneSpace
    alphaScaleBiasClamp {bInterpResult=true}
    boneToModify {boneName=dyn_hair_tail_01}
    rotation {pitch=20}
   }
   Anim Dynamics - Bone: dyn_hair_tail_01!(ComponentPose=@n5, Alpha=1, GravityOverride:Vector="0,0,-980", ExternalForce="0,0,0") -> (Pose) {
    angularDampingOverride=0.9999
    angularSpringConstant=7000
    bAngularSpring=true
    bOverrideAngularDamping=true
    bUseGravityOverride=true
    bUseSphericalLimits=true
    gravityScale=0.35
    simulationSpace=RootRelative
    boundBone {boneName=dyn_hair_tail_01}
    componentLinearAccScale=<1 1 1>
    componentLinearVelScale=<1 1 1>
    gravityOverride {z=-980}
    physicsBodyDefinitions [{
      boundBone {boneName=dyn_hair_tail_01}
      boxExtents=<20 20 20>
      constraintSetup {angularLimitsMax=<0 1 3>, angularLimitsMin=<0 -5 -3>, angularTarget=<1 0 0>}
      localJointOffset=<4 0 0>
     }]
    sphericalLimits [{limitRadius=8, drivingBone {boneName=head}, sphereLocalOffset=<0 0 0>}]
   }
   Local To Component()
   n6 = "Transform (Modify) Bone - Bone: dyn_hair_tail_01"(Alpha=1, Translation="0,0,0", Rotation="10,0,0", Scale="1,1,1") {
    rotationMode=BMM_Additive
    rotationSpace=BCS_BoneSpace
    translationMode=BMM_Additive
    translationSpace=BCS_BoneSpace
    alphaScaleBiasClamp {bInterpResult=true}
    boneToModify {boneName=dyn_hair_tail_01}
    rotation {pitch=10}
   }
   "Transform (Modify) Bone - Bone: dyn_hair_tail_01"(ComponentPose=@n6, Alpha=1, Translation="0,0,0", Rotation="15,0,0", Scale="1,1,1") {
    rotationMode=BMM_Additive
    rotationSpace=BCS_BoneSpace
    translationMode=BMM_Additive
    translationSpace=BCS_BoneSpace
    alphaScaleBiasClamp {bInterpResult=true}
    boneToModify {boneName=dyn_hair_tail_01}
    rotation {pitch=15}
   }
   "Transform (Modify) Bone - Bone: dyn_hair_tail_01"(Alpha=1, Translation="0,0,0", Rotation="30,0,0", Scale="1,1,1") {
    rotationMode=BMM_Additive
    rotationSpace=BCS_BoneSpace
    translationMode=BMM_Additive
    translationSpace=BCS_BoneSpace
    alphaScaleBiasClamp {bInterpResult=true}
    boneToModify {boneName=dyn_hair_tail_01}
    rotation {pitch=30}
   }
   "Transform (Modify) Bone - Bone: base_hair_tail_01"(Alpha=1, Translation="0,0,0", Rotation="5,0,0", Scale="1,1,1") {
    rotationMode=BMM_Additive
    rotationSpace=BCS_BoneSpace
    translationMode=BMM_Additive
    translationSpace=BCS_BoneSpace
    alphaScaleBiasClamp {bInterpResult=true}
    boneToModify {boneName=base_hair_tail_01}
    rotation {pitch=5}
   }
   n7 = Control Rig!(
    Source=Copy Pose From Mesh() {bUseAttachedParent=true}
    Alpha
    HeadUp_Tragetting_Alpha:double=Get HeadUp_Tragetting_Alpha() -> (HeadUp_Tragetting_Alpha:double)
   ) -> (Pose) {controlRigAssetReference {blueprintRigClass=&/M_MED_ZenCrash_FaceAcc_CtrlRig.M_MED_ZenCrash_FaceAcc_CtrlRig_C}}
   n8 = RigidBody!(
    ComponentPose=Local To Component(LocalPose=@n7)
    Alpha
    OverrideWorldGravity:Vector="0,0,0"
    ExternalForce
    SimSpaceSettings:Sim Space Settings Structure=(WorldAlpha=1,VelocityScaleZ=1,DampingAlpha=1,MaxLinearVelocity=10000,MaxAngularVelocity=10000,MaxLinearAcceleration=10000,MaxAngularAcceleration=10000,ExternalLinearDragV=(X=0,Y=0,Z=0),ExternalLinearVelocity=(X=0,Y=0,Z=0),ExternalAngularVelocity=(X=0,Y=0,Z=0))
   ) -> (Pose) {
    bOverrideWorldGravity=true
    overridePhysicsAsset=&/M_MED_ZenCrash_FaceAcc_Physics
    simulationSpace=BaseBoneSpace
    baseBoneRef {boneName=head}
    simSpaceSettings {worldAlpha=1}
   }
   n9 = Anim Dynamics - Bone: dyn_hair_tail_uppr_01!(ComponentPose=@n8, Alpha=1, GravityOverride:Vector="0,0,-980", ExternalForce="0,0,0") -> (Pose) {
    angularDampingOverride=0.9999
    angularSpringConstant=5000
    bAngularSpring=true
    bOverrideAngularDamping=true
    bUseGravityOverride=true
    bUseSphericalLimits=true
    gravityScale=0.35
    simulationSpace=RootRelative
    boundBone {boneName=dyn_hair_tail_uppr_01}
    componentLinearAccScale=<1 1 1>
    componentLinearVelScale=<1 1 1>
    gravityOverride {z=-980}
    physicsBodyDefinitions [{
      boundBone {boneName=dyn_hair_tail_uppr_01}
      boxExtents=<20 20 20>
      constraintSetup {angularLimitsMax=<0 2 5>, angularLimitsMin=<0 -8 -5>, angularTarget=<1 0 0>}
      localJointOffset=<4 0 0>
     }]
    sphericalLimits [{limitRadius=8, drivingBone {boneName=head}, sphereLocalOffset=<0 0 0>}]
   }
   n10 = Anim Dynamics - Bone: dyn_hair_tail_mid_01!(ComponentPose=@n9, Alpha=1, GravityOverride:Vector="0,0,-980", ExternalForce="0,0,0") -> (Pose) {
    angularDampingOverride=0.9999
    angularSpringConstant=5000
    bAngularSpring=true
    bOverrideAngularDamping=true
    bUseGravityOverride=true
    bUseSphericalLimits=true
    gravityScale=0.35
    simulationSpace=RootRelative
    boundBone {boneName=dyn_hair_tail_mid_01}
    componentLinearAccScale=<1 1 1>
    componentLinearVelScale=<1 1 1>
    gravityOverride {z=-980}
    physicsBodyDefinitions [{
      boundBone {boneName=dyn_hair_tail_mid_01}
      boxExtents=<20 20 20>
      constraintSetup {angularLimitsMax=<0 2 5>, angularLimitsMin=<0 -8 -5>, angularTarget=<1 0 0>}
      localJointOffset=<4 0 0>
     }]
    sphericalLimits [{limitRadius=8, drivingBone {boneName=head}, sphereLocalOffset=<0 0 0>}]
   }
   n11 = Anim Dynamics - Bone: dyn_hair_tail_lwr_01!(ComponentPose=@n10, Alpha=1, GravityOverride:Vector="0,0,-980", ExternalForce="0,0,0") -> (Pose) {
    angularDampingOverride=0.9999
    angularSpringConstant=5000
    bAngularSpring=true
    bOverrideAngularDamping=true
    bUseGravityOverride=true
    bUseSphericalLimits=true
    gravityScale=0.35
    simulationSpace=RootRelative
    boundBone {boneName=dyn_hair_tail_lwr_01}
    componentLinearAccScale=<1 1 1>
    componentLinearVelScale=<1 1 1>
    gravityOverride {z=-980}
    physicsBodyDefinitions [{
      boundBone {boneName=dyn_hair_tail_lwr_01}
      boxExtents=<20 20 20>
      constraintSetup {angularLimitsMax=<0 2 5>, angularLimitsMin=<0 -8 -5>, angularTarget=<1 0 0>}
      localJointOffset=<4 0 0>
     }]
    sphericalLimits [{limitRadius=8, drivingBone {boneName=head}, sphereLocalOffset=<0 0 0>}]
   }
   n12 = Anim Dynamics - Bone: dyn_m_upper_hair_01!(ComponentPose=@n11, Alpha=1, GravityOverride:Vector="0,0,-980", ExternalForce="0,0,0") -> (Pose) {
    angularDampingOverride=0.9999
    angularSpringConstant=10000
    bAngularSpring=true
    bOverrideAngularDamping=true
    bUseGravityOverride=true
    gravityScale=0.35
    simulationSpace=RootRelative
    boundBone {boneName=dyn_m_upper_hair_01}
    componentLinearAccScale=<1 1 1>
    componentLinearVelScale=<1 1 1>
    gravityOverride {z=-980}
    physicsBodyDefinitions [{
      boundBone {boneName=dyn_m_upper_hair_01}
      boxExtents=<15 15 15>
      constraintSetup {angularLimitsMax=<0 10 5>, angularLimitsMin=<0 -10 -5>, angularTarget=<1 0 0>}
      localJointOffset=<4 0 0>
     }]
   }
   n13 = Anim Dynamics - Bone: dyn_l_upper_hair_01!(ComponentPose=@n12, Alpha=1, GravityOverride:Vector="0,0,-980", ExternalForce="0,0,0") -> (Pose) {
    angularDampingOverride=0.9999
    angularSpringConstant=10000
    bAngularSpring=true
    bOverrideAngularDamping=true
    bUseGravityOverride=true
    gravityScale=0.35
    simulationSpace=RootRelative
    boundBone {boneName=dyn_l_upper_hair_01}
    componentLinearAccScale=<1 1 1>
    componentLinearVelScale=<1 1 1>
    gravityOverride {z=-980}
    physicsBodyDefinitions [{
      boundBone {boneName=dyn_l_upper_hair_01}
      boxExtents=<15 15 15>
      constraintSetup {angularLimitsMax=<0 10 5>, angularLimitsMin=<0 -10 -5>, angularTarget=<1 0 0>}
      localJointOffset=<4 0 0>
     }]
   }
   n14 = Anim Dynamics - Bone: dyn_l_upper_hair_02!(ComponentPose=@n13, Alpha=1, GravityOverride:Vector="0,0,-980", ExternalForce="0,0,0") -> (Pose) {
    angularDampingOverride=0.9999
    angularSpringConstant=10000
    bAngularSpring=true
    bOverrideAngularDamping=true
    bUseGravityOverride=true
    gravityScale=0.35
    simulationSpace=RootRelative
    boundBone {boneName=dyn_l_upper_hair_02}
    componentLinearAccScale=<1 1 1>
    componentLinearVelScale=<1 1 1>
    gravityOverride {z=-980}
    physicsBodyDefinitions [{
      boundBone {boneName=dyn_l_upper_hair_02}
      boxExtents=<15 15 15>
      constraintSetup {angularLimitsMax=<0 10 5>, angularLimitsMin=<0 -10 -5>, angularTarget=<1 0 0>}
      localJointOffset=<4 0 0>
     }]
   }
   n15 = Anim Dynamics - Bone: dyn_l_upper_hair_03!(ComponentPose=@n14, Alpha=1, GravityOverride:Vector="0,0,-980", ExternalForce="0,0,0") -> (Pose) {
    angularDampingOverride=0.9999
    angularSpringConstant=10000
    bAngularSpring=true
    bOverrideAngularDamping=true
    bUseGravityOverride=true
    gravityScale=0.35
    simulationSpace=RootRelative
    boundBone {boneName=dyn_l_upper_hair_03}
    componentLinearAccScale=<1 1 1>
    componentLinearVelScale=<1 1 1>
    gravityOverride {z=-980}
    physicsBodyDefinitions [{
      boundBone {boneName=dyn_l_upper_hair_03}
      boxExtents=<15 15 15>
      constraintSetup {angularLimitsMax=<0 10 5>, angularLimitsMin=<0 -10 -5>, angularTarget=<1 0 0>}
      localJointOffset=<4 0 0>
     }]
   }
   n16 = Anim Dynamics - Bone: dyn_l_upper_hair_04!(ComponentPose=@n15, Alpha=1, GravityOverride:Vector="0,0,-980", ExternalForce="0,0,0") -> (Pose) {
    angularDampingOverride=0.9999
    angularSpringConstant=10000
    bAngularSpring=true
    bOverrideAngularDamping=true
    bUseGravityOverride=true
    gravityScale=0.35
    simulationSpace=RootRelative
    boundBone {boneName=dyn_l_upper_hair_04}
    componentLinearAccScale=<1 1 1>
    componentLinearVelScale=<1 1 1>
    gravityOverride {z=-980}
    physicsBodyDefinitions [{
      boundBone {boneName=dyn_l_upper_hair_04}
      boxExtents=<15 15 15>
      constraintSetup {angularLimitsMax=<0 10 5>, angularLimitsMin=<0 -10 -5>, angularTarget=<1 0 0>}
      localJointOffset=<4 0 0>
     }]
   }
   n17 = Anim Dynamics - Bone: dyn_r_upper_hair_01!(ComponentPose=@n16, Alpha=1, GravityOverride:Vector="0,0,-980", ExternalForce="0,0,0") -> (Pose) {
    angularDampingOverride=0.9999
    angularSpringConstant=10000
    bAngularSpring=true
    bOverrideAngularDamping=true
    bUseGravityOverride=true
    gravityScale=0.35
    simulationSpace=RootRelative
    boundBone {boneName=dyn_r_upper_hair_01}
    componentLinearAccScale=<1 1 1>
    componentLinearVelScale=<1 1 1>
    gravityOverride {z=-980}
    physicsBodyDefinitions [{
      boundBone {boneName=dyn_r_upper_hair_01}
      boxExtents=<15 15 15>
      constraintSetup {angularLimitsMax=<0 10 5>, angularLimitsMin=<0 -10 -5>, angularTarget=<1 0 0>}
      localJointOffset=<4 0 0>
     }]
   }
   n18 = Anim Dynamics - Bone: dyn_r_upper_hair_02!(ComponentPose=@n17, Alpha=1, GravityOverride:Vector="0,0,-980", ExternalForce="0,0,0") -> (Pose) {
    angularDampingOverride=0.9999
    angularSpringConstant=10000
    bAngularSpring=true
    bOverrideAngularDamping=true
    bUseGravityOverride=true
    gravityScale=0.35
    simulationSpace=RootRelative
    boundBone {boneName=dyn_r_upper_hair_02}
    componentLinearAccScale=<1 1 1>
    componentLinearVelScale=<1 1 1>
    gravityOverride {z=-980}
    physicsBodyDefinitions [{
      boundBone {boneName=dyn_r_upper_hair_02}
      boxExtents=<15 15 15>
      constraintSetup {angularLimitsMax=<0 10 5>, angularLimitsMin=<0 -10 -5>, angularTarget=<1 0 0>}
      localJointOffset=<4 0 0>
     }]
   }
   n19 = Anim Dynamics - Bone: dyn_r_upper_hair_03!(ComponentPose=@n18, Alpha=1, GravityOverride:Vector="0,0,-980", ExternalForce="0,0,0") -> (Pose) {
    angularDampingOverride=0.9999
    angularSpringConstant=10000
    bAngularSpring=true
    bOverrideAngularDamping=true
    bUseGravityOverride=true
    gravityScale=0.35
    simulationSpace=RootRelative
    boundBone {boneName=dyn_r_upper_hair_03}
    componentLinearAccScale=<1 1 1>
    componentLinearVelScale=<1 1 1>
    gravityOverride {z=-980}
    physicsBodyDefinitions [{
      boundBone {boneName=dyn_r_upper_hair_03}
      boxExtents=<15 15 15>
      constraintSetup {angularLimitsMax=<0 10 5>, angularLimitsMin=<0 -10 -5>, angularTarget=<1 0 0>}
      localJointOffset=<4 0 0>
     }]
   }
   n20 = Anim Dynamics - Bone: dyn_l_side_hair_1a!(ComponentPose=@n19, Alpha=1, GravityOverride:Vector="0,0,-980", ExternalForce="0,0,0") -> (Pose) {
    angularDampingOverride=0.9999
    angularSpringConstant=5000
    bAngularSpring=true
    bOverrideAngularDamping=true
    bUseGravityOverride=true
    gravityScale=0.35
    simulationSpace=RootRelative
    boundBone {boneName=dyn_l_side_hair_1a}
    componentLinearAccScale=<1 1 1>
    componentLinearVelScale=<1 1 1>
    gravityOverride {z=-980}
    physicsBodyDefinitions [{
      boundBone {boneName=dyn_l_side_hair_1a}
      boxExtents=<20 20 20>
      constraintSetup {angularLimitsMax=<0 8 5>, angularLimitsMin=<0 -8 -5>, angularTarget=<1 0 0>}
      localJointOffset=<4 0 0>
     }]
   }
   n21 = Anim Dynamics - Bone: dyn_l_side_hair_1b!(ComponentPose=@n20, Alpha=1, GravityOverride:Vector="0,0,-980", ExternalForce="0,0,0") -> (Pose) {
    angularDampingOverride=0.9999
    angularSpringConstant=5000
    bAngularSpring=true
    bOverrideAngularDamping=true
    bUseGravityOverride=true
    gravityScale=0.35
    simulationSpace=RootRelative
    boundBone {boneName=dyn_l_side_hair_1b}
    componentLinearAccScale=<1 1 1>
    componentLinearVelScale=<1 1 1>
    gravityOverride {z=-980}
    physicsBodyDefinitions [{
      boundBone {boneName=dyn_l_side_hair_1b}
      boxExtents=<20 20 20>
      constraintSetup {angularLimitsMax=<0 8 5>, angularLimitsMin=<0 -8 -5>, angularTarget=<1 0 0>}
      localJointOffset=<4 0 0>
     }]
   }
   n22 = Anim Dynamics - Bone: dyn_l_side_hair_2a!(ComponentPose=@n21, Alpha=1, GravityOverride:Vector="0,0,-980", ExternalForce="0,0,0") -> (Pose) {
    angularDampingOverride=0.9999
    angularSpringConstant=5000
    bAngularSpring=true
    bOverrideAngularDamping=true
    bUseGravityOverride=true
    gravityScale=0.35
    simulationSpace=RootRelative
    boundBone {boneName=dyn_l_side_hair_2a}
    componentLinearAccScale=<1 1 1>
    componentLinearVelScale=<1 1 1>
    gravityOverride {z=-980}
    physicsBodyDefinitions [{
      boundBone {boneName=dyn_l_side_hair_2a}
      boxExtents=<20 20 20>
      constraintSetup {angularLimitsMax=<0 8 5>, angularLimitsMin=<0 -8 -5>, angularTarget=<1 0 0>}
      localJointOffset=<4 0 0>
     }]
   }
   n23 = Anim Dynamics - Bone: dyn_l_side_hair_2b!(ComponentPose=@n22, Alpha=1, GravityOverride:Vector="0,0,-980", ExternalForce="0,0,0") -> (Pose) {
    angularDampingOverride=0.9999
    angularSpringConstant=5000
    bAngularSpring=true
    bOverrideAngularDamping=true
    bUseGravityOverride=true
    gravityScale=0.35
    simulationSpace=RootRelative
    boundBone {boneName=dyn_l_side_hair_2b}
    componentLinearAccScale=<1 1 1>
    componentLinearVelScale=<1 1 1>
    gravityOverride {z=-980}
    physicsBodyDefinitions [{
      boundBone {boneName=dyn_l_side_hair_2b}
      boxExtents=<20 20 20>
      constraintSetup {angularLimitsMax=<0 8 5>, angularLimitsMin=<0 -8 -5>, angularTarget=<1 0 0>}
      localJointOffset=<4 0 0>
     }]
   }
   n24 = Anim Dynamics - Bone: dyn_l_side_hair_3a!(ComponentPose=@n23, Alpha=1, GravityOverride:Vector="0,0,-980", ExternalForce="0,0,0") -> (Pose) {
    angularDampingOverride=0.9999
    angularSpringConstant=5000
    bAngularSpring=true
    bOverrideAngularDamping=true
    bUseGravityOverride=true
    gravityScale=0.35
    simulationSpace=RootRelative
    boundBone {boneName=dyn_l_side_hair_3a}
    componentLinearAccScale=<1 1 1>
    componentLinearVelScale=<1 1 1>
    gravityOverride {z=-980}
    physicsBodyDefinitions [{
      boundBone {boneName=dyn_l_side_hair_3a}
      boxExtents=<20 20 20>
      constraintSetup {angularLimitsMax=<0 8 5>, angularLimitsMin=<0 -8 -5>, angularTarget=<1 0 0>}
      localJointOffset=<4 0 0>
     }]
   }
   n25 = Anim Dynamics - Bone: dyn_l_side_hair_3b!(ComponentPose=@n24, Alpha=1, GravityOverride:Vector="0,0,-980", ExternalForce="0,0,0") -> (Pose) {
    angularDampingOverride=0.9999
    angularSpringConstant=5000
    bAngularSpring=true
    bOverrideAngularDamping=true
    bUseGravityOverride=true
    gravityScale=0.35
    simulationSpace=RootRelative
    boundBone {boneName=dyn_l_side_hair_3b}
    componentLinearAccScale=<1 1 1>
    componentLinearVelScale=<1 1 1>
    gravityOverride {z=-980}
    physicsBodyDefinitions [{
      boundBone {boneName=dyn_l_side_hair_3b}
      boxExtents=<20 20 20>
      constraintSetup {angularLimitsMax=<0 8 5>, angularLimitsMin=<0 -8 -5>, angularTarget=<1 0 0>}
      localJointOffset=<4 0 0>
     }]
   }
   n26 = Anim Dynamics - Bone: dyn_r_side_hair_1a!(ComponentPose=@n25, Alpha=1, GravityOverride:Vector="0,0,-980", ExternalForce="0,0,0") -> (Pose) {
    angularDampingOverride=0.9999
    angularSpringConstant=5000
    bAngularSpring=true
    bOverrideAngularDamping=true
    bUseGravityOverride=true
    gravityScale=0.35
    simulationSpace=RootRelative
    boundBone {boneName=dyn_r_side_hair_1a}
    componentLinearAccScale=<1 1 1>
    componentLinearVelScale=<1 1 1>
    gravityOverride {z=-980}
    physicsBodyDefinitions [{
      boundBone {boneName=dyn_r_side_hair_1a}
      boxExtents=<20 20 20>
      constraintSetup {angularLimitsMax=<0 8 5>, angularLimitsMin=<0 -8 -5>, angularTarget=<1 0 0>}
      localJointOffset=<4 0 0>
     }]
   }
   n27 = Anim Dynamics - Bone: dyn_r_side_hair_1b!(ComponentPose=@n26, Alpha=1, GravityOverride:Vector="0,0,-980", ExternalForce="0,0,0") -> (Pose) {
    angularDampingOverride=0.9999
    angularSpringConstant=5000
    bAngularSpring=true
    bOverrideAngularDamping=true
    bUseGravityOverride=true
    gravityScale=0.35
    simulationSpace=RootRelative
    boundBone {boneName=dyn_r_side_hair_1b}
    componentLinearAccScale=<1 1 1>
    componentLinearVelScale=<1 1 1>
    gravityOverride {z=-980}
    physicsBodyDefinitions [{
      boundBone {boneName=dyn_r_side_hair_1b}
      boxExtents=<20 20 20>
      constraintSetup {angularLimitsMax=<0 8 5>, angularLimitsMin=<0 -8 -5>, angularTarget=<1 0 0>}
      localJointOffset=<4 0 0>
     }]
   }
   n28 = Anim Dynamics - Bone: dyn_r_side_hair_2a!(ComponentPose=@n27, Alpha=1, GravityOverride:Vector="0,0,-980", ExternalForce="0,0,0") -> (Pose) {
    angularDampingOverride=0.9999
    angularSpringConstant=5000
    bAngularSpring=true
    bOverrideAngularDamping=true
    bUseGravityOverride=true
    gravityScale=0.35
    simulationSpace=RootRelative
    boundBone {boneName=dyn_r_side_hair_2a}
    componentLinearAccScale=<1 1 1>
    componentLinearVelScale=<1 1 1>
    gravityOverride {z=-980}
    physicsBodyDefinitions [{
      boundBone {boneName=dyn_r_side_hair_2a}
      boxExtents=<20 20 20>
      constraintSetup {angularLimitsMax=<0 8 5>, angularLimitsMin=<0 -8 -5>, angularTarget=<1 0 0>}
      localJointOffset=<4 0 0>
     }]
   }
   n29 = Anim Dynamics - Bone: dyn_r_side_hair_2b!(ComponentPose=@n28, Alpha=1, GravityOverride:Vector="0,0,-980", ExternalForce="0,0,0") -> (Pose) {
    angularDampingOverride=0.9999
    angularSpringConstant=5000
    bAngularSpring=true
    bOverrideAngularDamping=true
    bUseGravityOverride=true
    gravityScale=0.35
    simulationSpace=RootRelative
    boundBone {boneName=dyn_r_side_hair_2b}
    componentLinearAccScale=<1 1 1>
    componentLinearVelScale=<1 1 1>
    gravityOverride {z=-980}
    physicsBodyDefinitions [{
      boundBone {boneName=dyn_r_side_hair_2b}
      boxExtents=<20 20 20>
      constraintSetup {angularLimitsMax=<0 8 5>, angularLimitsMin=<0 -8 -5>, angularTarget=<1 0 0>}
      localJointOffset=<4 0 0>
     }]
   }
   n30 = Anim Dynamics - Bone: dyn_r_side_hair_3a!(ComponentPose=@n29, Alpha=1, GravityOverride:Vector="0,0,-980", ExternalForce="0,0,0") -> (Pose) {
    angularDampingOverride=0.9999
    angularSpringConstant=5000
    bAngularSpring=true
    bOverrideAngularDamping=true
    bUseGravityOverride=true
    gravityScale=0.35
    simulationSpace=RootRelative
    boundBone {boneName=dyn_r_side_hair_3a}
    componentLinearAccScale=<1 1 1>
    componentLinearVelScale=<1 1 1>
    gravityOverride {z=-980}
    physicsBodyDefinitions [{
      boundBone {boneName=dyn_r_side_hair_3a}
      boxExtents=<20 20 20>
      constraintSetup {angularLimitsMax=<0 8 5>, angularLimitsMin=<0 -8 -5>, angularTarget=<1 0 0>}
      localJointOffset=<4 0 0>
     }]
   }
   n31 = Anim Dynamics - Bone: dyn_r_side_hair_3b!(ComponentPose=@n30, Alpha=1, GravityOverride:Vector="0,0,-980", ExternalForce="0,0,0") -> (Pose) {
    angularDampingOverride=0.9999
    angularSpringConstant=5000
    bAngularSpring=true
    bOverrideAngularDamping=true
    bUseGravityOverride=true
    gravityScale=0.35
    simulationSpace=RootRelative
    boundBone {boneName=dyn_r_side_hair_3b}
    componentLinearAccScale=<1 1 1>
    componentLinearVelScale=<1 1 1>
    gravityOverride {z=-980}
    physicsBodyDefinitions [{
      boundBone {boneName=dyn_r_side_hair_3b}
      boxExtents=<20 20 20>
      constraintSetup {angularLimitsMax=<0 8 5>, angularLimitsMin=<0 -8 -5>, angularTarget=<1 0 0>}
      localJointOffset=<4 0 0>
     }]
   }
   Output Pose(Result=Component To Local(ComponentPose=@n31)) {layerGroup=None}
   Get isCrouching_Walk_aim() -> (isCrouching_Walk_aim:bool)
   Get bIsSkyDiving()
  ]}
 EventGraph {nodes [
   n1 = Event BlueprintUpdateAnimation()
   > Set Delta Time X(Delta Time X=@n1.DeltaTimeX)
   > n2 = Sequence!(execute) -> (then_0, then_1, then_2:Exec, then_3:Exec)
   > n3 = "Gravity Override State Function (Parameter Asset) (Nativized)"(
    CharacterPartAnimInstance=Self-Reference()
    GameTimeInSeconds=GetGameTimeInSeconds()
    Parameters=&/M_MED_ZenCrash_FaceAcc_Parameters
   )
   > Set Gravity Override(Gravity Override=@n3)
   > Set RBAN Settings Out(RBAN Settings Out=@n3.OutRigidBodyAnimNodeParameters)
   Set CrouchWalk(execute=@n2.then_1, CrouchWalk:bool=AND Boolean(A=Get bIsCrouching(), B=Get bIsMoving2D())) -> (Output_Get:bool)
   n4 = GetPartSkeletalMeshComponent()
   n5 = GetAlphaFromReferenceTransformDelta(
    SkeletalMeshComponent=@n4
    BoneName=head
    BaseName=spine_05
    DeltaComponentType=EulerY
    InputRange=(X=-15,Y=75)
    OutputRange=(X=1,Y=0)
    TwistAxis="1,0,0"
   )
   n6 = NOT Boolean(A=Get bIsPlayingMeleeAnim())
   > Set Head look down alpha(Head look down alpha:double=SelectFloat(A=@n5, B=0, bPickA=@n6)) -> (Output_Get:double)
   n7 = float + float(
    A=GetAlphaFromReferenceTransformDelta(SkeletalMeshComponent=@n4, BoneName=head, BaseName=spine_05, InputRange=(X=-45,Y=-50), TwistAxis="1,0,0")
    B=GetAlphaFromReferenceTransformDelta(SkeletalMeshComponent=@n4, BoneName=head, BaseName=spine_05, DeltaComponentType=EulerY, InputRange=(X=20,Y=30), TwistAxis="1,0,0")
   )
   > Set Head look up alpha(Head look up alpha:double=SelectFloat(A="Clamp (Float)"(Value=@n7, Min=0, Max=1), B=0, bPickA=@n6)) -> (Output_Get:double)
   n8 = GetOwnerSkeletalMeshComponent()
   > Set Head_Right(Head_Right:double=GetAlphaFromReferenceTransformDelta(SkeletalMeshComponent=@n8, BoneName=head, BaseName=spine_05, DeltaComponentType=EulerX, InputRange=(X=-20,Y=-50), TwistAxis="1,0,0")) -> (Output_Get:double)
   > Set Head_Left(Head_Left:double=GetAlphaFromReferenceTransformDelta(SkeletalMeshComponent=@n8, BoneName=head, BaseName=spine_05, DeltaComponentType=EulerX, InputRange=(X=12,Y=68), TwistAxis="1,0,0")) -> (Output_Get:double)
   > Set Head look up _low_alpha(Head look up _low_alpha:double=GetAlphaFromReferenceTransformDelta(SkeletalMeshComponent=@n8, BoneName=head, BaseName=spine_05, InputRange=(X=-40,Y=-50), TwistAxis="1,0,0")) -> (Output_Get:double)
   > Set Crouch_look up _low_alpha(Crouch_look up _low_alpha:double=GetAlphaFromReferenceTransformDelta(SkeletalMeshComponent=@n8, BoneName=head, BaseName=spine_05, InputRange=(X=1,Y=-30), TwistAxis="1,0,0")) -> (Output_Get:double)
   n9 = NOT Boolean(A=Get bIsCrouching())
   n10 = Get bIsTargeting()
   n11 = Get Head look up alpha() -> (Head look up alpha:double)
   Set Head look up aim(execute=@n2.then_2, Head look up aim:double=SelectFloat(A=@n11, B=0, bPickA=AND Boolean(A=@n10, B=@n9))) -> (Output_Get:double)
   n12 = Get bIsMoving2D()
   n13 = Get bIsTargeting()
   n14 = Get bIsCrouching()
   > Set CrouchWalk_Aim(CrouchWalk_Aim:double=SelectFloat(A=@n11, B=0, bPickA=AND Boolean!(A=@n14, B=@n12, C:bool=@n13) -> (ReturnValue))) -> (Output_Get:double)
   n15 = NOT Boolean(A=Get bIsTargeting())
   n16 = NOT Boolean(A=Get bIsMoving2D())
   n17 = Get bIsCrouching()
   > Set CrouchAlpha(CrouchAlpha:double=SelectFloat(A=@n11, B=0, bPickA=AND Boolean(A=@n17, B=@n16, C=@n15))) -> (Output_Get:double)
   n18 = NOT Boolean(A=Get bIsMoving2D())
   n19 = Get bIsCrouching()
   n20 = Get bIsTargeting()
   > Set CrouchAim(CrouchAim:double=SelectFloat(A=Get Head look up _low_alpha() -> (Head look up _low_alpha:double), B=0, bPickA=AND Boolean(A=@n19, B=@n18, C=@n20))) -> (Output_Get:double)
   n21 = SelectFloat(A=Get Head look up alpha(), B=0, bPickA=AND Boolean(A=Get bIsCrouching(), B=Get bIsTargeting(), C=Get bIsMoving2D()))
   > Set CrouchAlpha_Extreme_neg(CrouchAlpha_Extreme_neg:double=@n21) -> (Output_Get:double)
   > n22 = Set CrouchAlpha_Extreme(CrouchAlpha_Extreme:double=@n21) -> (Output_Get:double)
   n23 = AND Boolean(A=Get bIsCrouching(), B=Get bIsTargeting(), C=NOT Boolean(A=Get bIsMoving2D()))
   n24 = Get Head look up alpha()
   > Set isCrouching_Walk_aim(isCrouching_Walk_aim:bool="Equal (Float)"(A=SelectFloat(A=@n24, B=0, bPickA=@n23), B=1)) -> (Output_Get:bool)
   > Set CrouchAlpha_Extreme_bool(CrouchAlpha_Extreme_bool:bool="Equal (Float)"(A=@n22, B=1)) -> (Output_Get:bool)
   n25 = NOT Boolean(A=Get bIsCrouching())
   n26 = NOT Boolean(A=Get bIsTargeting())
   > Set HeadUp_Only(HeadUp_Only:double=SelectFloat(A=Get Head look up alpha(), B=0, bPickA=AND Boolean!(A=@n25, B=@n26) -> (ReturnValue))) -> (Output_Get:double)
   n27 = Get Head look up alpha()
   > Set Emote_Left(Emote_Left:double=SelectFloat(A=Get Head_Left() -> (Head_Left:double), B=@n27, bPickA=Get bIsPlayingEmote())) -> (Output_Get:double)
   > Set Emote_Right(Emote_Right:double=SelectFloat(A=Get Head_Right() -> (Head_Right:double), B=@n27, bPickA=Get bIsPlayingEmote())) -> (Output_Get:double)
   n28 = "Not Equal (Float)"(A=Get CrouchAlpha_Extreme_neg() -> (CrouchAlpha_Extreme_neg:double))
   n29 = Get bIsTargeting()
   n30 = Get bIsCrouching()
   > Set CrouchAlpha_Extreme_Inplace(CrouchAlpha_Extreme_Inplace:double=SelectFloat(A=Get Head look up alpha(), B="To Float (Boolean)"(InBool=@n28), bPickA=AND Boolean(A=@n30, B=@n29))) -> (Output_Get:double)
   n31 = NOT Boolean(A=Get bIsTargeting())
   n32 = Get bIsCrouching()
   > Set Couch_HeadUp_Only(Couch_HeadUp_Only:double=SelectFloat(A=Get Crouch_look up _low_alpha() -> (Crouch_look up _low_alpha:double), B=0, bPickA=AND Boolean(A=@n32, B=@n31))) -> (Output_Get:double)
   n33 = NOT Boolean(A=Get bIsSkyDiving())
   n34 = Get bIsInAir() -> (bIsInAir:bool)
   Set On Jumping(execute=@n2.then_3, On Jumping:double="To Float (Boolean)"(InBool=AND Boolean(A=@n34, B=@n33))) -> (Output_Get:double)
   n35 = NOT Boolean(A=Get bIsCrouching())
   n36 = Get bIsTargeting()
   > Set HeadUp_Tragetting_Alpha(HeadUp_Tragetting_Alpha:double="To Float (Boolean)"(InBool=AND Boolean(A=@n35, B=@n36))) -> (Output_Get:double)
   Event OnWeaponChanged()
   > Set isUsingPistol(isUsingPistol:double=0) -> (Output_Get:double)
   > n37 = Switch on EFortWeaponCoreAnimation!(execute, Selection:EFortWeaponCoreAnimation Enum=Get WeaponCoreAnim!() -> (WeaponCoreAnim)) -> (
    Melee:Exec
    Pistol:Exec
    Shotgun:Exec
    PaperBlueprint:Exec
    Rifle:Exec
    MeleeOneHand:Exec
    MachinePistol:Exec
    RocketLauncher:Exec
    GrenadeLauncher:Exec
    GoingCommando:Exec
    AssaultRifle:Exec
    TacticalShotgun:Exec
    SniperRifle:Exec
    TrapPlacement:Exec
    ShoulderLauncher:Exec
    AbilityDecoTool:Exec
    Crossbow:Exec
    C4:Exec
    RemoteControl:Exec
    DualWield:Exec
    AR_BullPup:Exec
    AR_ForwardGrip:Exec
    MedPackPaddles:Exec
    SMG_P90:Exec
    AR_DrumGun:Exec
    Consumable_Small:Exec
    Consumable_Large:Exec
    Balloon:Exec
    MountedTurret:Exec
    CreativeTool:Exec
    ExplosiveBow:Exec
    AshtonIndigo:Exec
    AshtonChicago:Exec
    MeleeDualWield:Exec
    Unarmed:Exec
   )
   Set isUsingPistol(execute=@n37.Pistol, isUsingPistol=1)
   Get bIsMoving2D()
   Set Machine gun corr(Machine gun corr:double=SelectFloat(A=@n5, B=0, bPickA=false)) -> (Output_Get:double)
  ]}
]
variables [
 Couch_HeadUp_Only {type=number}
 Crouch_look up _low_alpha {type=number}
 CrouchAim {type=number}
 CrouchAlpha {type=number}
 CrouchAlpha_Extreme {type=number}
 CrouchAlpha_Extreme_bool {type=boolean}
 CrouchAlpha_Extreme_Inplace {type=number}
 CrouchAlpha_Extreme_neg {type=number}
 CrouchAlpha_Extreme_Rig {type=number}
 CrouchWalk {type=boolean}
 CrouchWalk_Aim {type=number}
 Delta Time X {type=number}
 Emote_Left {type=number}
 Emote_Right {type=number}
 Gravity Override {type=Vector}
 Head look down alpha {type=number}
 Head look up _low_alpha {type=number}
 Head look up aim {type=number}
 Head look up alpha {type=number}
 Head_Left {type=number}
 Head_Right {type=number}
 HeadUp_Only {type=number}
 HeadUp_Tragetting_Alpha {type=number}
 isCrouching float {type=number}
 isCrouching_Walk_aim {type=boolean}
 isSkyDivingFloat {type=number}
 isTargeting float {type=number}
 isUsingPistol {type=number}
 Machine gun corr {type=number}
 On Jumping {type=number}
 Pub Pawn Delta Rotation {type=Rotator}
 RBAN Settings Out {type=FortRigidBodyAnimNodeParameters}
]
