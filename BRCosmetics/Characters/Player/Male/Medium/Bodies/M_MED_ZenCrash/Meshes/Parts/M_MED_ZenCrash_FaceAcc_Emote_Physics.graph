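# Physics asset for the ZenCrash male medium character's emote face-accessory mesh, covering a hair tail and side hair strands.
# Capsule bodies cover upper spine, head, three side strands per side, and tail base and dynamic segment, which has two capsules.
# Side strands and tail base constrain to the head with zero-degree limits.
# The dynamic tail gets narrow, asymmetric swing and zero twist off its base, while head-to-spine rotation is free.
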PhysicsAsset /BRCosmetics/Characters/Player/Male/Medium/Bodies/M_MED_ZenCrash/Meshes/Parts/M_MED_ZenCrash_FaceAcc_Emote_Physics
bodies {
 base_hair_tail_01 [{shapeType=Capsule, length=0.5, radius=0.5, shapeName=base_hair_tail_01_capsule}]
 base_l_side_hair_1a [{
   shapeType=Capsule
   length=3.755
   radius=0.63
   shapeName=base_l_side_hair_1a_capsule
   center=<1.3373 0.1417 -0.1082>
   rotation=<rot 0 -84.999 99.9997>
  }]
 base_l_side_hair_2a [{
   shapeType=Capsule
   length=3.505
   radius=0.88
   shapeName=base_l_side_hair_1a_capsule
   center=<1.0429 0.1459 0.1945>
   rotation=<rot 0.8672 -89.923 94.9624>
  }]
 base_l_side_hair_3a [{
   shapeType=Capsule
   length=3.505
   radius=0.88
   shapeName=base_l_side_hair_1a_capsule
   center=<0.8395 -0.1971 0.3936>
   rotation=<rot -1.7469 -104.6947 80.0765>
  }]
 base_r_side_hair_1a [{
   shapeType=Capsule
   length=3.005
   radius=0.63
   shapeName=base_l_side_hair_1a_capsule
   center=<1.5238 -0.0076 0.0999>
   rotation=<rot -8.3949 -85.7471 93.3836>
  }]
 base_r_side_hair_2a [{
   shapeType=Capsule
   length=3.63
   radius=0.755
   shapeName=base_l_side_hair_1a_capsule
   center=<1.2196 0.0606 0.2023>
   rotation=<rot -0.8799 -99.7692 79.9635>
  }]
 base_r_side_hair_3a [{
   shapeType=Capsule
   length=3.505
   radius=0.88
   shapeName=base_l_side_hair_1a_capsule
   center=<2.0003 0.1523 -0.1295>
   rotation=<rot 1.7344 -84.9979 95.0742>
  }]
 dyn_hair_tail_01 [
  {
   shapeType=Capsule
   length=4.7815
   radius=2.7982
   shapeName=dyn_hair_tail_01_capsule
   center=<0.0987 -0.2008 -1.1387>
   rotation=<rot 0.39 -2.2153 -80.0076>
  }
  {
   shapeType=Capsule
   length=3.0393
   radius=4.9237
   shapeName=dyn_hair_tail_01_capsule
   center=<6.1476 -0.2671 -1.9153>
   rotation=<rot 0.2186 -4.9953 -84.0944>
  }
 ]
 head [{shapeType=Capsule, length=8.875, radius=6.2374, center=<4.4641 2.3257 0.7296>, rotation=<rot -5.5004 -141.5645 -86.8129>}]
 spine_05 [{shapeType=Capsule, length=39.9032, radius=11.7345, shapeName=spine_05_capsule, center=<4.7917 1.4175 0>, rotation=<rot 0 -80.0005 -180>}]
}
constraints [
 {bone1Name=base_hair_tail_01, bone2Name=head, swing1LimitDegrees=0, swing2LimitDegrees=0, twistLimitDegrees=0}
 {
  bone1Name=head
  bone2Name=spine_05
  swing1LimitDegrees=45
  swing1Motion=Free
  swing2LimitDegrees=45
  swing2Motion=Free
  twistLimitDegrees=45
  twistMotion=Free
 }
 {bone1Name=base_l_side_hair_1a, bone2Name=head, swing1LimitDegrees=0, swing2LimitDegrees=0, twistLimitDegrees=0}
 {bone1Name=base_l_side_hair_2a, bone2Name=head, swing1LimitDegrees=0, swing2LimitDegrees=0, twistLimitDegrees=0}
 {bone1Name=base_l_side_hair_3a, bone2Name=head, swing1LimitDegrees=0, swing2LimitDegrees=0, twistLimitDegrees=0}
 {bone1Name=base_r_side_hair_1a, bone2Name=head, swing1LimitDegrees=0, swing2LimitDegrees=0, twistLimitDegrees=0}
 {bone1Name=base_r_side_hair_2a, bone2Name=head, swing1LimitDegrees=0, swing2LimitDegrees=0, twistLimitDegrees=0}
 {bone1Name=base_r_side_hair_3a, bone2Name=head, swing1LimitDegrees=0, swing2LimitDegrees=0, twistLimitDegrees=0}
 {bone1Name=dyn_hair_tail_01, bone2Name=base_hair_tail_01, swing1LimitDegrees=5, swing2LimitDegrees=25, twistLimitDegrees=0}
]
