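# Control Rig for the ZenCrash face accessory mesh that offsets the hair tail bone as the head rotates during crouching and skydiving.
# A parent constraint ties the tail base's rotation and scale to neck_02 every update, repeated in the skydive branch.
# A reusable DeltaRotation function turns head-to-neck swing into a weight for ModifyTransforms on dyn_hair_tail_01.
# Crouch scales that weight by a pelvis-height curve, while skydive uses it directly with a larger offset.
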ControlRigBlueprint /BRCosmetics/Characters/Player/Male/Medium/Bodies/M_MED_ZenCrash/Meshes/Parts/M_MED_ZenCrash_FaceAcc_CtrlRig1
graphs [
 New Function_ContainedGraph {nodes [
   n1 = RigVMFunctionEntryNode()
   RigVMFunction_MathBoolToInteger = &RigVMFunction_MathBoolToInteger(Value=@n1.onlyTwist)
   RigVMFunction_MathQuaternionSwingTwist = &RigVMFunction_MathQuaternionSwingTwist(
    TwistAxis=@n1.twistAxis
    Input=&RigUnit_GetRelativeTransformForItem(bChildInitial=False, bParentInitial=False, Child=@n1.child, Parent=@n1.parent).RelativeTransform.Rotation
   )
   n2 = RigVMDispatchNode(
    Values=((),())
    Index=@RigVMFunction_MathBoolToInteger.Result
    Values.0=@RigVMFunction_MathQuaternionSwingTwist.Swing
    Values.1=@RigVMFunction_MathQuaternionSwingTwist.Twist
   ) {template="DISPATCH_RigVMDispatch_SelectInt32(in Index,in Values,out Result)"}
   RigVMFunction_MathQuaternionToEuler = &RigVMFunction_MathQuaternionToEuler(Value=@n2.Result, RotationOrder=@n1.Rotation Order)
   > DISPATCH_RigVMDispatch_Print = RigVMDispatchNode(ScreenColor=(R=1,G=1,B=1,A=1), ScreenDuration=1.5, Enabled=@n1.debugRotation, Value=@RigVMFunction_MathQuaternionToEuler.Result) {template="DISPATCH_RigVMDispatch_Print(in Prefix,in Value,in Enabled,in ScreenDuration,in ScreenColor)"}
   Cast = RigVMDispatchNode(Value=@n1.triggerAxis) {template="DISPATCH_RigVMDispatch_CastEnumToInt(in Value,out Result)"}
   DISPATCH_RigVMDispatch_SelectInt32 = RigVMDispatchNode(
    Values=(0,0,0)
    Index=@Cast.Result
    Values.0=@RigVMFunction_MathQuaternionToEuler.Result.X
    Values.1=@RigVMFunction_MathQuaternionToEuler.Result.Y
    Values.2=@RigVMFunction_MathQuaternionToEuler.Result.Z
   ) {template="DISPATCH_RigVMDispatch_SelectInt32(in Index,in Values,out Result)"}
   Remap = &RigVMFunction_MathDoubleRemap(
    bClamp=False
    Value=@DISPATCH_RigVMDispatch_SelectInt32.Result
    TargetMinimum=@n1.outputMinMax.X
    TargetMaximum=@n1.outputMinMax.Y
    SourceMinimum=@n1.rotationMinMax.X
    SourceMaximum=@n1.rotationMinMax.Y
   )
   Absolute = &RigVMFunction_MathDoubleAbs(Value=@Remap.Result)
   Multiply = &RigVMFunction_MathDoubleMul(
    A=&RigVMFunction_MathDoubleClamp(Value=RigVMDispatchNode(True=@Absolute.Result, Condition=@n1.absoluteValue, False=@Remap.Result) {template="DISPATCH_RigVMDispatch_If(in Condition,in True,in False,out Result)"}.Result).Result
    B=@n1.premult
   )
   > RigVMFunctionReturnNode(outputQuat=@n2.Result, outputWeight=@Multiply.Result)
  ]}
 RigVMFunctionLibrary {nodes [DeltaRotation = RigVMCollapseNode(
    ExecuteContext=()
    Rotation Order=ZYX
    child=(Type=Bone,Name="None")
    outputMinMax=(X=0,Y=1)
    parent=(Type=Bone,Name="None")
    premult=1
    rotationMinMax=(X=0,Y=0)
    twistAxis=(X=1,Y=0,Z=0)
   )]}
 RigVMModel {nodes [
   RigUnit_BeginExecution = &RigUnit_BeginExecution()
   RigVMFunction_Sequence = &RigVMFunction_Sequence(ExecuteContext=@RigUnit_BeginExecution.ExecutePin)
   n1 = &RigUnit_ParentConstraint(
    AdvancedSettings=(InterpolationType=Average,RotationOrderForFilter=XZY)
    Child=(Type=Bone,Name="base_hair_tail_01")
    Filter=(TranslationFilter=(bX=false,bY=false,bZ=false),RotationFilter=(bX=True,bY=True,bZ=True),ScaleFilter=(bX=True,bY=True,bZ=True))
    Parents=((Item=(Type=Bone,Name="neck_02"),Weight=1))
    Weight=1
    ExecutePin=@RigVMFunction_Sequence.A
   )
   RigVMFunction_ControlFlowBranch = &RigVMFunction_ControlFlowBranch(ExecuteContext=@n1.ExecutePin, Condition=RigVMVariableNode() {variable=IsCrouching}.Value)
   DeltaRotation_1 = RigVMFunctionReferenceNode(
    Rotation Order=ZYX
    child=(Type=Bone,Name="head")
    outputMinMax=(X=0,Y=1)
    parent=(Type=Bone,Name="neck_02")
    premult=1
    rotationMinMax=(X=-40,Y=-90)
    triggerAxis=Z
    twistAxis=(X=1,Y=0,Z=0)
    ExecuteContext=@RigVMFunction_ControlFlowBranch.True
   )
   > RigVMVariableNode(Value=@DeltaRotation_1.outputWeight) {variable=HeadRot}
   n2 = &RigVMFunction_AnimEvalRichCurve(
    Curve=(EditorCurveData=(Keys=((),(Time=0.5,Value=1),(Time=0.8))))
    SourceMaximum=2
    SourceMinimum=0
    TargetMinimum=0
    Value=&RigVMFunction_MathDoubleRemap(
     SourceMaximum=75
     SourceMinimum=45
     TargetMaximum=1
     TargetMinimum=0
     bClamp=false
     Value=&RigUnit_GetRelativeTransformForItem(Child=(Type=Bone,Name="pelvis"), Parent=(Type=Bone,Name="root"), bChildInitial=False, bParentInitial=False).RelativeTransform.Translation.Z
    ).Result
   )
   Multiply = &RigVMFunction_MathDoubleMul(A=@n2.Result, B=RigVMVariableNode() {variable=HeadRot}.Value)
   AlphaInterp = &RigVMFunction_AlphaInterp(InterpSpeedDecreasing=10, InterpSpeedIncreasing=10, Scale=1, bClampResult=False, bInterpResult=False, Value=@Multiply.Result)
   > &RigUnit_ModifyTransforms(
    ItemToModify=((Item=(Type=Bone,Name="dyn_hair_tail_01"),Transform=(Rotation=(X=0,Y=-0.2588,Z=0,W=0.9659),Translation=(X=2,Y=0,Z=3),Scale3D=(X=1,Y=1,Z=1))))
    WeightMinimum=0
    Weight=@AlphaInterp.Result
   )
   n3 = &RigVMFunction_ControlFlowBranch(ExecuteContext=@RigVMFunction_Sequence.B, Condition=RigVMVariableNode() {variable=Skydive}.Value)
   &RigUnit_ParentConstraint(
    AdvancedSettings=(InterpolationType=Average,RotationOrderForFilter=XZY)
    Child=(Type=Bone,Name="base_hair_tail_01")
    Filter=(TranslationFilter=(bX=false,bY=false,bZ=false),RotationFilter=(bX=True,bY=True,bZ=True),ScaleFilter=(bX=True,bY=True,bZ=True))
    Parents=((Item=(Type=Bone,Name="neck_02"),Weight=1))
    Weight=1
    ExecutePin=@n3.True
   )
   > DeltaRotation = RigVMFunctionReferenceNode(
    Rotation Order=ZYX
    child=(Type=Bone,Name="head")
    outputMinMax=(X=0,Y=1)
    parent=(Type=Bone,Name="neck_02")
    premult=1
    rotationMinMax=(X=-30,Y=-90)
    triggerAxis=Z
    twistAxis=(X=1,Y=0,Z=0)
   )
   n4 = &RigVMFunction_AlphaInterp(InterpSpeedDecreasing=10, InterpSpeedIncreasing=10, Scale=1, bClampResult=False, bInterpResult=False, Value=@DeltaRotation.outputWeight)
   > &RigUnit_ModifyTransforms(
    ItemToModify=((Item=(Type=Bone,Name="dyn_hair_tail_01"),Transform=(Rotation=(X=-0.1576,Y=-0.588,Z=-0.2053,W=0.7663),Translation=(X=2,Y=0,Z=4),Scale3D=(X=1,Y=1,Z=1))))
    WeightMinimum=0
    Weight=@n4.Result
   )
  ]}
]
hierarchy [
 root {loc=<0 0 0>, rot=<0 0 0 1>}
 pelvis {parent=root, loc=<0 -0.7071 93.1191>, rot=<0.0127 0.707 -0.0127 -0.707>}
 spine_01 {parent=pelvis, loc=<6.7494 -0.6388 0>, rot=<0 0 0.034 -0.9994>}
 spine_02 {parent=spine_01, loc=<7.2518 -0.035 0>, rot=<0 0 0.0238 -0.9997>}
 spine_03 {parent=spine_02, loc=<6.0104 0 0>, rot=<0 0 0.0149 -0.9999>}
 spine_04 {parent=spine_03, loc=<8.1024 0 0>, rot=<0 0 -0.0132 -0.9999>}
 spine_05 {parent=spine_04, loc=<10.0146 0 0>, rot=<0 0 -0.0015 -1>}
 neck_01 {parent=spine_05, loc=<11.3138 0.1529 0>, rot=<0 0 -0.2003 -0.9797>}
 neck_02 {parent=neck_01, loc=<4.5627 0 0>, rot=<0 0 0.0504 -0.9987>}
 head {parent=neck_02, loc=<5.3914 0 0>, rot=<0 0 0.115 -0.9934>}
 base_hair_tail_01 {parent=head, loc=<1.9479 -8.1627 0.3331>, rot=<0.3217 -0.5827 0.6281 -0.4031>}
 dyn_hair_tail_01 {parent=base_hair_tail_01, loc=<0 0 0>, rot=<0 0 0 1>}
 dyn_hair_tail_lwr_01 {parent=dyn_hair_tail_01, loc=<4.1296 -0.4916 -2.301>, rot=<0.1037 -0.1559 0.0165 -0.9822>}
 dyn_hair_tail_uppr_01 {parent=dyn_hair_tail_01, loc=<3.2558 -0.311 1.6972>, rot=<0.1018 0.2436 -0.0257 -0.9642>}
 dyn_hair_tail_mid_01 {parent=dyn_hair_tail_01, loc=<3.9957 -0.0526 -0.2462>, rot=<0.105 0 0 -0.9945>}
 base_l_upper_hair_01 {parent=head, loc=<5.4899 -7.7417 -5.976>, rot=<0.393 -0.4225 0.3106 -0.7553>}
 dyn_l_upper_hair_01 {parent=base_l_upper_hair_01, loc=<0 0 0>, rot=<0 0 0 1>}
 base_l_upper_hair_02 {parent=head, loc=<4.7097 -8.3265 -3.1027>, rot=<0.464 -0.5296 0.4092 -0.5803>}
 dyn_l_upper_hair_02 {parent=base_l_upper_hair_02, loc=<0 0 0>, rot=<0 0 0 1>}
 base_l_upper_hair_03 {parent=head, loc=<4.9982 -3.4697 -8.0595>, rot=<0.3276 -0.5571 0.3791 -0.6623>}
 dyn_l_upper_hair_03 {parent=base_l_upper_hair_03, loc=<0 0 0>, rot=<0 0 0 1>}
 base_l_upper_hair_04 {parent=head, loc=<2.2911 -4.5567 -7.663>, rot=<0.3243 -0.8782 0.2174 -0.2762>}
 dyn_l_upper_hair_04 {parent=base_l_upper_hair_04, loc=<0 0 0>, rot=<0 0 0 1>}
 base_r_upper_hair_01 {parent=head, loc=<5.2989 -7.224 5.612>, rot=<0.7799 -0.4551 0.3536 -0.2442>}
 dyn_r_upper_hair_01 {parent=base_r_upper_hair_01, loc=<0 0 0>, rot=<0 0 0 1>}
 base_r_upper_hair_02 {parent=head, loc=<5.582 -5.7347 7.0515>, rot=<0.6451 -0.236 0.5396 -0.4868>}
 dyn_r_upper_hair_02 {parent=base_r_upper_hair_02, loc=<0 0 0>, rot=<0 0 0 1>}
 base_r_upper_hair_03 {parent=head, loc=<4.9701 -3.0133 7.6834>, rot=<0.3995 -0.0795 0.8337 -0.373>}
 dyn_r_upper_hair_03 {parent=base_r_upper_hair_03, loc=<0 0 0>, rot=<0 0 0 1>}
 base_l_side_hair_3a {parent=head, loc=<0.6421 2.7489 -6.7146>, rot=<-0.1901 -0.9724 0.027 0.1322>}
 dyn_l_side_hair_3a {parent=base_l_side_hair_3a, loc=<0 0 0>, rot=<0 0 0 1>}
 dyn_l_side_hair_3b {parent=dyn_l_side_hair_3a, loc=<1.854 0 0>, rot=<0 0 0.0825 -0.9966>}
 base_l_side_hair_2a {parent=head, loc=<1.678 4.6969 -6.6905>, rot=<-0.2696 -0.9615 0.0526 -0.0086>}
 dyn_l_side_hair_2a {parent=base_l_side_hair_2a, loc=<0 0 0>, rot=<0 0 0 1>}
 dyn_l_side_hair_2b {parent=dyn_l_side_hair_2a, loc=<1.854 0 0>, rot=<0 0 0.0825 -0.9966>}
 base_l_side_hair_1a {parent=head, loc=<4.2404 5.7462 -6.558>, rot=<-0.3296 -0.9303 -0.1519 -0.0534>}
 dyn_l_side_hair_1a {parent=base_l_side_hair_1a, loc=<0 0 0>, rot=<0 0 0 1>}
 dyn_l_side_hair_1b {parent=dyn_l_side_hair_1a, loc=<1.9832 -0.0349 -0.4515>, rot=<-0.0545 -0.2507 -0.0472 -0.9654>}
 base_r_side_hair_1a {parent=head, loc=<4.6215 7.7758 6.9122>, rot=<0.1565 -0.0131 0.9834 -0.0908>}
 dyn_r_side_hair_1a {parent=base_r_side_hair_1a, loc=<0 0 0>, rot=<0 0 0 1>}
 dyn_r_side_hair_1b {parent=dyn_r_side_hair_1a, loc=<1.9667 0 0>, rot=<0 0 -0.0551 -0.9985>}
 base_r_side_hair_2a {parent=head, loc=<2.2128 5.4974 6.7353>, rot=<-0.1091 0.2216 0.9554 0.1617>}
 dyn_r_side_hair_2a {parent=base_r_side_hair_2a, loc=<0 0 0>, rot=<0 0 0 1>}
 dyn_r_side_hair_2b {parent=dyn_r_side_hair_2a, loc=<2.2694 0.0031 0.0185>, rot=<-0.011 0.2485 0.0737 -0.9658>}
 base_r_side_hair_3a {parent=head, loc=<1.5168 3.4346 7.1779>, rot=<-0.0331 0.0792 0.9781 0.1895>}
 dyn_r_side_hair_3a {parent=base_r_side_hair_3a, loc=<0 0 0>, rot=<0 0 0 1>}
 dyn_r_side_hair_3b {parent=dyn_r_side_hair_3a, loc=<2.4228 0.0392 -0.1019>, rot=<0.0135 -0.129 -0.0159 -0.9914>}
 base_m_upper_hair_01 {parent=head, loc=<5.266 -8.7525 0.1821>, rot=<0.6946 -0.2551 0.5177 -0.4295>}
 dyn_m_upper_hair_01 {parent=base_m_upper_hair_01, loc=<0 0 0>, rot=<0 0 0 1>}
 head_null {parent=head, type=Null, loc=<0 0 0>, rot=<0 0 0 1>}
 clavicle_l {parent=spine_05, loc=<4.3575 -0.1466 -2.3223>, rot=<0.0465 -0.7289 -0.009 -0.6829>}
 upperarm_l {parent=clavicle_l, loc=<15.5111 0 0>, rot=<0.0226 -0.3234 -0.0088 -0.9459>}
 lowerarm_l {parent=upperarm_l, loc=<26.6514 0 0>, rot=<0.0391 -0.0099 -0.2453 -0.9686>}
 upperarm_twist_01_l {parent=upperarm_l, loc=<0.0041 0 0>, rot=<0 0 0 -1>}
 clavicle_r {parent=spine_05, loc=<4.3575 -0.1466 2.3223>, rot=<0.7289 0.0465 -0.6829 0.009>}
 upperarm_r {parent=clavicle_r, loc=<-15.5111 0 0>, rot=<0.0226 -0.3234 -0.0088 -0.9459>}
 lowerarm_r {parent=upperarm_r, loc=<-26.6514 0 0>, rot=<0.0391 -0.0099 -0.2453 -0.9686>}
 upperarm_twist_01_r {parent=upperarm_r, loc=<-0.0041 0 0>, rot=<0 0 0 -1>}
]
variables [
 CrouchAlpha_Extreme {type=double, default=0}
 HeadRot {type=double, default=0}
 IsCrouching {type=bool, default=False}
 Skydive {type=bool, default=False}
]
